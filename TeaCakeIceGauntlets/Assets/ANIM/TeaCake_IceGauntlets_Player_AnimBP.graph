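# Player animation Blueprint for the TeaCake Ice Gauntlets, overriding Fortnite item animation layers for primary and secondary fire.
# ItemUpperBody runs a flag-driven left/right combo state machine whose entry functions clear their flags, adding a yaw aim offset while firing primary.
# Fire states layer blendspaces by stance using bone filters or masks.
# Other layers add pelvis twist with mesh-space upper-body compensation, pitch-driven secondary-fire additive and a BothHandFK curve.
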AnimBlueprint /TeaCakeIceGauntlets/Assets/ANIM/TeaCake_IceGauntlets_Player_AnimBP
target_skeleton=&Fortnite_M_Avg_Player_Skeleton
graphs [
 AnimGraph {nodes [Output Pose()]}
 EventGraph {links [], nodes []}
 ItemAimOffset {nodes [
   Input Pose!() -> (Pose, AimOffsetAlpha:double, Yaw:double, Pitch:double) {name=InPose_AimOffset}
   Input Pose!() -> (Pose) {name=InPose_PreAimOffsets}
   Output Pose() {layerGroup=Item Override}
  ]}
 ItemFallAdditive {nodes [
   Input Pose() {name=InPoseItemFallAdditive}
   Output Pose() {layerGroup=Item Override}
  ]}
 ItemFallLandAdditive {nodes [
   Input Pose() {name=InPoseItemFallLandAdditive}
   Output Pose() {layerGroup=Item Override}
  ]}
 ItemFallLevel2Override {nodes [
   Input Pose() {name=InPose_FallLevel2}
   Output Pose() {layerGroup=Item Override}
  ]}
 ItemFinalPoseOverride {nodes [
   n1 = Apply Additive(
    Base=Use cached pose 'InFinalPose'()
    Additive=Blendspace Player 'IceGauntlets_Secondary_BS'(X=Get PawnAimPitch() -> (PawnAimPitch:float), Y=0) {blendSpace=&/IceGauntlets_Secondary_BS}
   )
   n2 = Use cached pose 'InFinalPose'()
   n3 = Get bSecondaryFire() -> (bSecondaryFire:bool)
   Output Pose(Result=Modify Curve(SourcePose=Blend Poses by bool(BlendPose_0=@n1, BlendPose_1=@n2, BlendTime_0=0, BlendTime_1=0.2, bActiveValue=@n3), CurveValues_0=1, Alpha=1) {curveNames [BothHandFK], curveValues [0]}) {layerGroup=Item Override}
   Save cached pose 'InFinalPose'(Pose=Input Pose() {name=InFinalPose}) {cacheName=InFinalPose}
  ]}
 ItemFullBodyAimOffsets {nodes [
   Input Pose!() -> (Pose, FullBody_AimOffsetAlpha:double, Aim_Yaw:double, Aim_Pitch:double) {name=InPoseFullBodyAimOffsets}
   Input Pose!() -> (Pose) {name=InPosePreFullBodyAimOffsets}
   Output Pose() {layerGroup=Item Override}
  ]}
 ItemFullBodyOverride {nodes [
   n1 = Local To Component(LocalPose=Use cached pose 'InPoseFullBody'())
   n2 = Get CustomPelvisTwist() -> (CustomPelvisTwist:float)
   n3 = Layered blend per bone(
    BasePose=Component To Local(ComponentPose="Transform (Modify) Bone - Bone: pelvis"!(ComponentPose=@n1, Alpha=1, Translation="0,0,0", Rotation_Roll:float=@n2, Rotation_Pitch:float=0, Rotation_Yaw:float=0, Scale="1,1,1") -> (Pose) {rotationMode=BMM_Additive, rotationSpace=BCS_BoneSpace, boneToModify {boneName=pelvis}, rotation {roll=45}})
    BlendPoses_0=Use cached pose 'InPoseFullBody'()
   ) {
    bMeshSpaceRotationBlend=true
    layerSetup [{branchFilters [{blendDepth=3, boneName=spine_01}, {boneName=ik_hand_root}, {boneName=ik_foot_root}]}]
   }
   Output Pose(Result=Blend Poses by bool(
     BlendPose_0=@n3
     BlendPose_1=Use cached pose 'InPoseFullBody'()
     BlendTime_0=0.3
     BlendTime_1=0.3
     bActiveValue=Get bAllowCustomMeleeTwist() -> (bAllowCustomMeleeTwist:bool)
    )) {layerGroup=Item Override}
   Save cached pose 'InPoseFullBody'(Pose=Input Pose() {name=InPoseFullBody}) {cacheName=InPoseFullBody}
  ]}
 ItemIdleOverride {nodes [
   Input Pose() {name=InPose_Idle}
   Output Pose() {layerGroup=Item Override}
  ]}
 ItemJetPackJumpAdditive {nodes [Output Pose() {layerGroup=Item Override}]}
 ItemJetPackStartAdditive {nodes [Output Pose() {layerGroup=Item Override}]}
 ItemJumpLoopAdditive {nodes [
   Input Pose() {name=InPoseItemJumpLoopAdditive}
   Output Pose() {layerGroup=Item Override}
  ]}
 ItemJumpOverride {nodes [
   Input Pose() {name=InPoseJump}
   Output Pose() {layerGroup=Item Override}
  ]}
 ItemJumpUpAdditive {nodes [
   Input Pose() {name=InPoseItemJumpUpAdditive}
   Output Pose() {layerGroup=Item Override}
  ]}
 ItemPreAimOffsets {nodes [
   Input Pose() {name=InPosePreAimOffsets}
   Output Pose() {layerGroup=Item Override}
  ]}
 ItemPreIK {nodes [
   Input Pose() {name=InPose_PreIK}
   Output Pose() {layerGroup=Item Override}
  ]}
 ItemRightHandIK {nodes [
   Input Pose() {name=InPosePreRightHandIK}
   Input Pose() {name=InPosePostRightHandIK}
   Output Pose() {layerGroup=Item Override}
  ]}
 ItemSkeletalControl {nodes [
   Input Pose()
   Output Pose() {layerGroup=Item Override}
  ]}
 ItemSwimJumpFallLoop {nodes [Output Pose() {layerGroup=Item Override}]}
 ItemSwimJumpSurfaceEnd {nodes [Output Pose() {layerGroup=Item Override}]}
 ItemSwimJumpSurfaceLoop {nodes [Output Pose() {layerGroup=Item Override}]}
 ItemUpperBody {nodes [
   n1 = Use cached pose 'ComboFire_Pose'()
   n2 = Get PawnAimYaw() -> (PawnAimYaw:float)
   Output Pose(Result=Blend Poses by bool(
     BlendPose_0=AimOffset Player 'TeaCake_IceGauntlets_YawAO'!(X:float=@n2, Y:float=0, BasePose:Pose Link Structure=@n1, Alpha:float=1) -> (Pose:Pose Link Structure) {blendSpace=&/TeaCake_IceGauntlets_YawAO}
     BlendPose_1=Use cached pose 'ComboFire_Pose'()
     BlendTime_1=0.3
     bActiveValue=Get bIsFiringPrimary() -> (bIsFiringPrimary:bool)
    ) {blendType=ExpIn}) {layerGroup=Item Override}
   Save cached pose 'InPoseUpperBody'(Pose=Input Pose!() -> (Pose, InputParam:Fort Anim Input Adjusted Aim Structure) {name=InPoseUpperBody}) {cacheName=InPoseUpperBody}
   AnimGraphNode_StateMachine_0 = ComboFire_SM() {is=title AnimGraphNode_StateMachine}
   Save cached pose 'ComboFire_Pose'(Pose=@AnimGraphNode_StateMachine_0) {cacheName=ComboFire_Pose}
  ]}
 ComboFire_SM {
  entry=Default
  path=ItemUpperBody.AnimGraphNode_StateMachine_0.ComboFire_SM
  ids=id
  nodes [
   AnimStateNode_0 = Default() {is=title AnimStateNode}
   AnimStateNode_1 = Fire_Left() {is=title AnimStateNode}
   AnimStateNode_2 = Fire_Right() {is=title AnimStateNode}
   AnimStateTransitionNode_0 = AnimStateTransitionNode() {bAllowInertializationForSelfTransitions=true}
   AnimStateTransitionNode_1 = AnimStateTransitionNode() {bAllowInertializationForSelfTransitions=true}
   AnimStateTransitionNode_2 = AnimStateTransitionNode() {bAllowInertializationForSelfTransitions=true, bAutomaticRuleBasedOnSequencePlayerInState=true, crossfadeDuration=0.5}
   AnimStateTransitionNode_3 = AnimStateTransitionNode() {bAllowInertializationForSelfTransitions=true, bAutomaticRuleBasedOnSequencePlayerInState=true, crossfadeDuration=0.5}
   AnimStateTransitionNode_4 = AnimStateTransitionNode() {bAllowInertializationForSelfTransitions=true}
   AnimStateTransitionNode_5 = AnimStateTransitionNode() {bAllowInertializationForSelfTransitions=true}
  ]
  links [
   {
    dst=Fire_Left
    rule=ItemUpperBody.AnimGraphNode_StateMachine_0.ComboFire_SM.AnimStateTransitionNode_0.Transition
    src=Default
    via=AnimStateTransitionNode_0
   }
   {
    dst=Fire_Right
    rule=ItemUpperBody.AnimGraphNode_StateMachine_0.ComboFire_SM.AnimStateTransitionNode_1.Transition
    src=Default
    via=AnimStateTransitionNode_1
   }
   {
    dst=Default
    rule=ItemUpperBody.AnimGraphNode_StateMachine_0.ComboFire_SM.AnimStateTransitionNode_3.Transition
    src=Fire_Left
    via=AnimStateTransitionNode_3
   }
   {
    dst=Fire_Right
    rule=ItemUpperBody.AnimGraphNode_StateMachine_0.ComboFire_SM.AnimStateTransitionNode_5.Transition
    src=Fire_Left
    via=AnimStateTransitionNode_5
   }
   {
    dst=Default
    rule=ItemUpperBody.AnimGraphNode_StateMachine_0.ComboFire_SM.AnimStateTransitionNode_2.Transition
    src=Fire_Right
    via=AnimStateTransitionNode_2
   }
   {
    dst=Fire_Left
    rule=ItemUpperBody.AnimGraphNode_StateMachine_0.ComboFire_SM.AnimStateTransitionNode_4.Transition
    src=Fire_Right
    via=AnimStateTransitionNode_4
   }
  ]
 }
 Default {
  path=ItemUpperBody.AnimGraphNode_StateMachine_0.ComboFire_SM.AnimStateNode_0.Default
  nodes [Output Animation Pose(Result=Use cached pose 'InPoseUpperBody'())]
 }
 Fire_Left {
  path=ItemUpperBody.AnimGraphNode_StateMachine_0.ComboFire_SM.AnimStateNode_1.Fire_Left
  nodes [
   Layered blend per bone(BlendPoses_0=Blendspace Player 'IceGauntlets_Fire_Left_BS'(X=0, Y=0) {blendSpace=&/IceGauntlets_Fire_Left_BS, groupName=FireLeft, method=SyncGroup}) {layerSetup [{branchFilters [{boneName=spine_04}]}]}
   n1 = Blendspace Player 'IceGauntlets_Fire_Left_BS'(X=0, Y=0) {blendSpace=&IceGauntlets_Fire_Left_BS, groupName=FireLeft, method=SyncGroup}
   n2 = Use cached pose 'InPoseUpperBody'()
   Output Animation Pose(Result=Blend Poses by bool(
     BlendPose_0=Layered blend per bone(BasePose=@n2, BlendPoses_0=@n1) {bMeshSpaceRotationBlend=true, layerSetup [{branchFilters [{boneName=clavicle_l}]}]}
     BlendPose_1=Blendspace Player 'IceGauntlets_Fire_Left_BS'(X=0, Y=0) {blendSpace=&IceGauntlets_Fire_Left_BS, groupName=FireLeft, method=SyncGroup}
     bActiveValue=Get bIsCrouching()
    )) {
    node {stateEntryFunction {functionName=OnEnter_FireLeft}}
    stateEntryFunction {bSelfContext=true, memberGuid=#1, memberName=OnEnter_FireLeft}
   }
  ]
 }
 Fire_Right {
  path=ItemUpperBody.AnimGraphNode_StateMachine_0.ComboFire_SM.AnimStateNode_2.Fire_Right
  nodes [
   n1 = Blendspace Player 'IceGauntlets_Fire_Right_BS'(X=0, Y=0) {blendSpace=&/IceGauntlets_Fire_Right_BS, groupName=FireRight, method=SyncGroup}
   n2 = Layered blend per bone(
    BasePose=Blendspace Player 'IceGauntlets_Fire_Right_BS'(X=0, Y=0) {blendSpace=&IceGauntlets_Fire_Right_BS, groupName=FireRight, method=SyncGroup}
    BlendPoses_0=Use cached pose 'InPoseUpperBody'()
   ) {
    bMeshSpaceRotationBlend=true
    blendMode=BlendMask
    blendMasks [&Fortnite_M_Avg_Player_Skeleton:LowerBodyPartialUpperBody]
    layerSetup []
   }
   n3 = Get bIsSurfaceSwimming()
   n4 = Layered blend per bone(
    BasePose=Blendspace Player 'IceGauntlets_Fire_Right_BS'(X=0, Y=0) {blendSpace=&IceGauntlets_Fire_Right_BS, groupName=FireRight, method=SyncGroup}
    BlendPoses_0=Use cached pose 'InPoseUpperBody'()
   ) {
    bMeshSpaceRotationBlend=true
    blendMode=BlendMask
    blendMasks [&Fortnite_M_Avg_Player_Skeleton:LowerBodyPartialUpperBody]
    layerSetup []
   }
   Output Animation Pose(Result=Blend Poses by bool(BlendPose_0=@n4, BlendPose_1=Blend Poses by bool(BlendPose_0=@n2, BlendPose_1=@n1, bActiveValue=@n3), bActiveValue=Get bIsCrouching())) {
    node {stateEntryFunction {functionName=OnStateEntry}}
    stateEntryFunction {bSelfContext=true, memberGuid=#2, memberName=OnEnter_FireRight}
   }
  ]
 }
 Transition {
  path=ItemUpperBody.AnimGraphNode_StateMachine_0.ComboFire_SM.AnimStateTransitionNode_0.Transition
  nodes [Result(bCanEnterTransition=Get IsFiringLeft() -> (IsFiringLeft:bool))]
 }
 Transition {
  path=ItemUpperBody.AnimGraphNode_StateMachine_0.ComboFire_SM.AnimStateTransitionNode_1.Transition
  nodes [Result(bCanEnterTransition=Get IsFiringRight() -> (IsFiringRight:bool))]
 }
 Transition {path=ItemUpperBody.AnimGraphNode_StateMachine_0.ComboFire_SM.AnimStateTransitionNode_2.Transition, nodes [Result()]}
 Transition {path=ItemUpperBody.AnimGraphNode_StateMachine_0.ComboFire_SM.AnimStateTransitionNode_3.Transition, nodes [Result()]}
 Transition {
  path=ItemUpperBody.AnimGraphNode_StateMachine_0.ComboFire_SM.AnimStateTransitionNode_4.Transition
  nodes [Result(bCanEnterTransition=Get IsFiringLeft())]
 }
 Transition {
  path=ItemUpperBody.AnimGraphNode_StateMachine_0.ComboFire_SM.AnimStateTransitionNode_5.Transition
  nodes [Result(bCanEnterTransition=Get IsFiringRight())]
 }
 ItemZipLineStartAdditive {nodes [Output Pose() {layerGroup=Item Override}]}
 OnEnter_FireLeft {nodes [
   OnEnter_FireLeft!() -> (then, Context:"Anim Update Context Structure (by ref)", Node:"Anim Node Reference Structure (by ref)") {is=title K2Node_FunctionEntry}
   > Set IsFiringLeft(IsFiringLeft:bool=false) -> (Output_Get:bool)
   > Return Node()
  ]}
 OnEnter_FireRight {nodes [
   OnEnter_FireRight!() -> (then, Context:"Anim Update Context Structure (by ref)", Node:"Anim Node Reference Structure (by ref)") {is=title K2Node_FunctionEntry}
   > Set IsFiringRight(IsFiringRight:bool=false) -> (Output_Get:bool)
   > Return Node()
  ]}
]
variables [
 IsFiringLeft {type=boolean}
 IsFiringRight {type=boolean}
]
