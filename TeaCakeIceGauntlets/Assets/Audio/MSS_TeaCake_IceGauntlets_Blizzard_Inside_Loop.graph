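# Stereo MetaSound source for the interior loop of the TeaCake Ice Gauntlets blizzard.
# On play, the loop wave starts at a random offset and fades in through an ADSR envelope.
# A random start one-shot, from an array empty by default, gets randomized pitch and volume plus distance-curve attenuation.
# A two-input mixer combines both layers, with only the start layer receiving the in-graph attenuation.
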MetaSoundSource /TeaCakeIceGauntlets/Assets/Audio/MSS_TeaCake_IceGauntlets_Blizzard_Inside_Loop
inputs [
 Loop Sound {type=WaveAsset}
 Random.Pitch.Base {type=Float}
 Random.Pitch.Spread {type=Float}
 Random.Volume.Base {type=Float}
 Random.Volume.Spread {type=Float}
 Start Sounds {type=WaveAsset:Array}
 UE.Attenuation.Distance {type=Float}
 UE.Source.OnPlay {type=Trigger}
]
interfaces [UE.Attenuation:1.0, UE.OutputFormat.Stereo:1.0, UE.Source:1.1]
outputs [
 "UE.OutputFormat.Stereo.Audio:0" {type=Audio}
 "UE.OutputFormat.Stereo.Audio:1" {type=Audio}
 UE.Source.OneShot.OnFinished {type=Trigger}
]
pages [Default {
  input_defaults {
   Loop Sound=/TeaCakeIceGauntlets/Assets/Audio/Wavs/TeaCake_IceGauntlets_Blizz_Inside_Loop_01
   Random.Pitch.Base=0
   Random.Pitch.Spread=0
   Random.Volume.Base=0
   Random.Volume.Spread=0
   UE.Attenuation.Distance=0
   UE.Source.OnPlay=null
   Start Sounds []
  }
  nodes [
   n1 = &MSP_Attenuation_Volume_Curve(Distance=^UE.Attenuation.Distance) {also {defaults {AttenuationCurves=&/WTB_TeaCake_IceGauntlets, CurveIndexToEvaluate=2, FalloffDistance=3000, InnerRadius=800}}}
   n2 = AbsoluteValue.Abs:Float(Input=^Random.Volume.Spread)
   DecibelstoLinearGain_0 = UE.Decibels to Linear Gain:Float(Decibels=UE.RandomFloat(
     Next=^UE.Source.OnPlay
     Max=UE.Add:Float(PrimaryOperand=^Random.Volume.Base, AdditionalOperands=@n2.Value).Out
     Min=UE.Subtract:Float(PrimaryOperand=^Random.Volume.Base, AdditionalOperands=@n2.Value).Out
    ).Value)
   ArrayRandomGet_0 = Array.Random Get:WaveAsset:Array(In Array=^Start Sounds, Next=^UE.Source.OnPlay) {also {defaults {No Repeats=-1, Weights []}}}
   n3 = AbsoluteValue.Abs:Float(Input=^Random.Pitch.Spread)
   WavePlayer_1 = UE.Wave Player:Stereo(
    Play=@ArrayRandomGet_0.On Next
    Wave Asset=@ArrayRandomGet_0.Value
    Pitch Shift=UE.RandomFloat(
     Next=^UE.Source.OnPlay
     Max=UE.Add:Float(PrimaryOperand=^Random.Pitch.Base, AdditionalOperands=@n3.Value).Out
     Min=UE.Subtract:Float(PrimaryOperand=^Random.Pitch.Base, AdditionalOperands=@n3.Value).Out
    ).Value
   ) {also {version [1,1]}}
   ADSREnvelopeADSREnvelope_0 = ADSR Envelope.ADSR Envelope:Float(Trigger Attack=^UE.Source.OnPlay) {also {defaults {Attack Time=0.4, Sustain Level=1}}}
   GetWaveDuration_0 = "UE.Get Wave Duration: "(Wave=^Loop Sound)
   WavePlayer_0 = UE.Wave Player:Stereo(Wave Asset=^Loop Sound, Play=^UE.Source.OnPlay, Start Time=UE.RandomTime(Next=^UE.Source.OnPlay, Max=@GetWaveDuration_0.Duration).Value) {also {defaults {Loop="<MetaSoundBuilderResult.SUCCEEDED: 0>"}, version [1,1]}}
   AudioMixerAudioMixerStereo2_0 = "AudioMixer.Audio Mixer (Stereo, 2)"(
    In 0 L=UE.Multiply:Audio by Float(
     AdditionalOperands=@n1.VolumeMultiplier
     PrimaryOperand=UE.Multiply:Audio by Float(AdditionalOperands=@DecibelstoLinearGain_0.Linear Gain, PrimaryOperand=@WavePlayer_1.Out Left).Out
    ).Out
    In 0 R=UE.Multiply:Audio by Float(
     AdditionalOperands=@n1.VolumeMultiplier
     PrimaryOperand=UE.Multiply:Audio by Float(AdditionalOperands=@DecibelstoLinearGain_0.Linear Gain, PrimaryOperand=@WavePlayer_1.Out Right).Out
    ).Out
    In 1 R=UE.Multiply:Audio by Float(AdditionalOperands=@ADSREnvelopeADSREnvelope_0.Out Envelope, PrimaryOperand=@WavePlayer_0.Out Right).Out
    In 1 L=UE.Multiply:Audio by Float(AdditionalOperands=@ADSREnvelopeADSREnvelope_0.Out Envelope, PrimaryOperand=@WavePlayer_0.Out Left).Out
   ) {also {defaults {Gain 0=1.4, Gain 1=0.8}}}
  ]
  links [
   {dst="@UE.OutputFormat.Stereo.Audio:0", src=AudioMixerAudioMixerStereo2_0.Out L}
   {dst="@UE.OutputFormat.Stereo.Audio:1", src=AudioMixerAudioMixerStereo2_0.Out R}
   {dst="@UE.Source.OneShot.OnFinished", src=WavePlayer_0.On Finished}
  ]
 }]
properties {
 assetImportData=&self:AssetImportData
 bLooping=true
 duration=10000
 numChannels=2
 outputFormat=Stereo
 sampleRate=48000
 soundClassObject=&Fort_WeaponRanged_Fire_3P_PSM
 volume=0.5
 concurrencySet [
  &/CONC_TeaCake_IceGauntlets_Blizzard_Global
  &/CONC_TeaCake_IceGauntlets_Blizzard_Local
 ]
}
