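# MetaSound source preset for the explosion sound of the TeaCake Ice Gauntlets primary projectile, built on the shared MSS_Projectile_Explosion weapon graph.
# It supplies close and distant wave arrays, which the parent's inputs pair with a distance-curve bank, air-absorption low-pass range and affiliation-based levels.
# Output is one-shot stereo with a no-attenuation asset, limited concurrency, the explosives sound class and a send to the weapon sidechain bus.
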MetaSoundSource /TeaCakeIceGauntlets/Assets/Audio/MSS_TeaCake_IceGauntlets_Primary_Proj_Explosion
preset_parent=/FortMetaSoundWeapon/Items/Metasounds/Projectiles/MSS_Projectile_Explosion
inputs [
 Affiliation.Enemy {type=Float}
 Affiliation.Friendly {type=Float}
 Affiliation.Local {type=Float}
 Character.Affiliation {type=Enum:FortAudioAffiliation}
 Character.Health {type=Float}
 Character.UniqueID {type=Int32}
 DistantAirAbsorption.LowPassMax {type=Float}
 DistantAirAbsorption.LowPassMin {type=Float}
 DistantAirAbsorption.MaxDistance {type=Float}
 DistantAirAbsorption.MinDistance {type=Float}
 Explosion.CloseVolume {type=Float}
 Explosion.DistanceCurves {type=WaveTableBankAsset}
 Explosion.DistantVolume {type=Float}
 Explosion.FalloffDistance {type=Float}
 Explosion.InnerRadius {type=Float}
 Explosion.OutputVolume {type=Float}
 "Random.Max Pitch (Semitones)" {type=Float}
 Random.MinPitch {type=Float}
 Random.VolumeMax {type=Float}
 Random.VolumeMin {type=Float}
 Sounds.Close {type=WaveAsset:Array}
 Sounds.Distant {type=WaveAsset:Array}
 UE.Attenuation.Distance {type=Float}
 UE.Source.OnPlay {type=Trigger}
]
interfaces [Fort.Character:1.0, UE.Attenuation:1.0, UE.OutputFormat.Stereo:1.0, UE.Source.OneShot:1.0, UE.Source:1.1]
outputs [
 "UE.OutputFormat.Stereo.Audio:0" {type=Audio}
 "UE.OutputFormat.Stereo.Audio:1" {type=Audio}
 UE.Source.OneShot.OnFinished {type=Trigger}
]
pages [Default {input_defaults {
   Affiliation.Enemy=0
   Affiliation.Friendly=-6
   Affiliation.Local=0
   Character.Affiliation=0
   Character.Health=0
   Character.UniqueID=0
   DistantAirAbsorption.LowPassMax=6000
   DistantAirAbsorption.LowPassMin=20000
   DistantAirAbsorption.MaxDistance=27500
   DistantAirAbsorption.MinDistance=8000
   Explosion.CloseVolume=0
   Explosion.DistanceCurves=/FortMetaSoundWeapon/Items/Shared/Settings/Projectiles/WTB_Projectile_Explosion_Distance
   Explosion.DistantVolume=0
   Explosion.FalloffDistance=27500
   Explosion.InnerRadius=0
   Explosion.OutputVolume=0
   "Random.Max Pitch (Semitones)"=0
   Random.MinPitch=0
   Random.VolumeMax=0
   Random.VolumeMin=0
   UE.Attenuation.Distance=14519.9922
   UE.Source.OnPlay=null
   Sounds.Close [
    /TeaCakeIceGauntlets/Assets/Audio/Wavs/TeaCake_IceGauntlets_Primary_Explo_01
    /TeaCakeIceGauntlets/Assets/Audio/Wavs/TeaCake_IceGauntlets_Primary_Explo_02
    /TeaCakeIceGauntlets/Assets/Audio/Wavs/TeaCake_IceGauntlets_Primary_Explo_03
   ]
   Sounds.Distant [
    /TeaCakeIceGauntlets/Assets/Audio/Wavs/TeaCake_IceGauntlets_Primary_Explo_Distant_01
    /TeaCakeIceGauntlets/Assets/Audio/Wavs/TeaCake_IceGauntlets_Primary_Explo_Distant_02
    /TeaCakeIceGauntlets/Assets/Audio/Wavs/TeaCake_IceGauntlets_Primary_Explo_Distant_03
    /TeaCakeIceGauntlets/Assets/Audio/Wavs/TeaCake_IceGauntlets_Primary_Explo_Distant_04
   ]
  }}]
properties {
 assetImportData=&self:AssetImportData
 attenuationSettings=/FortMetaSoundWeapon/Items/Shared/Settings/Projectiles/ATT_Projectile_Explosion_NoAtt
 bOverrideConcurrency=true
 maxDistance=27500
 numChannels=2
 outputFormat=Stereo
 sampleRate=48000
 soundClassObject=/Game/Sounds/FortSoundClassesAndModes/Fort_Explosives
 volume=0.6
 busSends [{audioBus=&AB_Sidechain_FortWeapon_Vol, sendLevel=1}]
 concurrencyOverrides {maxCount=4}
}
