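# Sound Cue for the end of the TeaCake Ice Gauntlets blizzard effect.
# A Random node picks one of three end-variation waves with equal weight.
# It plays at slightly reduced volume, spatialized with the borrowed smoke grenade attenuation, on an outdoor-ducking ambient sound class.
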SoundCue /TeaCakeIceGauntlets/Assets/Audio/SC_TeaCake_IceGauntlets_Blizzard_End
root=Random_0
nodes [
 {id=Random_0, children [WavePlayer_0,WavePlayer_1,WavePlayer_2], properties {weights [1,1,1]}}
 {id=WavePlayer_0, properties {soundWaveAssetPtr=/TeaCakeIceGauntlets/Assets/Audio/Wavs/TeaCake_IceGauntlets_Blizzard_End_01}}
 {id=WavePlayer_1, properties {soundWaveAssetPtr=/TeaCakeIceGauntlets/Assets/Audio/Wavs/TeaCake_IceGauntlets_Blizzard_End_02}}
 {id=WavePlayer_2, properties {soundWaveAssetPtr=/TeaCakeIceGauntlets/Assets/Audio/Wavs/TeaCake_IceGauntlets_Blizzard_End_03}}
]
properties {
 attenuationSettings=/Game/Sounds/Attenuation/Fort_Grenade_Smoke_Att
 maxDistance=6800
 soundClassObject=/Game/Sounds/FortSoundClassesAndModes/Fort_Ambient_DuckOutdoors_Athena_PSM
 volumeMultiplier=0.8
}
