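# Sound cue for the break sound of the TeaCake Ice Gauntlets.
# A single wave player is the root and plays one UI break wave, with no randomization or modulation.
# It uses the Fortnite weapon foley attenuation and sound class, so it is set up as positional weapon foley despite the wave's UI name.
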SoundCue /TeaCakeIceGauntlets/Assets/Audio/SC_TeaCake_IceGauntlets_Break
root=WavePlayer_0
nodes [{id=WavePlayer_0, properties {soundWaveAssetPtr=/TeaCakeIceGauntlets/Assets/Audio/Wavs/TeaCake_IceGauntlets_UI_Break_01}}]
properties {
 attenuationSettings=&AS_Fort_WeaponFoley
 maxDistance=2250
 soundClassObject=&Fort_WeaponFoley
}
