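# Sound cue for the TeaCake Ice Gauntlets secondary charge, randomly playing one of three equally weighted 1P waves.
# An "islocal" branch sends the local player through the 1P sound class, while remote players use the 3P class with charge attenuation.
# For remote players, an "isenemy" branch boosts enemy or unset volume and halves non-enemy volume.
# An unset "islocal" bypasses the class override, and gunfire concurrency sets limit playback.
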SoundCue /TeaCakeIceGauntlets/Assets/Audio/SC_TeaCake_IceGauntlets_Secondary_Charge
root=Branch_2
nodes [
 {id=Attenuation_0, children [Random_0], properties {attenuationSettings=&/ATT_Teacake_IceGauntlets_Charge}}
 {id=Attenuation_2, children [Modulator_2]}
 {id=Branch_2, children [SoundClass_2,Branch_3,Modulator_2], properties {boolParameterName=islocal}}
 {id=Branch_3, children [Modulator_4,Modulator_3,Modulator_4], properties {boolParameterName=isenemy}}
 {id=Modulator_2, children [Random_0], properties {pitchMax=1, pitchMin=1, volumeMax=1, volumeMin=1}}
 {id=Modulator_3, children [SoundClass_0], properties {pitchMax=1, pitchMin=1, volumeMax=0.5, volumeMin=0.5}}
 {id=Modulator_4, children [SoundClass_0], properties {pitchMax=1, pitchMin=1, volumeMax=1.3, volumeMin=1.3}}
 {id=Random_0, children [WavePlayer_1,WavePlayer_2,WavePlayer_3], properties {weights [1,1,1]}}
 {
  id=SoundClass_0
  children [Attenuation_0]
  properties {soundClassOverride=&Fort_WeaponRanged_Fire_3P_PSM}
 }
 {
  id=SoundClass_2
  children [Attenuation_2]
  properties {soundClassOverride=&Fort_WeaponRanged_Fire_1P_PSM}
 }
 {id=WavePlayer_1, properties {soundWaveAssetPtr=/TeaCakeIceGauntlets/Assets/Audio/Wavs/TeaCake_IceGauntlets_Secondary_Charge_1P_01}}
 {id=WavePlayer_2, properties {soundWaveAssetPtr=/TeaCakeIceGauntlets/Assets/Audio/Wavs/TeaCake_IceGauntlets_Secondary_Charge_1P_02}}
 {id=WavePlayer_3, properties {soundWaveAssetPtr=/TeaCakeIceGauntlets/Assets/Audio/Wavs/TeaCake_IceGauntlets_Secondary_Charge_1P_03}}
]
properties {
 bDebug=true
 maxDistance=2097152
 soundClassObject=&Fort_WeaponRanged_Fire_1P_PSM
 volumeMultiplier=0.9
 concurrencySet [&Weapon_Gunfire_Local_Conc, &Weapon_Gunfire_Global_Conc]
}
