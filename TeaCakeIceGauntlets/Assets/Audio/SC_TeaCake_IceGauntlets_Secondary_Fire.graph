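# Sound cue for the secondary fire of the TeaCake Ice Gauntlets, limited by local and global gunfire concurrency sets.
# An islocal branch plays one of three random 1P waves under the 1P sound class for the local player.
# Otherwise an isenemy branch plays non-enemies at half volume, feeding a mix of two 3P-class layers.
# The close layer reuses the 1P waves, and the other plays random distant waves with medium mono attenuation.
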SoundCue /TeaCakeIceGauntlets/Assets/Audio/SC_TeaCake_IceGauntlets_Secondary_Fire
root=Branch_2
nodes [
 {id=Attenuation_0, children [Random_1], properties {attenuationSettings=&Fort_Weapon_Projectile_Close_Athena_Att}}
 {id=Attenuation_2, children [Modulator_2]}
 {
  id=Attenuation_3
  children [Random_0]
  properties {attenuationSettings=&Fort_Weapon_Projectile_Medium_Mono_Athena_Att}
 }
 {id=Branch_2, children [SoundClass_2,Branch_3,Modulator_2], properties {boolParameterName=islocal}}
 {id=Branch_3, children [Modulator_4,Modulator_3,Modulator_4], properties {boolParameterName=isenemy}}
 {id=Mixer_0, children [SoundClass_0,SoundClass_1], properties {inputVolume [1,1]}}
 {id=Modulator_2, children [Random_1], properties {pitchMax=1, pitchMin=1, volumeMax=1, volumeMin=1}}
 {id=Modulator_3, children [Mixer_0], properties {pitchMax=1, pitchMin=1, volumeMax=0.5, volumeMin=0.5}}
 {id=Modulator_4, children [Mixer_0], properties {pitchMax=1, pitchMin=1, volumeMax=1, volumeMin=1}}
 {id=Random_0, children [WavePlayer_15, WavePlayer_16, WavePlayer_17], properties {weights [1,1,1]}}
 {id=Random_1, children [WavePlayer_9, WavePlayer_10, WavePlayer_11], properties {weights [1,1,1]}}
 {
  id=SoundClass_0
  children [Attenuation_0]
  properties {soundClassOverride=&Fort_WeaponRanged_Fire_3P_PSM}
 }
 {
  id=SoundClass_1
  children [Attenuation_3]
  properties {soundClassOverride=&Fort_WeaponRanged_Fire_3P_PSM}
 }
 {
  id=SoundClass_2
  children [Attenuation_2]
  properties {soundClassOverride=&Fort_WeaponRanged_Fire_1P_PSM}
 }
 {id=WavePlayer_10, properties {soundWaveAssetPtr=/TeaCakeIceGauntlets/Assets/Audio/Wavs/TeaCake_IceGauntlets_Secondary_Fire_1P_02}}
 {id=WavePlayer_11, properties {soundWaveAssetPtr=/TeaCakeIceGauntlets/Assets/Audio/Wavs/TeaCake_IceGauntlets_Secondary_Fire_1P_03}}
 {id=WavePlayer_15, properties {soundWaveAssetPtr=/TeaCakeIceGauntlets/Assets/Audio/Wavs/TeaCake_IceGauntlets_Secondary_Fire_3P_Distant_01}}
 {id=WavePlayer_16, properties {soundWaveAssetPtr=/TeaCakeIceGauntlets/Assets/Audio/Wavs/TeaCake_IceGauntlets_Secondary_Fire_3P_Distant_02}}
 {id=WavePlayer_17, properties {soundWaveAssetPtr=/TeaCakeIceGauntlets/Assets/Audio/Wavs/TeaCake_IceGauntlets_Secondary_Fire_3P_Distant_03}}
 {id=WavePlayer_9, properties {soundWaveAssetPtr=/TeaCakeIceGauntlets/Assets/Audio/Wavs/TeaCake_IceGauntlets_Secondary_Fire_1P_01}}
]
properties {
 bDebug=true
 maxDistance=2097152
 soundClassObject=&Fort_WeaponRanged_Fire_1P_PSM
 volumeMultiplier=1.1
 concurrencySet [&Weapon_Gunfire_Local_Conc, &Weapon_Gunfire_Global_Conc]
}
