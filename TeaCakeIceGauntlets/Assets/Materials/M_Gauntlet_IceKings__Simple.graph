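# Material for the Ice Kings gauntlets, layering animated frost and glow over diffuse, mask, normal and emissive textures.
# Ice masks combine gradient, left/right, palm and noise terms for MF_Ice_TechArt, with static switches for static versus skinned meshes.
# Emissive math blends pulsing noise and Fresnel between near and far settings by pixel depth.
# An "Is Pickup?" switch selects a potion/hologram branch with vertex animation before shared character and Fortnite functions.
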Material /TeaCakeIceGauntlets/Assets/Materials/M_Gauntlet_IceKings__Simple
graph {
 properties {
  bAutomaticallySetUsageInEditor=false
  bUseMaterialAttributes=true
  bUsedWithClothing=true
  bUsedWithMeshDeformer=true
  bUsedWithMorphTargets=true
  bUsedWithNanite=true
  bUsedWithSkeletalMesh=true
  pixelDepthOffsetMode=PDOM_Legacy
  refractionMethod=RM_IndexOfRefraction
  usedShadingModels=Default Lit
 }
 nodes [
  n1 = TextureSampleParameter2D() {
   group=Specular
   parameterName=SpecularMasks
   samplerType=SAMPLERTYPE_Masks
   texture=&T_Fortnite_Default_S
  }
  LinearInterpolate(
   A=ScalarParameter() {defaultValue=0.1, group=Specular, parameterName=SpecRoughnessMin, sliderMax=1}
   B=ScalarParameter() {defaultValue=1, group=Specular, parameterName=SpecRoughnessMax, sliderMax=1}
   Alpha=@n1.B
  )
  n2 = VectorParameter() {group=Emissive, parameterName=Emissive Texture UVs RG_TopLeft BA_BottomRight, defaultValue=<rgba 0 0 1 1>}
  MaterialFunctionCall(
   Hue Shift Percentage=ScalarParameter() {group=Emissive, parameterName=Hue Shift Emissive, sliderMax=1}
   Texture=MaterialFunctionCall!(
    UVs
    TextureIn=TextureObjectParameter() {group=Emissive, parameterName=EmissiveTexture, texture=&T_BlacksRGB}
    UpperLeftCorner=ComponentMask(None=@n2.RGB) {g=true, r=true}
    LowerRightCorner=AppendVector(A=@n2.B, B=@n2.B)
    ExportFloat4
   ) {materialFunction=/Engine/Functions/Engine_MaterialFunctions01/Texturing/TextureCropping}.CroppedMasked
  ) {materialFunction=&HueShift}
  n3 = GetMaterialAttributes(None=NamedRerouteUsage()) {attributeGetTypes [
    #1
    #2
    #3
    #4
    #5
    #6
   ]}
  n4 = Multiply(A=@n3.Emissive Color, B=VectorParameter() {group=Emissive, parameterName=EmissiveMultiplier, defaultValue=<rgba 0 0 0 1>}.RGB)
  n5 = TextureSampleParameter2D() {parameterName=Diffuse, texture=&T_BlacksRGB}
  n6 = ReflectionVectorWS()
  n7 = ConstantBiasScale(None=ComponentMask(None=Transform(None=@n6) {transformSourceType=TRANSFORMSOURCE_World, transformType=TRANSFORM_View}) {g=true, r=true})
  n8 = Constant2Vector() {g=-1, r=1}
  n9 = Multiply(
   A=TextureSampleParameter2D(UVs=Multiply(A=@n7, B=@n8)) {parameterName=SphereReflectionMap}.RGB
   B=ScalarParameter() {defaultValue=1, group=reflection, parameterName=Reflection Multiplier}
  )
  n10 = TextureSampleParameter2D() {parameterName=ReflectionTextureMask, samplerType=SAMPLERTYPE_Masks, texture=&T_Fortnite_Default_S}
  n11 = StaticComponentMaskParameter(None=@n10.RGB) {defaultR=true, group=reflection, parameterName=ReflectionMaskChannel}
  MaterialFunctionCall!(
   DiffuseFresnelAmount=ScalarParameter() {defaultValue=0.75, group=Diffuse, parameterName=DiffuseFresnelAmount}
   Emissive=@n4
   Base Color=StaticSwitchParameter(True=LinearInterpolate(A=@n5.RGB, B=@n9, Alpha=@n11), False=@n5.RGB) {group=reflection, parameterName=Use Reflection}
  ) {materialFunction=/Game/Packages/Fortress_SharedMaterials/Base_Material_Functions/FresnelFunction}
  n12 = StaticSwitchParameter(
   True=MaterialFunctionCall!(
    AverageHandPosition=VectorParameter() {group=Super Charge, parameterName=AverageHandPosition, defaultValue=<rgba 0 0 0 1>}.RGB
    SuperChargedFlickerBias=ScalarParameter() {defaultValue=1, group=Super Charge, parameterName=SuperChargedFlickerBias}
    SuperChargedEffect=ScalarParameter() {group=Super Charge, parameterName=SuperChargedEffect}
    UseDistanceCalc=StaticBool()
    Emissive=@n4
   ) {materialFunction=/Game/Characters/Fortress_CharacterShader/SuperChargedGlow}.Result
   False=@n4
  ) {desc=e, group=Super Charge, parameterName=Use Super Charged Effect}
  n13 = Fresnel(ExponentIn=ScalarParameter() {defaultValue=0.5, group=Emissive, parameterName=EmissveFresnelExponent})
  n14 = Multiply(A=@n12, B=StaticSwitchParameter(True=OneMinus(None=@n13), False=@n13) {group=Emissive, parameterName=InvertEmissiveFresnel})
  n15 = VectorParameter() {parameterName=NormalizedRotationAxis, defaultValue=<rgba 0 1 0 0>}
  n16 = Sine(None=Multiply(
    A=Time() {bIgnorePause=true, bOverride_Period=true, period=0.3333}
    B=ScalarParameter() {defaultValue=3, parameterName=WiggleOnInteractSpeed}
   ))
  n17 = ScalarParameter() {parameterName=WiggleOnInteractAmount}
  n18 = Multiply(
   A=RotateAboutAxis(
    NormalizedRotationAxis=Transform(None=@n15.RGB) {transformSourceType=TRANSFORMSOURCE_Local}
    RotationAngle=Multiply(A=@n16, B=@n17)
    PivotPoint=ActorPositionWS()
    Position=WorldPosition().XYZ
   )
   B=MaterialFunctionCall!(Center, Height) {materialFunction=/Game/Effects/Fort_Effects/Materials/_Functions/MF_Z_GradientFromObjectHeight_01}.Result
  )
  n19 = VectorParameter() {parameterName=WorldPostionOfBlackHole, defaultValue=<rgba -30000 25536 1952 0>}
  n20 = MaterialFunctionCall() {materialFunction=/Engine/Functions/Engine_MaterialFunctions02/PivotPainter/PivotPainter_PerObjectData}
  n21 = Time()
  n22 = ScalarParameter() {defaultValue=0.5, group=Animation, parameterName=Anim Speed}
  n23 = StaticSwitchParameter(
   True=MaterialFunctionCall!(
    Sine Phase=Multiply(A=@n21, B=@n22) {constB=0.5}
    Value 1=ScalarParameter() {defaultValue=0.065, group=Animation, parameterName=Sin Value 1}
    Value 2=ScalarParameter() {defaultValue=-0.02, group=Animation, parameterName=Sin Value 2}
   ) {materialFunction=/Engine/Functions/Engine_MaterialFunctions02/Utility/Sine_Remapped}.Remapped Sine
   False=Multiply(A=Time(), B=ScalarParameter() {defaultValue=0.5, group=Animation, parameterName=Anim Speed})
  ) {group=Animation, parameterName=Sin Anim}
  n24 = Multiply(
   A=@n20.Black Mask
   B=RotateAboutAxis(NormalizedRotationAxis=@n20.X-Axis Vector, RotationAngle=@n23, PivotPoint=@n20.X-Axis Vector, Position=WorldPosition().XYZ)
  )
  n25 = WorldPosition()
  n26 = MaterialFunctionCall!(
   World Position=StaticSwitchParameter(True=Add(A=@n24, B=@n25.XYZ), False=@n25.XYZ) {group=Animation, parameterName=Animate}
   DistanceCullFadeAnimationOffset=MaterialFunctionCall() {materialFunction=/Game/Packages/Fortress_SharedMaterials/Base_Material_Functions/PickupEffectDistanceCullOverride}.Result
   Use actor center instead of pivot point
  ) {materialFunction=/Game/Packages/Fortress_SharedMaterials/Base_Material_Functions/LODEffect}
  n27 = WorldPosition()
  n28 = Add(
   A=MaterialFunctionCall!(WorldPostionOfBlackHole=AppendVector(A=@n19.RGB, B=@n19.RGB), World Position=Add(A=@n26.Result, B=@n27.XYZ)) {materialFunction=/Game/Packages/Fortnite_UniqueMaterials_1/BlackHoleFunction}.Result
   B=@n26.Result
  )
  n29 = StaticSwitchParameter(True=Add(A=@n28, B=@n24), False=@n28) {group=Animation, parameterName=Animate}
  n30 = MaterialFunctionCall(
   BaseNormal=MaterialFunctionCall(Rotation Axis=@n20.X-Axis Vector, Rotation Angle=@n23, Movement Mask=@n20.X-Axis Vector) {materialFunction=&FixRotateAboutAxisNormals}.New Tangent Space Vertex Normal
   AdditionalNormal=@n3.Normal
  ) {materialFunction=&BlendAngleCorrectedNormals}
  n31 = TextureSampleParameter2D() {parameterName=Normals, samplerType=SAMPLERTYPE_Normal, texture=&FlatNormalBlue}
  n32 = MaterialFunctionCall!(In=MakeMaterialAttributes!(
    BaseColor=@n3.Base Color
    Metallic=@n3.Base Color
    Specular=@n3.Base Color
    Roughness=@n3.Base Color
    Anisotropy
    EmissiveColor=StaticSwitchParameter(True=@n14, False=@n12) {group=Emissive, parameterName=Emissive Fresnel}
    Opacity
    OpacityMask
    Normal=StaticSwitchParameter(True=@n30.Result, False=@n31.RGB) {group=Animation, parameterName=Animate}
    Tangent
    WorldPositionOffset=StaticSwitchParameter(True=@n18, False=@n29) {parameterName=WiggleOnInteract}
    SubsurfaceColor
    ClearCoat
    ClearCoatRoughness
    AmbientOcclusion
    Refraction
    CustomizedUVs_0
    CustomizedUVs_1
    CustomizedUVs_2
    CustomizedUVs_3
    CustomizedUVs_4
    CustomizedUVs_5
    CustomizedUVs_6
    CustomizedUVs_7
    PixelDepthOffset
    ShadingModel
    Displacement
   )) {materialFunction=/Game/Packages/Fortress_SharedMaterials/Base_Material_Functions/MF_Potion}
  n33 = TextureSampleParameter2D() {
   group=SprocketPoppy
   parameterName=S
   samplerType=SAMPLERTYPE_Masks
   sortPriority=0
   texture=&T_BlackMask
  }
  n34 = TextureSampleParameter2D() {
   group=SprocketPoppy
   parameterName=Normals
   samplerType=SAMPLERTYPE_Normal
   sortPriority=0
   texture=&FlatNormalBlue
  }
  n35 = SetMaterialAttributes!(
   MaterialAttributes
   Base Color=TextureSampleParameter2D() {group=SprocketPoppy, parameterName=Diffuse, sortPriority=0, texture=&T_BlacksRGB}.RGB
   Metallic=NamedRerouteDeclaration(None=@n33.G) {name=IN_Metallic, nodeColor=<rgba 0 1 0.0391 1>}
   Specular=NamedRerouteDeclaration(None=@n33.R) {name=IN_Specular, nodeColor=<rgba 0.3594 1 0 1>}
   Roughness=NamedRerouteDeclaration(None=@n33.B) {name=IN_Roughness, nodeColor=<rgba 0.7578 0 1 1>}
   Normal=@n34.RGB
   Emissive Color=TextureSampleParameter2D() {parameterName=Emissive, texture=/TeaCakeIceGauntlets/Assets/Textures/T_Gauntlet_IceKings_E}.RGB
  ) {attributeSetTypes [
    #1
    #2
    #3
    #4
    #6
    #5
   ]}
  n36 = MaterialFunctionCall!(In=@n35, InMask, InFXMask=NamedRerouteUsage()) {materialFunction=/Game/Global/MaterialFunctions/TechArt/MF_Ice_TechArt}
  n37 = GetMaterialAttributes(None=StaticSwitchParameter(
    True=MaterialFunctionCall!(In=@n32.Result) {materialFunction=/Game/Packages/Fortress_SharedMaterials/Base_Material_Functions/Hologram}.Material Att
    False=NamedRerouteDeclaration(None=@n36.Result) {name=Ice, nodeColor=<rgba 0.9922 0 1 1>}
   ) {parameterName=Is Pickup?}) {attributeGetTypes [#5, #1]}
  n38 = Add(A=@n37.Emissive Color, B=NamedRerouteUsage())
  n39 = Fresnel(ExponentIn=Constant() {r=1}) {baseReflectFraction=0}
  n40 = MaterialFunctionCall(In=MaterialFunctionCall(
    Material=SetMaterialAttributes!(MaterialAttributes=@n37.MaterialAttributes, Emissive Color=@n38) {attributeSetTypes [#5]}
    M=Constant()
    Fresnel=NamedRerouteDeclaration(None=@n39) {name=Fresnel, nodeColor=<rgba 0.7578 0 1 1>}
   ) {materialFunction=&MF_CharacterParent_2023_PostFX}.Result) {materialFunction=&MF_TechArtOverrides_accessories}
  MaterialFunctionCall(MA In=MaterialFunctionCall!(Material attributes=@n40.Result, IsGameplayAsset=StaticBool()) {materialFunction=/Game/Packages/Fortress_SharedMaterials/Base_Material_Functions/Fortnite_BaseFunction}.Material Attributes) {materialFunction=&MF_CreativeMoveTool}
  MaterialFunctionCall() {materialFunction=&MF_CreativeMoveTool}
  Multiply(A=@n1.R, B=ScalarParameter() {defaultValue=1, group=Specular, parameterName=Specular Multiply, sliderMax=1})
  Multiply(A=@n3.Emissive Color, B=ScalarParameter() {defaultValue=1, parameterName=Emissive Contribution})
  Multiply(A=@n1.G, B=ScalarParameter() {defaultValue=1, group=Specular, parameterName=Metallic, sliderMax=1})
  n41 = LocalPosition() {localOrigin=InstancePreSkinning}
  n42 = ShaderStageSwitch(PixelShader=VertexInterpolator(VS=@n41.XYZ).PS, VertexShader=@n41.XYZ)
  n43 = StaticSwitchParameter(True=Constant(), False=@n42) {defaultValue=true, parameterName=Is Static Mesh?}
  NamedRerouteDeclaration(None=Step(Y=ComponentMask(None=@n43) {r=true})) {name=M_Left, nodeColor=<rgba 0.5703 1 0 1>}
  n44 = NamedRerouteUsage()
  n45 = Time()
  n46 = ScalarParameter() {defaultValue=2.5, group=SprocketPoppy, parameterName=Emissive_Noise_Speed}
  n47 = Multiply(A=TextureCoordinate(), B=Constant() {r=2})
  n48 = NamedRerouteUsage()
  n49 = Add(
   A=Multiply(
    A=TextureSample(UVs=@n47) {samplerType=SAMPLERTYPE_Grayscale, texture=/Game/Global/Textures/Default/Noise/LowResBlurredNoiseNormalized}.R
    B=Constant() {r=0.4}
   ) {constB=0.1}
   B=@n48
  )
  n50 = ScalarParameter() {defaultValue=2, group=SprocketPoppy, parameterName=Emissive_Noise_Freq}
  n51 = Sine(None=Add(A=Divide(A=@n45, B=@n46), B=Multiply(A=@n49, B=@n50)))
  n52 = StaticSwitchParameter(True=Multiply(A=@n51, B=@n48), False=@n51) {defaultValue=true, parameterName=Is Static Mesh?}
  n53 = ScalarParameter() {defaultValue=-0.75, group=SprocketPoppy, parameterName=Emissive_Noise_Bias}
  n54 = ScalarParameter() {defaultValue=0.3, group=SprocketPoppy, parameterName=Emissive_Noise_Mult}
  n55 = NamedRerouteUsage()
  n56 = NamedRerouteUsage()
  n57 = Add(A=@n56, B=ScalarParameter() {defaultValue=0.6, group=SprocketPoppy, parameterName=Emissive_Fresnel_Bias_Near})
  n58 = ScalarParameter() {defaultValue=1, group=SprocketPoppy, parameterName=Emissive_Fresnel_Scale_Near}
  n59 = Add(A=@n56, B=ScalarParameter() {defaultValue=0.2, group=SprocketPoppy, parameterName=Emissive_Fresnel_Bias_Far})
  n60 = ScalarParameter() {defaultValue=1.5, group=SprocketPoppy, parameterName=Emissive_Fresnel_Scale_Far}
  n61 = Add(
   A=Multiply(A=Add(A=@n52, B=@n53), B=LinearInterpolate(A=@n54, Alpha=@n55) {constB=0})
   B=LinearInterpolate(A=Multiply(A=@n57, B=@n58), B=Multiply(A=@n59, B=@n60), Alpha=NamedRerouteUsage())
  )
  n62 = Subtract(A=@n61, B=Multiply(A=OneMinus(None=@n44), B=ScalarParameter() {defaultValue=0.8, group=SprocketPoppy, parameterName=Emissive_IcePattern}))
  n63 = NamedRerouteUsage()
  n64 = ScalarParameter() {defaultValue=0.5, group=SprocketPoppy, parameterName=Emissive_Gradient_Bias_Near}
  n65 = ScalarParameter() {group=SprocketPoppy, parameterName=Emissive_Gradient_Bias_Far}
  n66 = LinearInterpolate(
   A=Multiply(A=Add(A=@n63, B=@n64), B=ScalarParameter() {defaultValue=2, group=SprocketPoppy, parameterName=Emissivel_Gradient_Mult_Near})
   B=Multiply(A=Add(A=@n63, B=@n65), B=ScalarParameter() {defaultValue=3, group=SprocketPoppy, parameterName=Emissive_Gradient_Mult_Far})
   Alpha=NamedRerouteUsage()
  )
  n67 = Multiply(A=Subtract(A=Add(A=@n61, B=@n62), B=@n66), B=Saturate(None=@n66))
  n68 = VectorParameter() {group=SprocketPoppy, parameterName=Emissive_Color2, defaultValue=<rgba 0 0.912 3.3 1>}
  n69 = VectorParameter() {group=SprocketPoppy, parameterName=Emissive_Color1, defaultValue=<rgba 0.0046 0 0.2 1>}
  n70 = NamedRerouteUsage()
  NamedRerouteDeclaration(None=Multiply(A=Saturate(None=@n67), B=LinearInterpolate(A=@n68.RGB, B=@n69.RGB, Alpha=@n70))) {name=EmissiveFesnel, nodeColor=<rgba 0 1 0.6484 1>}
  n71 = Subtract(A=PixelDepth().R, B=ScalarParameter() {defaultValue=500, group=SprocketPoppy, parameterName=Depht_bias})
  n72 = ScalarParameter() {defaultValue=2000, group=SprocketPoppy, parameterName=Depht_Falloff}
  NamedRerouteDeclaration(None=Saturate(None=Divide(A=@n71, B=@n72))) {name=M_Depht, nodeColor=<rgba 1 0 0.0937 1>}
  n73 = TextureCoordinate()
  n74 = ScalarParameter() {defaultValue=4, group=SprocketPoppy, parameterName=Ice_Patten_Tile}
  n75 = NamedRerouteDeclaration(None=TextureSample(UVs=Multiply(A=@n73, B=@n74)) {texture=/Game/Global/Textures/Default/Noise/T_UI_Ice_Noise}.R) {name=T_IcePattern, nodeColor=<rgba 1 0.1172 0 1>}
  n76 = ScalarParameter() {defaultValue=0.38, group=SprocketPoppy, parameterName=Ice_Pattern_Mult}
  n77 = Constant3Vector() {constant=<rgba 0 -15 140 0>}
  n78 = Constant() {r=85}
  n79 = Constant() {r=0.45}
  n80 = StaticSwitchParameter(
   True=MaterialFunctionCall() {materialFunction=&BoundingBoxBased_0-1_UVW}.B
   False=NamedRerouteDeclaration(None=SphereMask(A=@n42, B=@n77, Radius=@n78, Hardness=@n79) {attenuationRadius=50, hardnessPercent=1}) {name=M_Gradient_Inv, nodeColor=<rgba 0 0.2031 1 1>}
  ) {defaultValue=true, parameterName=Is Static Mesh?}
  n81 = Subtract(
   A=NamedRerouteDeclaration(None=OneMinus(None=@n80)) {name=M_Gradient, nodeColor=<rgba 0.8047 1 0 1>}
   B=ScalarParameter() {defaultValue=1, desc=1, group=SprocketPoppy, parameterName=IceEffect_Bias}
  )
  n82 = ScalarParameter() {group=SprocketPoppy, parameterName=IceEffect_Left_Bias}
  n83 = Add(A=@n81, B=@n82)
  n84 = ScalarParameter() {group=SprocketPoppy, parameterName=IceEffect_Right_Bias}
  n85 = Add(A=@n81, B=@n84)
  n86 = Multiply(A=NamedRerouteUsage(), B=ScalarParameter() {defaultValue=0.8, group=SprocketPoppy, parameterName=Idle_Ice_Falloff})
  n87 = ScalarParameter() {defaultValue=0.5, group=SprocketPoppy, parameterName=Idle_Ice_Bias}
  n88 = Multiply(A=Saturate(None=Add(A=@n86, B=@n87)), B=ScalarParameter() {defaultValue=0.55, group=SprocketPoppy, parameterName=Idle_Ice_Intensity})
  n89 = Add(A=LinearInterpolate(A=Multiply(A=@n83, B=@n82), B=Multiply(A=@n85, B=@n84), Alpha=NamedRerouteUsage()), B=@n88)
  n90 = Multiply(A=Multiply(A=@n88, B=NamedRerouteUsage()), B=ScalarParameter() {defaultValue=0.15, group=SprocketPoppy, parameterName=Idle_Ice_Palm})
  n91 = ScalarParameter() {defaultValue=0.02, group=SprocketPoppy, parameterName=Idle_NoiseAnim_Mult}
  n92 = Multiply(
   A=Add(A=Add(A=@n89, B=@n90), B=Multiply(A=@n52, B=@n91))
   B=ScalarParameter() {defaultValue=1, desc=3, group=SprocketPoppy, parameterName=IceEffect_Falloff}
  )
  n93 = Subtract(A=@n92, B=OneMinus(None=Multiply(A=@n75, B=@n76) {constB=0.5}))
  NamedRerouteDeclaration(None=Saturate(None=Add(A=@n92, B=@n93) {constB=1.3})) {name=Ice_Mask, nodeColor=<rgba 0 1 0.3672 1>}
  n94 = Constant3Vector() {constant=<rgba 0 0.5 -1 1>}
  n95 = DotProduct(A=PixelNormalWS(), B=@n94.R)
  n96 = DotProduct(A=PreSkinnedNormal(), B=@n94)
  n97 = ShaderStageSwitch(PixelShader=VertexInterpolator(VS=@n96).PS, VertexShader=@n96)
  NamedRerouteDeclaration(None=Saturate(None=StaticSwitchParameter(True=@n95, False=@n97) {defaultValue=true, parameterName=Is Static Mesh?})) {name=M_Palms, nodeColor=<rgba 0 0.5547 1 1>}
  ScalarParameter() {group=Clear Coat, parameterName=Clear Coat Roughness}
  n98 = TextureSampleParameter2D() {
   group=Specular
   parameterName=SpecularMasks
   samplerType=SAMPLERTYPE_Masks
   texture=&T_Fortnite_Default_S
  }
  StaticSwitchParameter(True=@n98.A, False=ScalarParameter() {defaultValue=1, group=Clear Coat, parameterName=Clear Coat Opacity, sliderMax=1}) {defaultValue=true, group=Clear Coat, parameterName=Clear Coat Mask}
 ]
}
