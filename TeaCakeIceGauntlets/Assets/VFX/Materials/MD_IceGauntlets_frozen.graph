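# Translucent unlit deferred decal material for the Ice Gauntlets frozen effect, building frost in a falloff around the object's position.
# A distance-based 3D mask uses reroute-fed offsets, alongside two world-space noise samples varied by particle random.
# Frost blends triplanar by scene world normal on High and Epic, uses one projection by default, and becomes constant on Low and Medium.
# Decal lifetime opacity drives the noise exponent and fade, with frost tinted by a color parameter.
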Material /TeaCakeIceGauntlets/Assets/VFX/Materials/MD_IceGauntlets_frozen
graph {
 properties {
  bEnableResponsiveAA=true
  bOutputTranslucentVelocity=true
  bUsedWithNiagaraRibbons=true
  bUsedWithNiagaraSprites=true
  blendMode=BLEND_Translucent
  materialDomain=MD_DeferredDecal
  pixelDepthOffsetMode=PDOM_Legacy
  shadingModel=MSM_Unlit
  twoSided=true
  usedShadingModels=Unlit
 }
 nodes [
  Constant() {r=0.5}
  n1 = ObjectPositionWS()
  n2 = Add(A=WorldPosition().XYZ, B=Multiply(A=NamedRerouteUsage()) {constB=30})
  n3 = Subtract(A=Multiply(A=Distance(A=@n1, B=@n2)) {constB=0.001}, B=ScalarParameter() {defaultValue=0.256, parameterName=Decal Range}) {constB=0.02}
  n4 = Power(Base=OneMinus(None=Saturate(None=@n3)), Exp=ScalarParameter() {defaultValue=40, parameterName=Pow}) {constExponent=150}
  Multiply(A=@n4) {constB=11}
  n5 = Multiply(A=WorldPosition().XYZ, B=ScalarParameter() {defaultValue=0.0054, parameterName=Tiling A})
  n6 = TextureSample(UVs=ComponentMask(None=@n5) {g=true, r=true}) {texture=/Game/Effects/Fort_Effects/Textures/Snowflake/T_LightFrost}
  n7 = Abs(None=SceneTexture() {sceneTextureId=PPI_WorldNormal}.Color)
  n8 = Multiply(A=Add(A=@n6.R, B=@n6.R), B=ComponentMask(None=@n7) {b=true})
  n9 = Constant() {r=1}
  n10 = TextureSample(UVs=ComponentMask(None=@n5) {b=true, r=true}) {texture=&T_LightFrost}
  n11 = Add(A=@n10.R, B=@n10.R)
  n12 = ComponentMask(None=@n7) {g=true}
  n13 = TextureSample(UVs=ComponentMask(None=@n5) {b=true, g=true}) {texture=&T_LightFrost}
  n14 = Add(A=Add(A=@n8, B=Multiply(A=@n11, B=@n12)), B=Multiply(A=Add(A=@n13.R, B=@n13.R), B=ComponentMask(None=@n7) {r=true}))
  n15 = QualitySwitch(Default=@n8, Low=@n9, High=@n14, Medium=@n9, Epic=@n14)
  Multiply(A=@n15) {constB=0.5}
  Multiply(A=VectorParameter() {parameterName=Emissive Frost Color, defaultValue=<rgba 0.2355 0.5128 0.7292 1>}.RGB, B=@n15)
  n16 = WorldPosition()
  n17 = AppendVector(A=@n16.XY, B=Add(A=@n16.Z, B=ParticleRandom()))
  n18 = TextureSample(UVs=Multiply(A=@n17, B=ScalarParameter() {defaultValue=0.01, parameterName=Detail2}) {constB=0.007}) {samplerType=SAMPLERTYPE_LinearColor, texture=/Game/Global/MaterialFunctions/VT_ClothSimWPO}
  Multiply(A=OneMinus(None=@n18.R), B=@n15)
  n19 = TextureSample(UVs=Multiply(A=@n17, B=ScalarParameter() {defaultValue=0.0025, parameterName=Detail1}) {constB=0.003}) {samplerType=SAMPLERTYPE_LinearColor, texture=&VT_ClothSimWPO}
  n20 = ScalarParameter() {defaultValue=0.9, parameterName=Noise Power}
  n21 = DecalLifetimeOpacity()
  n22 = Multiply(A=OneMinus(None=@n21.Opacity)) {constB=0.5}
  n23 = Power(Base=@n19.RGB, Exp=Multiply(A=@n20, B=Add(A=@n20, B=@n22)))
  n24 = Multiply(A=@n23, B=NamedRerouteUsage())
  Multiply(
   A=@n23
   B=Multiply(
    A=Saturate(None=@n24)
    B=MaterialFunctionCall(Float2=@n21.Opacity) {materialFunction=&BreakOutFloat2Components}.R
   )
  )
  n25 = Multiply(A=Saturate(None=Multiply(A=@n4) {constB=300}), B=NamedRerouteUsage())
  NamedRerouteDeclaration(None=Saturate(None=Add(A=@n4, B=@n25))) {name=3D Mask, nodeColor=<rgba 0 0.6016 1 1>}
  NamedRerouteDeclaration(None=Multiply(A=Subtract(A=@n18.R) {constB=0.5}) {constB=0.5}) {name=Noise out, nodeColor=<rgba 0.7344 1 0 1>}
  NamedRerouteDeclaration(None=Subtract(A=@n19.RGB) {constB=0.5}) {name=Noise1, nodeColor=<rgba 0.8516 0 1 1>}
  VectorParameter() {parameterName=Ice Color, defaultValue=<rgba 0.1771 0.5355 1 1>}
 ]
}
