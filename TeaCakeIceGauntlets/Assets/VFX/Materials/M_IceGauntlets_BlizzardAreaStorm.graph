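# Translucent unlit material for Niagara mesh particles, drawing the Ice Gauntlets blizzard area storm as swirling wind and snow.
# Panning wind, frost and noise textures form the body, with two layers of snow bits, a three-color ramp and sun-aware time-of-day tint.
# Dynamic parameters drive erosion, Fresnel and normal-based vertex displacement.
# Edge, depth, near-camera and view-angle fades soften opacity, while quality and platform switches simplify costlier layers.
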Material /TeaCakeIceGauntlets/Assets/VFX/Materials/M_IceGauntlets_BlizzardAreaStorm
graph {
 properties {
  bAutomaticallySetUsageInEditor=false
  bUsedWithNiagaraMeshParticles=true
  blendMode=BLEND_Translucent
  shadingModel=MSM_Unlit
  usedShadingModels=Unlit
 }
 nodes [
  n1 = TextureCoordinate()
  Add(A=@n1) {constB=0}
  n2 = TextureCoordinate()
  Add(A=@n2) {constB=0}
  n3 = MaterialFunctionCall(Float2=Multiply(A=@n1, B=ScalarParameter() {defaultValue=2, parameterName=Bits Size})) {materialFunction=&BreakOutFloat2Components}
  n4 = Multiply(A=@n3.R) {constB=7}
  n5 = Time()
  n6 = Multiply(A=@n5, B=ScalarParameter() {defaultValue=3.3, parameterName=SnowBits2_U_Speed}) {constB=4}
  n7 = Multiply(A=@n3.G) {constB=5}
  n8 = Multiply(A=@n5, B=ScalarParameter() {defaultValue=0.9, parameterName=SnowBits2_V_Speed}) {constB=1.2}
  n9 = TextureSampleParameter2D(UVs=AppendVector(A=Add(A=@n4, B=@n6), B=Add(A=@n7, B=@n8))) {parameterName=TexSnowBits2, texture=/Game/Effects/Fort_Effects/Textures/Tile/T_Dots_Small}
  LinearInterpolate(Alpha=@n9.R) {constB=0.6}
  n10 = Power(Base=NamedRerouteUsage()) {constExponent=3}
  n11 = Saturate(None=Power(
    Base=OneMinus(None=@n10)
    Exp=DynamicParameter() {defaultValue=<rgba 1 2 0 1>, paramNames [WPO_Direction, Fresnel_Intensity, Erosion, WPO_Intensity]}.Fresnel_Intensity
   ))
  MaterialFunctionCall(In=@n11, Contrast=Constant() {r=1}) {materialFunction=&CheapContrast}
  n12 = NamedRerouteUsage()
  n13 = VectorParameter() {parameterName=LightColorMorning, defaultValue=<rgba 0.9 0.7459 0.7632 1>}
  n14 = VectorParameter() {parameterName=LightColorDay, defaultValue=<rgba 0.55 0.8643 1 1>}
  n15 = VectorParameter() {parameterName=LightColorEvening, defaultValue=<rgba 0.85 0.8209 0.8042 1>}
  n16 = VectorParameter() {parameterName=LightColorNight, defaultValue=<rgba 0.36 0.585 0.9 1>}
  n17 = NamedRerouteUsage()
  n18 = OneMinus(None=@n17)
  n19 = LinearInterpolate(
   A=VectorParameter() {group=Color, parameterName=ColorMid, sortPriority=35, defaultValue=<rgba 0.13 1 1 1>}.RGB
   B=VectorParameter() {group=Color, parameterName=ColorDark, sortPriority=40, defaultValue=<rgba 0.01 0.0625 0.1 1>}.RGB
   Alpha=MaterialFunctionCall(In=Power(Base=@n18) {constExponent=2.5}, Contrast=Constant() {r=0.1}) {materialFunction=&CheapContrast}.Result
  )
  n20 = Saturate(None=Power(Base=@n17) {constExponent=1.5})
  n21 = VectorParameter() {group=Color, parameterName=ColorLight, sortPriority=30, defaultValue=<rgba 1 1 1 1>}
  n22 = NamedRerouteUsage()
  n23 = Constant() {r=1}
  n24 = MaterialFunctionCall(Float2=TextureCoordinate()) {materialFunction=&BreakOutFloat2Components}
  n25 = Multiply(A=@n24.R) {constB=2}
  n26 = Time()
  n27 = Multiply(A=@n26, B=ScalarParameter() {defaultValue=-1, group=Noise, parameterName=Noise_PanSpeed_U}) {constB=-1.2}
  n28 = Multiply(A=@n24.G)
  n29 = Multiply(A=@n26, B=ScalarParameter() {defaultValue=0.5, group=Noise, parameterName=Noise_PanSpeed_V}) {constB=0.7}
  n30 = TextureSampleParameter2D(UVs=AppendVector(A=Add(A=@n25, B=@n27), B=Add(A=@n28, B=@n29))) {
   group=Noise
   parameterName=Noise
   samplerType=SAMPLERTYPE_LinearColor
   texture=/Game/Weapons/FORT_Melee/Pickaxe_RhombCamo/FX/T_Pickaxe_RhombCamo_Perlin
  }
  n31 = Power(Base=@n30.R) {constExponent=1.5}
  n32 = QualitySwitch(Default=@n31, Low=@n23, High=@n31, Medium=@n23, Epic=@n31)
  n33 = ShadingPathSwitch(Default=@n32, Deferred=@n32, Mobile=@n23) {desc=Migrated from FeatureLevelSwitch}
  n34 = Multiply(
   A=NamedRerouteUsage()
   B=LinearInterpolate(B=LinearInterpolate(Alpha=@n33) {constA=0.4}, Alpha=ScalarParameter() {defaultValue=1, parameterName=Use_Color_Noise}) {constA=1}
  )
  n35 = NamedRerouteUsage()
  n36 = Multiply(A=Multiply(A=Max(A=@n22, B=@n34), B=@n35)) {constB=3}
  n37 = ParticleColor()
  n38 = Multiply(A=LinearInterpolate(A=LinearInterpolate(A=@n19, B=@n21.RGB, Alpha=@n20), B=@n21.RGB, Alpha=Saturate(None=@n36)), B=@n37.RGB)
  n39 = Constant() {r=0.4}
  n40 = CollectionParameter() {
   collection=&FortniteMaterialParameters
   parameterName=SunlightDirectionalVector
  }
  n41 = Transform(None=Normalize(VectorInput=ComponentMask(None=@n40) {b=true, g=true, r=true})) {transformSourceType=TRANSFORMSOURCE_World}
  n42 = VertexNormalWS()
  n43 = Power(Base=ConstantBiasScale(None=DotProduct(A=@n41, B=@n42))) {constExponent=7}
  n44 = LinearInterpolate(
   A=Multiply(
    A=MaterialFunctionCall(Color_Morning=@n13.RGB, Color_Day=@n14.RGB, Color_Evening=@n15.RGB, Color_Night=@n16.RGB) {materialFunction=&Color_Time_of_Day}.Result
    B=Desaturation(None=@n38, Fraction=@n39)
   ) {constA=0.5}
   B=@n38
   Alpha=LinearInterpolate(A=@n43, B=OneMinus(None=@n43), Alpha=ScalarParameter() {parameterName=Invert_Normal_ToD})
  )
  n45 = Power(Base=NamedRerouteUsage(), Exp=ScalarParameter() {defaultValue=0.3, parameterName=Blend Pow})
  MaterialFunctionCall(Emissive=LinearInterpolate(A=Max(A=@n12, B=@n44), B=Multiply(A=@n44, B=@n45))) {materialFunction=&MF_FX_EmissiveAdjustments}
  n46 = DepthFade(FadeDistance=ScalarParameter() {defaultValue=200, parameterName=DepthFadeDistance})
  n47 = Constant() {r=1}
  n48 = QualitySwitch(Default=@n46, Low=@n47, High=@n46, Medium=@n46, Epic=@n46)
  n49 = ShadingPathSwitch(Default=@n48, Deferred=@n48, Mobile=@n47) {desc=Migrated from FeatureLevelSwitch}
  NamedRerouteDeclaration(None=Power(Base=ShadingPathSwitch(Default=@n49, Forward=@n47)) {constExponent=4}) {name=Depth Fade, nodeColor=<rgba 1 0.1577 0 1>}
  n50 = MaterialFunctionCall(Float2=TextureCoordinate()) {materialFunction=&BreakOutFloat2Components}
  n51 = Power(Base=Sine(None=Multiply(A=@n50.G) {constB=0.5})) {constExponent=1}
  NamedRerouteDeclaration(None=Saturate(None=Multiply(A=@n51) {constB=2})) {name=TopBottomCreasesFade, nodeColor=<rgba 0.5516 0.0833 1 1>}
  n52 = MaterialFunctionCall(Float2=TextureCoordinate()) {materialFunction=&BreakOutFloat2Components}
  n53 = Multiply(
   A=Power(Base=NamedRerouteUsage(), Exp=ScalarParameter() {defaultValue=2, parameterName=Erosion_Power}) {constExponent=3}
   B=ScalarParameter() {defaultValue=1, parameterName=Erosion_Hardness}
  ) {constB=2}
  n54 = Multiply(
   A=Sine(None=Multiply(A=@n52.G) {constB=0.5})
   B=MaterialFunctionCall(In=Saturate(None=@n53), Contrast=ScalarParameter() {parameterName=Erosion_Extra_Contrast}) {materialFunction=&CheapContrast}.Result
  )
  n55 = LinearInterpolate(Alpha=DynamicParameter() {defaultValue=<rgba 1 2 0 1>, paramNames [WPO_Direction, Fresnel_Intensity, Erosion, WPO_Intensity]}.Erosion) {constA=-1}
  NamedRerouteDeclaration(None=Saturate(None=Subtract(A=@n54, B=@n55))) {name=Erosion, nodeColor=<rgba 0.491 1 0.0616 1>}
  n56 = NamedRerouteUsage()
  n57 = Fresnel() {baseReflectFraction=0, exponent=1}
  NamedRerouteDeclaration(None=LinearInterpolate(A=@n57, B=Multiply(A=@n56, B=@n57), Alpha=OneMinus(None=@n57))) {name=Edge Mask Textured, nodeColor=<rgba 1 0 0.5859 1>}
  n58 = Constant()
  n59 = NamedRerouteUsage()
  n60 = MaterialFunctionCall(Float2=Multiply(A=@n2, B=ScalarParameter() {defaultValue=2, parameterName=Bits Size2})) {materialFunction=&BreakOutFloat2Components}
  n61 = Multiply(A=@n60.R) {constB=7}
  n62 = Time()
  n63 = Multiply(A=@n62, B=ScalarParameter() {defaultValue=2.6, parameterName=SnowBits1_U_Speed}) {constB=3}
  n64 = Multiply(A=@n60.G) {constB=4}
  n65 = Multiply(A=@n62, B=ScalarParameter() {defaultValue=0.35, parameterName=SnowBits1_V_Speed}) {constB=0.4}
  n66 = TextureSampleParameter2D(UVs=AppendVector(A=Add(A=@n61, B=@n63), B=Add(A=@n64, B=@n65))) {parameterName=TexSnowBits, texture=&T_Dots_Small}
  n67 = Multiply(A=Max(A=Power(Base=@n66.B) {constExponent=4}, B=Power(Base=@n9.R) {constExponent=2.2}) {constB=0}, B=@n33)
  n68 = Saturate(None=Multiply(A=Power(Base=@n67) {constExponent=1}))
  n69 = LinearInterpolate(Alpha=Saturate(None=Multiply(A=@n59, B=@n68))) {constB=3}
  n70 = QualitySwitch(Default=@n69, Low=@n58, High=@n69, Medium=@n58, Epic=@n69)
  NamedRerouteDeclaration(None=ShadingPathSwitch(Default=@n70, Deferred=@n70, Mobile=@n58) {desc=Migrated from FeatureLevelSwitch}) {name=Snow Bits, nodeColor=<rgba 0.1146 1 1 1>}
  n71 = MaterialFunctionCall(Float2=TextureCoordinate()) {materialFunction=&BreakOutFloat2Components}
  n72 = Multiply(A=@n71.R, B=ScalarParameter() {defaultValue=1, group=Main Texture, parameterName=Main_Tex_Scale_U})
  n73 = Time()
  n74 = Multiply(A=ScalarParameter() {defaultValue=1, group=Main Texture, parameterName=Main_Tex_PanSpeed_U}, B=@n73) {constB=1.45}
  n75 = Multiply(A=@n71.G, B=ScalarParameter() {defaultValue=3, group=Main Texture, parameterName=Main_Tex_Scale_V}) {constB=3}
  n76 = Multiply(A=@n73, B=ScalarParameter() {defaultValue=0.2, group=Main Texture, parameterName=Main_Tex_PanSpeed_V}) {constB=0.33}
  n77 = TextureSampleParameter2D(UVs=AppendVector(A=Add(A=@n72, B=@n74), B=Add(A=@n75, B=@n76))) {group=Main Texture, parameterName=MainTex, texture=/Game/Effects/FogVolume/noise_flamy_Tile}
  n78 = LinearInterpolate(
   A=@n77.R
   B=MaterialFunctionCall(In=@n77.R, Contrast=Constant() {r=0.5}) {materialFunction=&CheapContrast}.Result
  ) {constAlpha=0.6}
  n79 = LinearInterpolate(Alpha=@n78) {constA=0.1}
  n80 = Multiply(A=@n71.R, B=ScalarParameter() {defaultValue=1, group=Frost Texture, parameterName=FrostTex_Scale_U})
  n81 = Multiply(A=@n71.G, B=ScalarParameter() {defaultValue=2, group=Frost Texture, parameterName=FrostTex_Scale_V}) {constB=2}
  n82 = ScalarParameter() {defaultValue=-0.45, group=Frost Texture, parameterName=FrostTex_PanSpeed_U}
  n83 = ScalarParameter() {defaultValue=0.15, group=Frost Texture, parameterName=FrostTex_PanSpeed_V}
  n84 = TextureSampleParameter2D(UVs=Panner(Coordinate=AppendVector(A=@n80, B=@n81), Speed=AppendVector(A=@n82, B=@n83)) {speedX=-0.7, speedY=0.33}) {group=Frost Texture, parameterName=TexFrost, texture=/Game/Effects/Fort_Effects/Textures/Smoke/T_Clouds_01}
  n85 = Constant() {r=0.1}
  n86 = LinearInterpolate(
   A=@n78
   B=Saturate(None=MaterialFunctionCall(In=@n84.R, Contrast=@n85) {materialFunction=&CheapContrast}.Result)
   Alpha=ScalarParameter() {defaultValue=0.3, parameterName=TexLerp}
  ) {constAlpha=0.2}
  n87 = Multiply(A=MaterialFunctionCall("Xb1,PS4"=@n86, "PC-High&&Epic,XSX,PS5"=@n86, "PC-Perf,Android"=@n79, Switch=@n79, "PC-Low(NotPerf)&&Medium"=@n86) {materialFunction=&PlatformSwitch_1D}.Result) {constB=1.55}
  NamedRerouteDeclaration(None=LinearInterpolate(B=Saturate(None=@n87), Alpha=NamedRerouteUsage()) {constA=1}) {name=Spin Wind Texture, nodeColor=<rgba 0.1255 0.8195 1 1>}
  n88 = Saturate(None=Multiply(A=@n35, B=@n34))
  n89 = NamedRerouteUsage()
  n90 = Multiply(A=Power(Base=Multiply(A=@n88, B=@n89)) {constExponent=1})
  n91 = DynamicParameter() {defaultValue=<rgba 1 2 0 1>, paramNames [WPO_Direction, Fresnel_Intensity, Erosion, WPO_Intensity]}
  n92 = Multiply(A=LinearInterpolate(Alpha=Saturate(None=@n90)) {constA=-0.2}, B=Multiply(A=@n91.WPO_Intensity) {constB=85}) {constB=85}
  n93 = Multiply(A=@n91.WPO_Direction, B=Constant3Vector() {constant=<rgba 1 1 0 0>})
  NamedRerouteDeclaration(None=Multiply(A=Multiply(A=@n92, B=@n93), B=VertexNormalWS())) {name=WPO, nodeColor=<rgba 1 0.4453 0 1>}
  n94 = NamedRerouteUsage()
  NamedRerouteDeclaration(None=Multiply(A=Multiply(A=@n94, B=@n11), B=ScalarParameter() {defaultValue=1, parameterName=Opacity})) {name=Opacity, nodeColor=<rgba 0 1 0.8125 1>}
  NamedRerouteUsage()
  PerInstanceRandom()
  PerInstanceRandom()
  n95 = LinearInterpolate(B=NamedRerouteUsage(), Alpha=ScalarParameter() {defaultValue=1, parameterName=EdgeMask_Intensity}) {constA=1}
  n96 = NamedRerouteUsage()
  n97 = Power(Base=MaterialFunctionCall(Distance=ScalarParameter() {defaultValue=500, parameterName=CameraFade_Distance}) {materialFunction=&Near_Camera_Fade}.Result) {constExponent=5}
  n98 = Multiply(A=Multiply(A=Multiply(A=@n37.A, B=@n96), B=NamedRerouteUsage()), B=Saturate(None=Multiply(A=@n97) {constB=7}))
  n99 = ScalarParameter() {parameterName=AngleFade_Sensitivity}
  Saturate(None=Multiply(
    A=Multiply(A=@n95, B=@n98) {constA=1}
    B=MaterialFunctionCall("FadeSensitivity (optional)"=@n99) {materialFunction=&FadeBasedOnViewAngle}.Result
   ))
  VectorParameter() {parameterName=Base Color, defaultValue=<rgba 0.0028 0.0091 0.0365 1>}
 ]
}
