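# Translucent unlit material for the Ice Gauntlets ice pillar Niagara mesh particles, with controls for scattering the mesh.
# Color adds a desaturated, hue-shifted ice texture, tinted by time of day, to normal-perturbed fake reflections optionally tinted by particle color.
# Opacity math combines an ice mask, depth fade and particle alpha.
# Vertex offset math uses vertex colors, noise and dynamic parameters to spread, drop and spin pieces.
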Material /TeaCakeIceGauntlets/Assets/VFX/Materials/M_IceGauntlets_IcePillar
graph {
 expressions {
  MaterialExpressionAdd_0 {
   class=MaterialExpressionAdd
   inputs [
    A {from=MaterialExpressionMultiply_0}
    B {from=MaterialExpressionParticlePositionWS_0}
   ]
  }
  MaterialExpressionAdd_2 {
   class=MaterialExpressionAdd
   inputs [
    A {from=MaterialExpressionMultiply_1}
    B {from=MaterialExpressionMultiply_11}
   ]
  }
  MaterialExpressionAdd_3 {
   class=MaterialExpressionAdd
   inputs [
    A {from=MaterialExpressionMultiply_10}
    B {from=""}
   ]
   properties {constB=0.5}
  }
  MaterialExpressionAdd_4 {
   class=MaterialExpressionAdd
   inputs [
    A {from=MaterialExpressionAdd_2}
    B {from=MaterialExpressionMaterialFunctionCall_5, output=WPO}
   ]
  }
  MaterialExpressionAdd_5 {
   class=MaterialExpressionAdd
   inputs [
    A {from=MaterialExpressionVectorNoise_0}
    B {from=MaterialExpressionParticlePositionWS_0}
   ]
  }
  MaterialExpressionAdd_8 {
   class=MaterialExpressionAdd
   inputs [
    A {from=MaterialExpressionReroute_3}
    B {from=MaterialExpressionLinearInterpolate_1}
   ]
  }
  MaterialExpressionConstant_0 {class=MaterialExpressionConstant}
  MaterialExpressionConvert_0 {
   class=MaterialExpressionConvert
   inputs [
    None {from=MaterialExpressionVertexColor_1, output=G}
    None {from=MaterialExpressionVertexColor_1, output=G}
    None {from=MaterialExpressionVertexColor_1, output=G}
   ]
   properties {
    nodeName=Make Vector3
    convertInputs [
     {
      type=Scalar
      expressionInput {
       expression=&self:MaterialExpressionVertexColor_1
       inputName=None
       mask=1
       maskA=0
       maskB=0
       maskG=1
       maskR=0
       outputIndex=2
      }
     }
     {
      type=Scalar
      expressionInput {
       expression=&self:MaterialExpressionVertexColor_1
       inputName=None
       mask=1
       maskA=1
       maskB=0
       maskG=0
       maskR=0
       outputIndex=4
      }
     }
     {
      type=Scalar
      expressionInput {
       expression=&self:MaterialExpressionVertexColor_1
       inputName=None
       mask=1
       maskA=0
       maskB=1
       maskG=0
       maskR=0
       outputIndex=3
      }
     }
    ]
    convertOutputs [{type=Vector3}]
   }
  }
  MaterialExpressionDepthFade_2 {
   class=MaterialExpressionDepthFade
   inputs [
    Opacity {from="", output=""}
    FadeDistance {from=MaterialExpressionScalarParameter_13, output=""}
   ]
   properties {fadeDistanceDefault=20}
  }
  MaterialExpressionDesaturation_1 {
   class=MaterialExpressionDesaturation
   inputs [
    None {from=MaterialExpressionMaterialFunctionCall_14, output=RGB}
    Fraction {from=MaterialExpressionScalarParameter_21, output=""}
   ]
   properties {}
  }
  MaterialExpressionDivide_1 {
   class=MaterialExpressionDivide
   inputs [
    A {from=MaterialExpressionAdd_8}
    B {from=""}
   ]
   properties {constB=1.97}
  }
  MaterialExpressionDynamicParameter_0 {
   class=MaterialExpressionDynamicParameter
   inputs []
   properties {defaultValue=<rgba 1 1 1 0>, paramNames [Interaction, Spread, Set, Fake Gravity]}
  }
  MaterialExpressionDynamicParameter_1 {
   class=MaterialExpressionDynamicParameter
   inputs []
   properties {defaultValue=<rgba 1 1 1 0>, paramNames [Interaction, Spread, Set, Fake Gravity]}
  }
  MaterialExpressionDynamicParameter_2 {
   class=MaterialExpressionDynamicParameter
   inputs []
   properties {defaultValue=<rgba 1 1 1 0>, paramNames [Interaction, Spread, Set, Fake Gravity]}
  }
  MaterialExpressionDynamicParameter_4 {class=MaterialExpressionDynamicParameter, inputs [], properties {paramNames [R,Param2,Param3,Param4]}}
  MaterialExpressionLinearInterpolate_1 {
   class=MaterialExpressionLinearInterpolate
   inputs [
    A {from=MaterialExpressionMultiply_25}
    B {from=MaterialExpressionMultiply_26}
    Alpha {from=MaterialExpressionScalarParameter_20}
   ]
  }
  MaterialExpressionMaterialFunctionCall_11 {
   class=MaterialExpressionMaterialFunctionCall
   inputs [Emissive {from=MaterialExpressionDivide_1}]
   properties {materialFunction=&MF_FX_EmissiveAdjustments}
  }
  MaterialExpressionMaterialFunctionCall_12 {
   class=MaterialExpressionMaterialFunctionCall
   inputs [
    Normal {from=MaterialExpressionMaterialFunctionCall_13, output=RGB}
    Flatness {from=MaterialExpressionScalarParameter_15}
   ]
   properties {materialFunction=&FlattenNormal}
  }
  MaterialExpressionMaterialFunctionCall_13 {
   class=MaterialExpressionMaterialFunctionCall
   inputs [
    Texture {from=MaterialExpressionTextureObjectParameter_0}
    AdditionalUVOffset {from=""}
    UVControl_RGB {from=MaterialExpressionVectorParameter_0}
    UVControl_A {from=MaterialExpressionVectorParameter_0}
    "[Override]TimeInput" {from=""}
    "[Override]UVInput" {from=""}
   ]
   properties {materialFunction=&FXCommon_TextureSample}
  }
  MaterialExpressionMaterialFunctionCall_14 {
   class=MaterialExpressionMaterialFunctionCall
   inputs [
    Texture {from=MaterialExpressionTextureObjectParameter_4}
    AdditionalUVOffset {from=MaterialExpressionDynamicParameter_4, output=R}
    UVControl_RGB {from=MaterialExpressionVectorParameter_2}
    UVControl_A {from=MaterialExpressionVectorParameter_2}
    "[Override]TimeInput" {from=""}
    "[Override]UVInput" {from=""}
   ]
   properties {materialFunction=&FXCommon_TextureSample}
  }
  MaterialExpressionMaterialFunctionCall_15 {
   class=MaterialExpressionMaterialFunctionCall
   inputs [
    ReflectionVector {from=MaterialExpressionReflectionVectorWS_1}
    ReflectionTexture {from=MaterialExpressionTextureObjectParameter_5}
   ]
   properties {materialFunction=/Engine/Functions/Engine_MaterialFunctions01/Reflections/ViewAlignedReflection}
  }
  MaterialExpressionMaterialFunctionCall_16 {
   class=MaterialExpressionMaterialFunctionCall
   inputs [
    Texture {from=MaterialExpressionTextureObjectParameter_5}
    AdditionalUVOffset {from=""}
    UVControl_RGB {from=MaterialExpressionVectorParameter_7}
    UVControl_A {from=MaterialExpressionVectorParameter_7}
    "[Override]TimeInput" {from=""}
    "[Override]UVInput" {from=MaterialExpressionMaterialFunctionCall_15, output=UVs}
   ]
   properties {materialFunction=&FXCommon_TextureSample}
  }
  MaterialExpressionMaterialFunctionCall_17 {
   class=MaterialExpressionMaterialFunctionCall
   inputs [
    Hue Shift Percentage {from=MaterialExpressionScalarParameter_19}
    Texture {from=MaterialExpressionDesaturation_1}
   ]
   properties {materialFunction=&HueShift}
  }
  MaterialExpressionMaterialFunctionCall_18 {
   class=MaterialExpressionMaterialFunctionCall
   inputs [
    Color_Morning {from=MaterialExpressionVectorParameter_8, output=RGB}
    Color_Day {from=MaterialExpressionVectorParameter_9, output=RGB}
    Color_Evening {from=MaterialExpressionVectorParameter_8, output=RGB}
    Color_Night {from=MaterialExpressionVectorParameter_10, output=RGB}
   ]
   properties {materialFunction=&Color_Time_of_Day}
  }
  MaterialExpressionMaterialFunctionCall_2 {
   class=MaterialExpressionMaterialFunctionCall
   inputs [
    X {from=MaterialExpressionConstant_0}
    Y {from=MaterialExpressionConstant_0}
    Z {from=MaterialExpressionAdd_3}
   ]
   properties {materialFunction=&MakeFloat3}
  }
  MaterialExpressionMaterialFunctionCall_5 {
   class=MaterialExpressionMaterialFunctionCall
   inputs [
    RotationAngle {from=MaterialExpressionMultiply_9}
    NormalizedRotationAxis {from=MaterialExpressionNormalize_0}
    PivotPoint {from=MaterialExpressionAdd_0}
    Position {from=""}
    Normal {from=""}
    TangentU {from=""}
    TangentV {from=""}
   ]
   properties {materialFunction=/Game/Characters/Player/Male/Medium/Bodies/M_MED_Inferno/FX/Mat/RotateVertexAboutAxis}
  }
  MaterialExpressionMultiply_0 {
   class=MaterialExpressionMultiply
   inputs [
    A {from=MaterialExpressionConvert_0}
    B {from=MaterialExpressionScalarParameter_0}
   ]
  }
  MaterialExpressionMultiply_1 {
   class=MaterialExpressionMultiply
   inputs [
    A {from=MaterialExpressionMultiply_4}
    B {from=MaterialExpressionSubtract_1}
   ]
  }
  MaterialExpressionMultiply_10 {
   class=MaterialExpressionMultiply
   inputs [
    A {from=MaterialExpressionVertexColor_3, output=R}
    B {from=""}
   ]
   properties {constB=0.5}
  }
  MaterialExpressionMultiply_11 {
   class=MaterialExpressionMultiply
   inputs [
    A {from=MaterialExpressionDynamicParameter_0, output=Interaction}
    B {from=MaterialExpressionMultiply_8}
   ]
  }
  MaterialExpressionMultiply_12 {
   class=MaterialExpressionMultiply
   inputs [
    A {from=MaterialExpressionOneMinus_0}
    B {from=MaterialExpressionMultiply_5}
   ]
  }
  MaterialExpressionMultiply_2 {
   class=MaterialExpressionMultiply
   inputs [
    A {from=MaterialExpressionDynamicParameter_2, output=Interaction}
    B {from=MaterialExpressionSubtract_2}
   ]
  }
  MaterialExpressionMultiply_21 {
   class=MaterialExpressionMultiply
   inputs [
    A {from=MaterialExpressionMultiply_23}
    B {from=MaterialExpressionParticleColor_2, output=A}
   ]
  }
  MaterialExpressionMultiply_22 {
   class=MaterialExpressionMultiply
   inputs [
    A {from=MaterialExpressionMaterialFunctionCall_17, output=Result}
    B {from=MaterialExpressionScalarParameter_17}
   ]
  }
  MaterialExpressionMultiply_23 {
   class=MaterialExpressionMultiply
   inputs [
    A {from=MaterialExpressionMultiply_24}
    B {from=MaterialExpressionDepthFade_2}
   ]
  }
  MaterialExpressionMultiply_24 {
   class=MaterialExpressionMultiply
   inputs [
    A {from=MaterialExpressionPower_2}
    B {from=MaterialExpressionScalarParameter_14}
   ]
  }
  MaterialExpressionMultiply_25 {
   class=MaterialExpressionMultiply
   inputs [
    A {from=MaterialExpressionMaterialFunctionCall_16, output=RGB}
    B {from=MaterialExpressionScalarParameter_18}
   ]
  }
  MaterialExpressionMultiply_26 {
   class=MaterialExpressionMultiply
   inputs [
    A {from=MaterialExpressionParticleColor_2, output=RGB}
    B {from=MaterialExpressionMultiply_25}
   ]
  }
  MaterialExpressionMultiply_27 {
   class=MaterialExpressionMultiply
   inputs [
    A {from=MaterialExpressionMaterialFunctionCall_18, output=Result}
    B {from=MaterialExpressionMultiply_22}
   ]
  }
  MaterialExpressionMultiply_3 {
   class=MaterialExpressionMultiply
   inputs [
    A {from=MaterialExpressionSubtract_0}
    B {from=""}
   ]
   properties {constB=100}
  }
  MaterialExpressionMultiply_4 {
   class=MaterialExpressionMultiply
   inputs [
    A {from=MaterialExpressionDynamicParameter_0, output=Spread}
    B {from=MaterialExpressionDynamicParameter_0, output=Spread}
   ]
  }
  MaterialExpressionMultiply_5 {
   class=MaterialExpressionMultiply
   inputs [
    A {from=MaterialExpressionSign_0}
    B {from=MaterialExpressionScalarParameter_1}
   ]
  }
  MaterialExpressionMultiply_8 {
   class=MaterialExpressionMultiply
   inputs [
    A {from=MaterialExpressionDynamicParameter_1, output=Fake Gravity}
    B {from=MaterialExpressionMaterialFunctionCall_2, output=Result}
   ]
  }
  MaterialExpressionMultiply_9 {
   class=MaterialExpressionMultiply
   inputs [
    A {from=MaterialExpressionMultiply_2}
    B {from=MaterialExpressionMultiply_12}
   ]
   properties {constB=100}
  }
  MaterialExpressionNormalize_0 {class=MaterialExpressionNormalize, inputs [VectorInput {from=MaterialExpressionAdd_5, output=""}]}
  MaterialExpressionOneMinus_0 {class=MaterialExpressionOneMinus, inputs [None {from=MaterialExpressionVertexColor_1, output=R}]}
  MaterialExpressionParticleColor_2 {class=MaterialExpressionParticleColor, inputs [], properties {}}
  MaterialExpressionParticlePositionWS_0 {class=MaterialExpressionParticlePositionWS, inputs [], properties {}}
  MaterialExpressionParticleRandom_0 {class=MaterialExpressionParticleRandom, inputs [], properties {}}
  MaterialExpressionPower_2 {
   class=MaterialExpressionPower
   inputs [
    Base {from=MaterialExpressionTextureSample_1, output=R}
    Exp {from=MaterialExpressionScalarParameter_16}
   ]
  }
  MaterialExpressionReflectionVectorWS_1 {class=MaterialExpressionReflectionVectorWS, inputs [CustomWorldNormal {from=MaterialExpressionTransform_0, output=""}], properties {}}
  MaterialExpressionReroute_1 {class=MaterialExpressionReroute, inputs [None {from=MaterialExpressionMaterialFunctionCall_12, output=Result}]}
  MaterialExpressionReroute_3 {class=MaterialExpressionReroute, inputs [None {from=MaterialExpressionMultiply_27}]}
  MaterialExpressionScalarParameter_0 {class=MaterialExpressionScalarParameter, properties {defaultValue=10, parameterName=Scatter Ofs Scale}}
  MaterialExpressionScalarParameter_1 {class=MaterialExpressionScalarParameter, properties {defaultValue=12, parameterName=Rot Speed}}
  MaterialExpressionScalarParameter_13 {class=MaterialExpressionScalarParameter, properties {defaultValue=20, parameterName=DepthFade}}
  MaterialExpressionScalarParameter_14 {class=MaterialExpressionScalarParameter, properties {defaultValue=1.25}}
  MaterialExpressionScalarParameter_15 {class=MaterialExpressionScalarParameter, properties {parameterName=f2}}
  MaterialExpressionScalarParameter_16 {class=MaterialExpressionScalarParameter, properties {defaultValue=0.2, parameterName=ssS}}
  MaterialExpressionScalarParameter_17 {class=MaterialExpressionScalarParameter, properties {defaultValue=2.5, parameterName=aef}}
  MaterialExpressionScalarParameter_18 {class=MaterialExpressionScalarParameter, properties {defaultValue=2, parameterName=a2t}}
  MaterialExpressionScalarParameter_19 {class=MaterialExpressionScalarParameter, properties {defaultValue=0.08, parameterName=seh}}
  MaterialExpressionScalarParameter_20 {class=MaterialExpressionScalarParameter, properties {defaultValue=1, parameterName=EnablePC}}
  MaterialExpressionScalarParameter_21 {class=MaterialExpressionScalarParameter, properties {defaultValue=0.8, parameterName=Desat}}
  MaterialExpressionSign_0 {class=MaterialExpressionSign, inputs [None {from=MaterialExpressionSubtract_3, output=""}], properties {}}
  MaterialExpressionSubtract_0 {
   class=MaterialExpressionSubtract
   inputs [
    A {from=MaterialExpressionVertexColor_0, output=R}
    B {from=""}
   ]
   properties {constB=0.5}
  }
  MaterialExpressionSubtract_1 {
   class=MaterialExpressionSubtract
   inputs [
    A {from=MaterialExpressionVectorNoise_0}
    B {from=""}
   ]
   properties {constB=0.5}
  }
  MaterialExpressionSubtract_2 {
   class=MaterialExpressionSubtract
   inputs [
    A {from=MaterialExpressionVertexColor_1, output=R}
    B {from=""}
   ]
   properties {constB=0.25}
  }
  MaterialExpressionSubtract_3 {
   class=MaterialExpressionSubtract
   inputs [
    A {from=MaterialExpressionParticleRandom_0}
    B {from=""}
   ]
   properties {constB=0.5}
  }
  MaterialExpressionTextureObjectParameter_0 {
   class=MaterialExpressionTextureObjectParameter
   inputs [
    UVs {from="", output=""}
    Apply View MipBias {from="", output=""}
   ]
   properties {
    group=MainNormal
    parameterName=MainNormal_Texture
    samplerType=SAMPLERTYPE_Normal
    sortPriority=0
    texture=/Game/Environments/Props/Rocks/Textures/T_Corrupt_Rock_AB_N
   }
  }
  MaterialExpressionTextureObjectParameter_4 {
   class=MaterialExpressionTextureObjectParameter
   inputs [
    UVs {from="", output=""}
    Apply View MipBias {from="", output=""}
   ]
   properties {group=Ice2, parameterName=Ice2_Texture, sortPriority=0, texture=&/T_IceGauntlets_Ice}
  }
  MaterialExpressionTextureObjectParameter_5 {
   class=MaterialExpressionTextureObjectParameter
   inputs [
    UVs {from="", output=""}
    Apply View MipBias {from="", output=""}
   ]
   properties {group=BumpIce, parameterName=BumpIce_Texture, sortPriority=0, texture=&T_IceGauntlets_Ice}
  }
  MaterialExpressionTextureSample_1 {
   class=MaterialExpressionTextureSample
   inputs [
    UVs {from=MaterialExpressionMaterialFunctionCall_14, output=UV}
    Tex {from=""}
    Apply View MipBias {from=""}
   ]
   properties {texture=&/T_IceGauntlets_IceD}
  }
  MaterialExpressionTransform_0 {class=MaterialExpressionTransform, inputs [None {from=MaterialExpressionMaterialFunctionCall_12, output=Result}]}
  MaterialExpressionVectorNoise_0 {class=MaterialExpressionVectorNoise, inputs [World Position {from=MaterialExpressionMultiply_3, output=""}], properties {}}
  MaterialExpressionVectorParameter_0 {
   class=MaterialExpressionVectorParameter
   properties {group=MainNormal, parameterName=MainNormal_UVControl, channelNames {a=VPan, b=UPan, g=VScale, r=UScale}, defaultValue=<rgba 1 1 0 0>}
  }
  MaterialExpressionVectorParameter_10 {class=MaterialExpressionVectorParameter, properties {parameterName=Color_Night, defaultValue=<rgba 0.65 0.7498 1 1>}}
  MaterialExpressionVectorParameter_2 {
   class=MaterialExpressionVectorParameter
   properties {group=Ice2, parameterName=Ice2_UVControl, channelNames {a=VPan, b=UPan, g=VScale, r=UScale}, defaultValue=<rgba 1 1 0 0>}
  }
  MaterialExpressionVectorParameter_7 {
   class=MaterialExpressionVectorParameter
   properties {group=Ref, parameterName=Ref_UVControl, channelNames {a=VPan, b=UPan, g=VScale, r=UScale}, defaultValue=<rgba 0.7 0.7 0 0>}
  }
  MaterialExpressionVectorParameter_8 {class=MaterialExpressionVectorParameter, properties {parameterName=Color_Morning, defaultValue=<rgba 1 0.8676 0.8 1>}}
  MaterialExpressionVectorParameter_9 {class=MaterialExpressionVectorParameter, properties {parameterName=Color_Day, defaultValue=<rgba 1 1 1 1>}}
  MaterialExpressionVertexColor_0 {class=MaterialExpressionVertexColor, inputs [], properties {}}
  MaterialExpressionVertexColor_1 {class=MaterialExpressionVertexColor, inputs [], properties {}}
  MaterialExpressionVertexColor_3 {class=MaterialExpressionVertexColor, inputs [], properties {}}
 }
 properties {
  bAutomaticallySetUsageInEditor=false
  bUsedWithNiagaraMeshParticles=true
  blendMode=BLEND_Translucent
  refractionMethod=RM_PixelNormalOffset
  shadingModel=MSM_Unlit
  usedShadingModels=Unlit
 }
}
