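# Unlit material for ice shard mesh particles in the Ice Gauntlets VFX, used with Niagara.
# A Fresnel term, with its exponent driven by a dynamic parameter, blends pale blue to near-white.
# The blend is tinted by particle color and fed through the shared emissive adjustment function.
# A separate branch multiplies each vertex's offset from the particle position by a cubed camera-distance fade, set by the "Size Fade by Distance" parameter.
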Material /TeaCakeIceGauntlets/Assets/VFX/Materials/M_IceGauntlets_IceShard
graph {
 properties {bAutomaticallySetUsageInEditor=false, bUsedWithNiagaraMeshParticles=true, shadingModel=MSM_Unlit, usedShadingModels=Unlit}
 nodes [
  n1 = Constant3Vector() {constant=<rgba 0.2244 0.4422 0.66 1>}
  n2 = Constant3Vector() {constant=<rgba 0.9 0.9917 1 0>}
  n3 = Fresnel(ExponentIn=DynamicParameter() {defaultValue=<rgba 2 1 1 1>, paramNames [Fresnel_exponent, Param2, Param3, Param4]}.Fresnel_exponent) {baseReflectFraction=0, exponent=2}
  MaterialFunctionCall(Emissive=Multiply(A=LinearInterpolate(A=@n1, B=@n2, Alpha=@n3), B=ParticleColor().RGB)) {materialFunction=&MF_FX_EmissiveAdjustments}
  n4 = ScalarParameter() {defaultValue=15000, parameterName=Size Fade by Distance}
  Multiply(
   A=Subtract(A=WorldPosition().XYZ, B=ParticlePositionWS())
   B=Power(Base=MaterialFunctionCall!(Distance=@n4) {materialFunction=/Game/Effects/Fort_Effects/Materials/_Functions/MF_CameraDistanceFade}.Result) {constExponent=3}
  )
 ]
}
