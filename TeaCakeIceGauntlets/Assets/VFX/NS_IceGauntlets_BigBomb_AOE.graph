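# Niagara system for the Ice Gauntlets big bomb area effect: a glowing ice dome filled with swirling snow and wind.
# A persistent three-mesh dome layers animated materials, surrounded by orbiting snow puffs, spiral air, wind gusts, and GPU ice shards and edge blizzard.
# User.FX_Scale sizes most layers, while User.Finish stops continuous spawning.
# Time-of-day tints adjust color and opacity, and system-level interpolated values drive dome fades.
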NiagaraSystem /TeaCakeIceGauntlets/Assets/VFX/NS_IceGauntlets_BigBomb_AOE
dependencies {
 usedDataInterfaces [
  &NiagaraDataInterfaceCurve
  &NiagaraDataInterfaceArrayFloat
  &NiagaraDataInterfaceSpriteRendererInfo
  &NiagaraDataInterfaceMeshRendererInfo
  &NiagaraDataInterfaceCamera
  &NiagaraDataInterfaceVectorCurve
  &NiagaraDataInterfaceCollisionQuery
  &NiagaraDataInterfaceColorCurve
 ]
 usedDynamicInputs [/Game/Effects/Niagara/Modules/OrientMestToVectorFromCustomRot]
 usedModules [
  &self:SystemSpawnScript.SystemScriptSource.SystemScriptGraph.NiagaraNodeAssignment_0.SetVariables_#1
  &SystemState
  &InterpolateOverTime
  &self:SystemSpawnScript.SystemScriptSource.SystemScriptGraph.NiagaraNodeAssignment_1.SetVariables_#2
 ]
}
emitters {
 Dome {
  emitterUpdateScript [
   &EmitterState(Life Cycle Mode=::NewEnumerator1, Loop Behavior=::NewEnumerator1, Loop Duration Mode=::NewEnumerator1, UseLoopDelay=false)
   &SpawnBurst_Instantaneous(Use Spawn Probability=false, UseLoopCountLimit=false)
  ]
  particleSpawnScript [
   &InitializeParticle(
    Color=&FNTimeOfDayColor(
     Night Color=<rgba 1.6406 2.5488 3 0.5>
     Morning Color=<rgba 2.474 3.3461 5 0.5>
     Day Color=<rgba 1.3151 2.1067 2.5 0.5>
     Evening Color=<rgba 1.3672 2.124 2.5 0.5>
    )
    Lifetime=4
    Mesh Scale Mode=::NewEnumerator3
    Mesh Uniform Scale=&Multiply_Float(A=10.05, B=$User.FX_Scale)
    Sprite Rotation Mode=::NewEnumerator2
    Sprite Size Mode=::NewEnumerator3
    Uniform Sprite Size=400
    UsePositionOffset=false
    Write Material Random=true
   )
   /SetVariables(Particles.CameraOffsetCore=-120) {isSetParameters=true}
   /SetVariables(Particles.MeshOrientation=$System.DomeOrientation, Particles.SpriteCamOfs=80) {isSetParameters=true}
  ]
  particleUpdateScript [
   &ParticleState(Kill Particles When Lifetime Has Elapsed=false, Let Infinitely Lived Particles Die When Emitter Deactivates=true)
   &ScaleMeshSize(Scale Factor=&VectorFromFloat(Value=&FloatFromCurve(
      FloatCurve=DataInterface() {propertyValues {Curve {
         defaultValue=1
         keys [
          {arriveTangent=4.1785, interpMode=RCIM_Cubic, leaveTangent=4.1785, tangentMode=RCTM_User, time=0, value=0.5}
          {arriveTangent=0.0224, interpMode=RCIM_Cubic, leaveTangent=0.0224, tangentMode=RCTM_User, time=0.2, value=1}
         ]
        }}}
      CurveIndex=$Emitter.Age
      Scale Curve=4
     )))
   &ScaleColor(
    Scale Alpha=&FloatFromCurve(
     FloatCurve=DataInterface() {propertyValues {Curve {keys [{interpMode=RCIM_Cubic, tangentMode=RCTM_User, time=0, value=2}, {interpMode=RCIM_Cubic, tangentMode=RCTM_User, time=0.06, value=1}]}}}
     Scale Curve=&Multiply_Float(
      A=&FNTimeOfDayScalar(Night Scale=1.9, Morning Scale=1.1, Day Scale=1.5, Evening Scale=1.3)
      B=&Multiply_Float(A=2, B=$System.InterpolateOverTime.InterpolateOverTime_Float.Moving Average)
     )
    )
    Scale RGB=&MakeVectorFromLinearColorRGB(Color=&MultiplyLinearColor(
      Linear Color=<rgba 0.1687 0.5831 0.9 1>
      Scale Factor=&MakeVector4_FromLinearColor(LinearColor=$System.ToDColor)
     ))
    ScaleA=true
    ScaleRGB=true
   )
   &DynamicMaterialParameters(
    Index 0 Param 1=1.3
    Index 0 Param 2=15
    Index 0 Param 3=&FloatFromCurve(
     FloatCurve=DataInterface() {propertyValues {Curve {keys [
         {arriveTangent=-0.9649, interpMode=RCIM_Cubic, leaveTangent=-0.9649, tangentMode=RCTM_User, time=0, value=1}
         {arriveTangent=-6.2582, leaveTangent=-1.25, tangentMode=RCTM_User, time=0.0624, value=0.1635}
         {arriveTangent=3.1809, interpMode=RCIM_Cubic, leaveTangent=3.1809, tangentMode=RCTM_User, time=0.8, value=0.1}
        ]}}}
     Scale Curve=&Lerp_Float(A=5, B=1, Alpha=$System.InterpolateOverTime001.InterpolateOverTime_Float.Moving Average)
    )
    Index 0 Param 4=&Multiply_Float(A=$User.FX_Scale, B=0.5)
    Module.Index0Param0Write=true
    Module.Index0Param1Write=true
    Module.Index0Param2Write=true
    Module.Index0Param3Write=true
   )
   /SetVariables(Particles.DynMatPar_Inner:Vector4f=&MakeVector4(
     X=2
     Y=9
     Z=&FloatFromCurve(
      FloatCurve=DataInterface() {propertyValues {Curve {keys [
          {arriveTangent=-17.4004, interpMode=RCIM_Cubic, leaveTangent=-17.4004, tangentMode=RCTM_User, time=0.02, value=1}
          {arriveTangent=-1.25, time=0.15, value=0.1}
          {arriveTangent=0.6539, interpMode=RCIM_Cubic, leaveTangent=0.6539, tangentMode=RCTM_User, time=0.8, value=0.1}
         ]}}}
      Scale Curve=&Lerp_Float(A=1, B=5, Alpha=$System.InterpolateOverTime001.InterpolateOverTime_Float.Moving Average)
     )
     W=&Multiply_Float(A=$User.FX_Scale, B=2)
    )) {isSetParameters=true}
   /SetVariables(Particles.DynMatPar_SnowThick:Vector4f=&MakeVector4(
     X=2
     Y=20
     Z=&FloatFromCurve(
      FloatCurve=DataInterface() {propertyValues {Curve {keys [
          {arriveTangent=-4.8072, interpMode=RCIM_Cubic, leaveTangent=-4.8072, tangentMode=RCTM_User, time=0, value=1}
          {arriveTangent=-0.2555, interpMode=RCIM_Constant, tangentMode=RCTM_User, time=0.15, value=0.73}
          {arriveTangentWeight=0.2033, interpMode=RCIM_Cubic, leaveTangentWeight=0.0467, tangentMode=RCTM_Break, time=0.8, value=0.73}
         ]}}}
      Scale Curve=&Lerp_Float(A=1.25, B=1.05, Alpha=$System.InterpolateOverTime001.InterpolateOverTime_Float.Moving Average)
     )
     W=&Multiply_Float(A=$User.FX_Scale, B=1)
    )) {isSetParameters=true}
   /SetVariables(Particles.ColorSnowThick:LinearColor=&MultiplyLinearColor(
     Linear Color=$System.ToDColor
     Scale Factor=&MakeVector4(
      X=0
      Y=1
      Z=5
      W=&Multiply_Float(
       A=&FloatFromCurve(
        FloatCurve=DataInterface() {propertyValues {Curve {keys [
            {interpMode=RCIM_Cubic, tangentMode=RCTM_User, time=0, value=0}
            {interpMode=RCIM_Cubic, tangentMode=RCTM_User, time=0.1038, value=1.1895}
            {arriveTangent=-0.0025, interpMode=RCIM_Cubic, leaveTangent=-0.0025, tangentMode=RCTM_User, time=0.9038, value=1.1895}
           ]}}}
        Scale Curve=&Multiply_Float(A=2, B=$System.InterpolateOverTime.InterpolateOverTime_Float.Moving Average)
       )
       B=&FNTimeOfDayScalar(Night Scale=1.7, Morning Scale=0.7, Evening Scale=0.8)
      )
     )
    )) {isSetParameters=true}
   /SetVariables(Particles.ColorCore:LinearColor=&MultiplyLinearColor(
     Linear Color=$Particles.Color
     Scale Factor=&MakeFloat4FromVectorAndFloat(
      "Vector (RGB)"=&MakeVectorFromLinearColorRGB(Color=&FNTimeOfDayColor(
        Night Color=<rgba 0.0241 0.2121 0.4698 1>
        Morning Color=<rgba 0 0.752 2.2 1>
        Day Color=<rgba 0 0.752 2.2 1>
        Evening Color=<rgba 0 0.752 2.2 1>
       ))
      "Float (Alpha)"=&Multiply_Float(A=1, B=$System.InterpolateOverTime.InterpolateOverTime_Float.Moving Average)
     )
    )) {isSetParameters=true}
   /SetVariables(Particles.ScaleThickSnow:Vector3f=&Multiply_VectorByFloat(Vector=&Multiply_VectorByFloat(Vector=$Particles.Scale, Float=0.45), Float=$User.FX_Scale)) {isSetParameters=true}
  ]
  rendererClasses [&NiagaraMeshRendererProperties]
  renderers [
   {rendererClass=&NiagaraMeshRendererProperties, rendererIndex=0}
   {rendererClass=&NiagaraMeshRendererProperties, rendererIndex=1}
   {rendererClass=&NiagaraMeshRendererProperties, rendererIndex=2}
  ]
 }
 Ice_Shards {
  simTarget=GPUComputeSim
  emitterUpdateScript [
   &EmitterState(Life Cycle Mode=::NewEnumerator1, Loop Duration=1, UseLoopDelay=false)
   &SpawnRate(
    Spawn Probability=&MakeCustomFloatFromBool(Bool=$User.Finish, True Float=0, False Float=1)
    SpawnRate=120
    Use Spawn Probability=true
   )
  ]
  particleSpawnScript [
   /SetVariables(
    Particles.Particle_Depth:NiagaraFloat=&FloatFromCurve(
     FloatCurve=DataInterface() {propertyValues {Curve {keys [
         {arriveTangent=1.2753, interpMode=RCIM_Cubic, leaveTangent=1.2753, tangentMode=RCTM_User, time=0, value=0}
         {arriveTangent=0.204, interpMode=RCIM_Cubic, leaveTangent=0.204, tangentMode=RCTM_User, time=0.9, value=1}
        ]}}}
     CurveIndex=&RandomRangeFloat(Minimum=0, Maximum=1)
    )
    Particles.Particle_Height:NiagaraFloat=&FloatFromCurve(
     FloatCurve=DataInterface() {propertyValues {Curve {keys [
         {arriveTangent=1.9477, interpMode=RCIM_Cubic, leaveTangent=1.9477, tangentMode=RCTM_User, time=0, value=0}
         {arriveTangent=0.1423, interpMode=RCIM_Cubic, leaveTangent=0.1423, tangentMode=RCTM_User, time=0.5, value=0.5}
         {arriveTangent=1.4427, interpMode=RCIM_Cubic, leaveTangent=1.4427, tangentMode=RCTM_User, time=1, value=1}
        ]}}}
     CurveIndex=&RandomRangeFloat(Minimum=0, Maximum=1)
    )
    Particles.Particle_Scale:NiagaraFloat=&RandomRangeFloat(Minimum=0, Maximum=1)
   ) {isSetParameters=true}
   &InitializeParticle(
    Color Maximum=<rgba 0.9 0.9833 1 1>
    Color Minimum=<rgba 0.6 0.8667 1 1>
    Color Mode=::NewEnumerator2
    Lifetime Max=1.5
    Lifetime Min=0.7
    Lifetime Mode=::NewEnumerator1
    Mass Mode=::NewEnumerator3
    Mesh Renderer Array Visibility Mode=::NewEnumerator3
    Mesh Renderer Info=DataInterface() {propertyValues {MeshRenderer=&self:Ice_Shards_1.NiagaraMeshRendererProperties_0}}
    Mesh Scale Mode=::NewEnumerator3
    Mesh Uniform Scale=&Multiply_Float(A=$User.FX_Scale, B=&Lerp_Float(A=4, B=10, Alpha=$Particles.Particle_Scale))
    Sprite Rotation Mode=::NewEnumerator2
    Sprite Size Mode=::NewEnumerator3
    Uniform Sprite Size=&Multiply_Float(A=$User.FX_Scale, B=&Lerp_Float(A=12, B=27, Alpha=$Particles.Particle_Scale))
    UsePositionOffset=false
    Write Material Random=true
   )
   &CameraOffset(Camera Offset Amount=6)
   &InitialMeshOrientation(Use Rotation Vector=false)
   /SetVariables(Particles.RotationDirection:Vector3f=&Lerp_Vector(
     A=&RandomRangeVector(Minimum=<-0.6 -0.6 1>, Maximum=<0.8 0.8 1>)
     B=&RandomRangeVector(Minimum=<-0.2 -0.2 1>, Maximum=<0.3 0.3 1>)
     Alpha=&FloatFromCurve(
      FloatCurve=DataInterface() {propertyValues {Curve {keys [
          {interpMode=RCIM_Cubic, tangentMode=RCTM_User, time=0, value=0}
          {arriveTangent=0.5342, interpMode=RCIM_Cubic, leaveTangent=0.5342, tangentMode=RCTM_User, time=1, value=1}
         ]}}}
      CurveIndex=$Particles.Particle_Depth
     )
    )) {isSetParameters=true}
   /SetVariables(Particles.isSnowFlake:NiagaraBool=&RandomBool()) {isSetParameters=true}
   /SetVariables(Particles.VisibilityTag=&MakeCustomIntFromBool(Bool=$Particles.isSnowFlake, True INT=0, False INT=1)) {isSetParameters=true}
   &DynamicMaterialParameters(
    Index 0 Param 1=&RandomRangeFloat(Minimum=1, Maximum=2)
    Module.Index0Param0Write=true
    Module.Index0Param1Write=false
    Module.Index0Param2Write=false
    Module.Index0Param3Write=false
   )
  ]
  particleUpdateScript [
   /SetVariables(Particles.ColorSnowFlake:LinearColor=&MakeLinearColorFromVectorAndFloat(
     "Vector (RGB)"=&MakeVectorFromLinearColorRGB(Color=&MultiplyLinearColor(
       Linear Color=&RandomRangeLinearColor(Minimum=<rgba 0.8 0.9333 1 1>, Maximum=<rgba 0.6 0.7333 1 1>)
       Scale Factor=&MakeVector4_FromLinearColor(LinearColor=$System.ToDColor)
      ))
     "Float (Alpha)"=&MakeCustomFloatFromBool(
      Bool=$Particles.isSnowFlake
      True Float=&FloatFromCurve(
       FloatCurve=DataInterface() {propertyValues {Curve {keys [
           {interpMode=RCIM_Cubic, tangentMode=RCTM_User, time=0, value=0}
           {interpMode=RCIM_Cubic, tangentMode=RCTM_User, time=0.15, value=1}
           {arriveTangent=-0.0592, interpMode=RCIM_Cubic, leaveTangent=-0.0592, tangentMode=RCTM_User, time=0.7, value=1}
           {interpMode=RCIM_Cubic, tangentMode=RCTM_User, time=1, value=0}
          ]}}}
       Scale Curve=&RandomRangeFloat(Minimum=0.2, Maximum=0.7)
      )
      False Float=0
     )
    )) {isSetParameters=true}
   &ShapeLocation(
    Handle Radius=40
    Large Radius=850
    MaskSpawnGroup=false
    Non Uniform Scale=<1 1 10>
    Offset=<0 0 250>
    Offset Mode=::NewEnumerator0
    Override Seed=false
    Rotation Mode=::NewEnumerator2
    Shape Primitive=::NewEnumerator3
    UseOwnerScale=true
    Yaw / Pitch / Roll=&MakeVector(
     X=&Multiply_Float(A=$Emitter.Age, B=&RandomRangeFloat(Minimum=0.2, Maximum=0.8))
     Y=0
     Z=0
    )
   )
   &ParticleState()
   &SpriteRotationRate(Rotation Rate=&Lerp_Float(A=100, B=300, Alpha=$Particles.Particle_Scale))
   &UpdateMeshOrientation(
    Rotation Rate=&Multiply_Float(
     A=&Lerp_Float(A=2, B=1, Alpha=$Particles.Particle_Depth)
     B=&Lerp_Float(A=3, B=1, Alpha=$Particles.Particle_Scale)
    )
    Rotation Vector=&RandomRangeVector(Minimum=<-1 -1 -1>, Maximum=<1 1 1>)
   )
   &ScaleMeshSize(Scale Factor=&VectorFromFloat(Value=&FloatFromCurve(FloatCurve=DataInterface() {propertyValues {Curve {keys [
          {interpMode=RCIM_Cubic, tangentMode=RCTM_User, time=0, value=0}
          {interpMode=RCIM_Cubic, tangentMode=RCTM_User, time=0.1, value=1}
          {arriveTangent=-3.406, interpMode=RCIM_Cubic, leaveTangent=-3.406, tangentMode=RCTM_User, time=1, value=0}
         ]}}})))
   &ScaleSpriteSize(Uniform Curve Sprite Scale=DataInterface() {propertyValues {Curve {keys [{interpMode=RCIM_Cubic, tangentMode=RCTM_User, time=0, value=0.6}, {interpMode=RCIM_Cubic, tangentMode=RCTM_User, time=0.2, value=1}]}}})
   &SpriteFacingAndAlignment(
    Sprite Facing=&Lerp_Vector(
     A=&RandomRangeVector(Minimum=<-1 -1 -1>, Maximum=<1 1 1>)
     B=&ReturnCameraVector(Camera Vectors=::NewEnumerator1, Coordinate Space=::Simulation)
     Alpha=&FloatFromCurve(
      FloatCurve=DataInterface() {propertyValues {Curve {keys [
          {arriveTangent=2.0555, interpMode=RCIM_Cubic, leaveTangent=2.0555, tangentMode=RCTM_User, time=0, value=0}
          {interpMode=RCIM_Cubic, tangentMode=RCTM_User, time=1, value=1}
         ]}}}
      CurveIndex=&RandomRangeFloat(Minimum=0, Maximum=1)
     )
    )
    WriteAlignment=false
    WriteFacing=true
   )
   &ScaleMeshSize(Scale Factor=&VectorFromFloat(Value=&FloatFromCurve(
      FloatCurve=DataInterface() {propertyValues {Curve {keys [{interpMode=RCIM_Cubic, tangentMode=RCTM_User, time=280, value=0}, {interpMode=RCIM_Cubic, tangentMode=RCTM_User, time=300, value=1}]}}}
      CurveIndex=&Abs_Float(Float:NiagaraFloat=&DistanceBetweenPositions(
        A=$Particles.Position
        B=&ReturnCameraPosition(Coordinate Space:ENiagaraCoordinateSpace=::Simulation, Camera Query:NiagaraDataInterfaceCamera=DataInterface())
       ))
     ))) {module=ScaleMeshSize001}
   &SolveForcesAndVelocity(Clamp Velocity=false, Limit Acceleration=false)
   &ScaleColor(Scale RGB=&MakeVectorFromLinearColorRGB(Color=$System.ToDColor), ScaleA=true, ScaleRGB=true)
  ]
  rendererClasses [&NiagaraSpriteRendererProperties, &NiagaraMeshRendererProperties]
  renderers [
   {rendererClass=&NiagaraSpriteRendererProperties}
   {rendererClass=&NiagaraMeshRendererProperties, rendererIndex=1}
  ]
 }
 SDF_Edge_Blizzard {
  simTarget=GPUComputeSim
  emitterUpdateScript [
   &EmitterState()
   &SpawnRate(
    Spawn Probability=&MakeCustomFloatFromBool(Bool=$User.Finish, True Float=0, False Float=1)
    SpawnRate=2000
    Use Spawn Probability=true
   )
  ]
  particleSpawnScript [
   &InitializeParticle(
    Color=&MultiplyLinearColor(
     Linear Color=<rgba 0.299 0.6857 1.4 1.1>
     Scale Factor=&MakeVector4_FromLinearColor(LinearColor=$System.ToDColor)
    )
    Lifetime Max=0.86
    Lifetime Min=0.4
    Lifetime Mode=::NewEnumerator1
    Sprite Size Max=<140 250>
    Sprite Size Min=<111 180>
    Sprite Size Mode=::NewEnumerator2
    Sprite UV Mode=::NewEnumerator3
    UsePositionOffset=true
    Write Material Random=true
   )
   &ShapeLocation(
    MaskSpawnGroup=false
    Override Seed=false
    Rotation Mode=::NewEnumerator2
    Sphere Radius=&Multiply_Float(A=&RandomRangeFloat(Minimum=690, Maximum=705), B=$User.FX_Scale)
    Sphere Surface Distribution=1
    UseOwnerScale=true
    Yaw / Pitch / Roll=&RandomRangeVector(Minimum=<0 0 0>, Maximum=<1 1 1>)
   )
   &self:ScratchModule(Surface Distance=0) {module=ScratchModule}
   &CameraOffset(Camera Offset Amount=50)
   &OffsetPosition(Position Offset=<0 0 50>, Position Offset Coordinate Space:ENiagaraCoordinateSpace=::Local)
   &DynamicMaterialParameters(
    Index 0 Param 1=3
    Module.Index0Float0Write=true
    Module.Index0Float1Write=false
    Module.Index0Float2Write=false
    Module.Index0Float3Write=false
   )
  ]
  particleUpdateScript [
   &ParticleState()
   &ScaleColor(
    Linear Color Curve=DataInterface() {propertyValues {
      AlphaCurve {keys [<at 0 0>,<at 0.1528 1.8>,<at 0.4497 0.6728>,<at 1 0>]}
      BlueCurve {keys [<at 0 1>,<at 1 1>]}
      GreenCurve {keys [<at 0 1>,<at 1 1>]}
      RedCurve {keys [<at 0 1>,<at 1 1>]}
     }}
    Scale Mode=2
   )
   &VortexForce(Origin Pull Amount=500, UseOriginOffset=false, Vortex Force Amount=2000)
   &GravityForce(Gravity=&RandomRangeVector(Minimum=<0 0 150>, Maximum=<0 0 350>))
   &SolveForcesAndVelocity(Clamp Velocity=false, Limit Acceleration=false)
   &ScaleSpriteSizeBySpeed(Max Scale Factor=<4 6>, Min Scale Factor=<0.9 1>)
  ]
  rendererClasses [&NiagaraSpriteRendererProperties]
  renderers [{rendererClass=&NiagaraSpriteRendererProperties}]
 }
 Snow_Puffs {
  emitterUpdateScript [
   &EmitterState(Delay First Loop Only=true, Life Cycle Mode=::NewEnumerator1, Loop Delay=0.8, Loop Duration=1, UseLoopDelay=true)
   &SpawnRate(
    Spawn Probability=&MakeCustomFloatFromBool(Bool=$User.Finish, True Float=0, False Float=1)
    SpawnRate=15
    Use Spawn Probability=true
   )
  ]
  particleSpawnScript [
   &InitializeParticle(
    Color Maximum=<rgba 0.505 0.6288 1 0.6>
    Color Minimum=<rgba 0.3863 0.5936 0.7344 0.2>
    Color Mode=::NewEnumerator2
    Lifetime Max=1
    Lifetime Min=0.7
    Lifetime Mode=::NewEnumerator1
    Mass Mode=::NewEnumerator3
    Position Offset=&Multiply_VectorByFloat(Vector=<1 1 1>, Float=$User.FX_Scale)
    Sprite Rotation Mode=::NewEnumerator2
    Sprite Size Max=<600 500>
    Sprite Size Min=<500 600>
    Sprite Size Mode=::NewEnumerator2
    Sprite UV Mode=::NewEnumerator3
    UsePositionOffset=false
    Write Material Random=true
   )
   &CameraOffset(Camera Offset Amount=&Multiply_Float(A=160, B=$User.FX_Scale))
   &ShapeLocation(
    MaskSpawnGroup=false
    Non Uniform Scale=<1 0.8 0.8>
    Override Seed=false
    Sphere Radius=&Multiply_Float(A=200, B=$User.FX_Scale)
    UseOwnerScale=true
   ) {module=ShapeLocation001}
  ]
  particleUpdateScript [
   &ParticleState()
   &ScaleSpriteSize(
    Uniform Curve Scale=&RandomRangeFloat(Minimum=1.1, Maximum=1.5)
    Uniform Curve Sprite Scale=DataInterface() {propertyValues {Curve {keys [{interpMode=RCIM_Cubic, tangentMode=RCTM_User, time=0, value=0.6}, {interpMode=RCIM_Cubic, tangentMode=RCTM_User, time=0.7, value=1}]}}}
   )
   &ScaleColor(
    Scale Alpha=&FloatFromCurve(
     FloatCurve=DataInterface() {propertyValues {Curve {keys [
         {interpMode=RCIM_Cubic, tangentMode=RCTM_User, time=0.0021, value=0.2238}
         {interpMode=RCIM_Cubic, tangentMode=RCTM_User, time=0.2964, value=0.4519}
         {interpMode=RCIM_Cubic, tangentMode=RCTM_User, time=1, value=0}
        ]}}}
     Scale Curve=&FNTimeOfDayScalar(Night Scale=0.5, Morning Scale=0.2, Day Scale=0.3, Evening Scale=0.3)
    )
    Scale RGB=&MakeVectorFromLinearColorRGB(Color=&FNTimeOfDayColor(
      Night Color=<rgba 0.48 0.5895 0.8 1>
      Morning Color=<rgba 0.9347 0.9078 1 1>
      Day Color=<rgba 1 0.925 0.85 1>
      Evening Color=<rgba 1 0.9427 0.9857 1>
     ))
    ScaleA=true
    ScaleRGB=true
   )
   &SubUVAnimation(
    "Sprite Renderer "=DataInterface() {propertyValues {SpriteRenderer=&self:NiagaraEmitter_0.NiagaraSpriteRendererProperties_0}}
    UseEndFrame=false
    UseStartFrame=false
   )
   &Drag(Drag=4, Rotational Drag=2, Use Linear Drag=true, Use Rotational Drag=true)
   &SolveForcesAndVelocity(Clamp Velocity=false, Limit Acceleration=false)
   &RotateAroundPoint(
    Derive the Rotation Axis=false
    Radius:NiagaraFloat=&Multiply_Float(A=&RandomRangeFloat(Minimum=650, Maximum=900), B=$User.FX_Scale)
    Rotation Center:NiagaraPosition=&AddVectorToPosition(
     Position=&AddVectorToPosition(
      Position=$Engine.Owner.Position
      Vector=&Multiply_VectorByFloat(Vector=&RandomRangeVector(Minimum=<0 0 -400>, Maximum=<1 1 400>), Float=$User.FX_Scale)
     )
     Vector=&MakeVector(
      X=0
      Y=0
      Z=&FloatFromCurve(
       FloatCurve=DataInterface() {propertyValues {Curve {defaultValue=1, keys [{tangentMode=RCTM_User, time=0, value=0}, {tangentMode=RCTM_User, time=1, value=1}]}}}
       Scale Curve=&RandomRangeFloat(Minimum=200, Maximum=300)
      )
     )
    )
    Rotation Phase:NiagaraFloat=&RandomRangeFloat(Minimum=0, Maximum=1)
    Rotation Rate:NiagaraFloat=&RandomRangeFloat(Minimum=0.4, Maximum=0.6)
   )
   &DynamicMaterialParameters(
    Index 0 Param 1=2
    Index 0 Param 3=1
    Module.Index0Param0Write=true
    Module.Index0Param1Write=true
    Module.Index0Param2Write=true
    Module.Index0Param3Write=true
   )
  ]
  rendererClasses [&NiagaraSpriteRendererProperties]
  renderers [{rendererClass=&NiagaraSpriteRendererProperties}]
 }
 Spiral_Air {
  emitterUpdateScript [
   &EmitterState(Life Cycle Mode=::NewEnumerator1, Loop Delay=0.5, Loop Duration=1, UseLoopDelay=false)
   &SpawnRate(
    Spawn Probability=&MakeCustomFloatFromBool(Bool=$User.Finish, True Float=0, False Float=1)
    SpawnRate=10
    Use Spawn Probability=true
   )
   &SpawnBurst_Instantaneous(Spawn Count=2, Use Spawn Probability=false, UseLoopCountLimit=true)
  ]
  particleSpawnScript [
   &InitializeParticle(
    Color=<rgba 0.3599 0.691 1 1>
    Mesh Scale=&Multiply_VectorByFloat(Vector=&RandomRangeVector(Minimum=<10 10 15>, Maximum=<12 12 22>), Float=$User.FX_Scale)
    Mesh Scale Mode=::NewEnumerator4
    Position Offset=<0 0 -250>
    UsePositionOffset=true
    Write Material Random=true
   )
   &InitialMeshOrientation(
    Mesh Orientation Mode=::NewEnumerator2
    Rotation=&RandomRangeVector(Minimum=<0 0 0>, Maximum=<0 0 1>)
    Use Rotation Vector=true
   )
  ]
  particleUpdateScript [
   &ParticleState()
   &UpdateMeshOrientation(Rotation Rate=0.1, Rotation Vector=<0 0 1>)
   &ScaleMeshSize(Scale Factor=&Multiply_VectorByFloat(
     Vector=&VectorFromCurve(
      VectorCurve=DataInterface() {propertyValues {XCurve {keys [<at 0 0.6>,<at 1 1.4>]}, YCurve {keys [<at 0 0.6>,<at 1 1.4>]}, ZCurve {keys [<at 0 1>,<at 1 1>]}}}
      Scale Curve=&Multiply_VectorByFloat(Vector=<1 1 1>, Float=1)
     )
     Float=1
    ))
   &DynamicMaterialParameters(
    Index 0 Param 1=&FloatFromCurve(FloatCurve=DataInterface() {propertyValues {Curve {keys [<at 0 0.1>,<at 1 0.5>]}}})
    Index 0 Param 2=&RandomRangeFloat(Minimum=500, Maximum=400)
    Module.Index0Param0Write=true
    Module.Index0Param1Write=true
    Module.Index0Param2Write=true
    Module.Index0Param3Write=true
   )
   &ScaleColor(
    Scale RGB=&Multiply_VectorByFloat(
     Vector=&MakeVectorFromLinearColorRGB(Color=&FNTimeOfDayColor(
       Night Color=<rgba 0.474 0.6858 1 1>
       Morning Color=<rgba 0.75 0.8507 1 1>
       Day Color=<rgba 0.75 0.8507 1 1>
       Evening Color=<rgba 0.75 0.8507 1 1>
      ))
     Float=&FloatFromCurve(
      FloatCurve=DataInterface() {propertyValues {Curve {keys [
          <at 0 0 RCIM_Cubic>
          <at 0.2 1 RCIM_Cubic>
          {arriveTangent=-0.6427, interpMode=RCIM_Cubic, leaveTangent=-0.6427, tangentMode=RCTM_User, time=1, value=0}
         ]}}}
      Scale Curve=&Multiply_Float(A=1, B=&FNTimeOfDayScalar(Night Scale=0.8, Morning Scale=0.8, Evening Scale=0.8))
     )
    )
    ScaleA=true
    ScaleRGB=true
   )
  ]
  rendererClasses [&NiagaraMeshRendererProperties]
  renderers [{rendererClass=&NiagaraMeshRendererProperties, rendererIndex=0}]
 }
 WindGusts {
  emitterUpdateScript [
   &EmitterState(Life Cycle Mode=::NewEnumerator1, Loop Duration=0.9, UseLoopDelay=false)
   &SpawnRate(
    Spawn Probability=&MakeCustomFloatFromBool(Bool=$User.Finish, True Float=0, False Float=1)
    SpawnRate=20
    Use Spawn Probability=true
   )
  ]
  particleSpawnScript [
   &PartitionParticles()
   /SetVariables(
    Particles.Particle_Height:NiagaraFloat=&RandomRangeFloat(Minimum=0, Maximum=1)
    Particles.VisibilityTag=$Particles.PartitionParticles.Partition
   ) {isSetParameters=true}
   /SetVariables(Particles.Position2:NiagaraPosition=&AddVectorToPosition(
     Position=<0 0 0>
     Vector=&MakeVector(
      X=0
      Y=0
      Z=&MergeHelper_FloatSwitch(
       0=&Multiply_Float(A=$User.FX_Scale, B=&Lerp_Float(A=-300, B=300, Alpha=$Particles.Particle_Height))
       1=0
       2=?
       3=?
       Index=$Particles.PartitionParticles.Partition
       Num=::NewEnumerator0
      )
     )
    )) {isSetParameters=true}
   &InitializeParticle(
    Color=&RandomRangeLinearColor(Minimum=<rgba 0.67 0.8912 1.25 1>, Maximum=<rgba 0.8125 1.461 2 0.4957>)
    Lifetime Max=1.3
    Lifetime Min=0.7
    Lifetime Mode=::NewEnumerator1
    Mass Mode=::NewEnumerator3
    Mesh Scale=&Multiply_VectorByFloat(
     Vector=&MergeHelper_VectorSwitch(
      0=&Multiply_VectorByFloat(
       Vector=&RandomRangeVector(Minimum=<1 1 0.5>, Maximum=<1 1 1.5>)
       Float=&Multiply_Float(
        A=$User.FX_Scale
        B=&Lerp_Float(
         A=15
         B=15
         Alpha=&FloatFromCurve(
          FloatCurve=DataInterface() {propertyValues {Curve {keys [
              {arriveTangent=5.7133, interpMode=RCIM_Cubic, leaveTangent=5.7133, tangentMode=RCTM_User, time=0, value=0}
              {interpMode=RCIM_Cubic, tangentMode=RCTM_User, time=0.5, value=1}
              {arriveTangent=-6.0152, interpMode=RCIM_Cubic, leaveTangent=-6.0152, tangentMode=RCTM_User, time=1, value=0}
             ]}}}
          CurveIndex=$Particles.Particle_Height
          Scale Curve=&RandomRangeFloat(Minimum=0.9, Maximum=1)
         )
        )
       )
      )
      1=&RandomRangeVector(Minimum=<5 5 5>, Maximum=<15 15 7>)
      2=?
      3=?
      Index=$Particles.PartitionParticles.Partition
      Num=::NewEnumerator0
     )
     Float=&FloatFromCurve(
      FloatCurve=DataInterface() {propertyValues {Curve {defaultValue=1, keys [{tangentMode=RCTM_User, time=0, value=0.3}, {tangentMode=RCTM_User, time=0.7, value=1}]}}}
      CurveIndex=$Emitter.Age
      Scale Curve=&Multiply_Float(A=1, B=1)
     )
    )
    Mesh Scale Mode=::NewEnumerator4
    Sprite Rotation Mode=::NewEnumerator2
    UsePositionOffset=false
    Write Material Random=true
   )
   &SystemLocation(
    MaskSpawnGroup=false
    Offset=&MakeVector(
     X=0
     Y=0
     Z=&MergeHelper_FloatSwitch(
      0=&Multiply_Float(A=$User.FX_Scale, B=&Lerp_Float(A=-200, B=350, Alpha=$Particles.Particle_Height))
      1=0
      2=?
      3=?
      Index=$Particles.PartitionParticles.Partition
      Num=::NewEnumerator0
     )
    )
   )
   &InitialMeshOrientation(
    Mesh Orientation Mode=::NewEnumerator2
    Orientation Coordinate Space=::World
    Rotation=&MergeHelper_VectorSwitch(
     0=&RandomRangeVector(Minimum=<-0.04 -0.04 -1>, Maximum=<0 0 1>)
     1=&RandomRangeVector(Minimum=<0 0 0>, Maximum=<1 1 1>)
     2=?
     3=?
     Index=$Particles.PartitionParticles.Partition
     Num=::NewEnumerator0
    )
    Use Rotation Vector=true
   )
   &CameraOffset(Camera Offset Amount=15)
  ]
  particleUpdateScript [
   &ParticleState()
   &ShapeLocation(
    Handle Radius=40
    Large Radius=850
    MaskSpawnGroup=false
    Non Uniform Scale=<1 1 10>
    Offset=&MakeVector(
     X=0
     Y=0
     Z=&FloatFromCurve(
      FloatCurve=DataInterface() {propertyValues {Curve {defaultValue=1, keys [{tangentMode=RCTM_User, time=0, value=0}, {tangentMode=RCTM_User, time=1, value=1}]}}}
      Scale Curve=200
     )
    )
    Offset Mode=::NewEnumerator0
    Override Seed=false
    Rotation Mode=::NewEnumerator2
    Shape Primitive=::NewEnumerator3
    UseOwnerScale=true
    Yaw / Pitch / Roll=&MakeVector(
     X=&Multiply_Float(A=$Emitter.Age, B=&RandomRangeFloat(Minimum=0.2, Maximum=0.8))
     Y=0
     Z=0
    )
   ) {module=ShapeLocation001}
   &ScaleMeshSize(Scale Factor=&MergeHelper_VectorSwitch(
     0=&Multiply_VectorByFloat(
      Vector=&MakeVector(
       X=1.5
       Y=1.3
       Z=&SelectFloatFromArray(
        Array Sampling Mode=::NewEnumerator0
        Direct Array Index=?
        Float Selection Array=DataInterface() {propertyValues {FloatData [1,-1]}}
        Random Seed=0
       )
      )
      Float=&Lerp_Float(
       A=0.8
       B=1.2
       Alpha=&FloatFromCurve(FloatCurve=DataInterface() {propertyValues {Curve {keys [
            {arriveTangent=2.1221, interpMode=RCIM_Cubic, leaveTangent=2.1221, tangentMode=RCTM_User, time=0, value=0}
            {interpMode=RCIM_Cubic, tangentMode=RCTM_User, time=0.86, value=1}
           ]}}})
      )
     )
     1=&Multiply_VectorByFloat(
      Vector=<1 1 1>
      Float=&FloatFromCurve(FloatCurve=DataInterface() {propertyValues {Curve {
          defaultValue=1
          keys [
           {arriveTangent=6.6205, interpMode=RCIM_Cubic, leaveTangent=6.6205, tangentMode=RCTM_User, time=0, value=0}
           {interpMode=RCIM_Cubic, tangentMode=RCTM_User, time=0.5, value=1}
           {arriveTangent=-6.754, tangentMode=RCTM_User, time=1, value=0}
          ]
         }}})
     )
     2=?
     3=?
     Index=$Particles.PartitionParticles.Partition
     Num=::NewEnumerator0
    ))
   &UpdateMeshOrientation(
    Rotation Rate=&RandomRangeFloat(Minimum=0.5, Maximum=1.1)
    Rotation Vector=&MergeHelper_VectorSwitch(0=<0 0 1>, 1=<0.5 0.5 1>, 2=?, 3=?, Index=$Particles.PartitionParticles.Partition, Num=::NewEnumerator0)
   )
   &ScaleColor(
    Scale Alpha=&FloatFromCurve(
     FloatCurve=DataInterface() {propertyValues {Curve {keys [
         {arriveTangent=6.2301, interpMode=RCIM_Cubic, leaveTangent=6.2301, tangentMode=RCTM_User, time=0, value=0}
         {arriveTangent=-6.912, interpMode=RCIM_Cubic, leaveTangent=-6.912, tangentMode=RCTM_User, time=0.1, value=3}
         {arriveTangent=0.4048, interpMode=RCIM_Cubic, leaveTangent=0.4048, tangentMode=RCTM_User, time=0.7, value=0}
        ]}}}
     Scale Curve=&RandomRangeFloat(Minimum=0.8, Maximum=1)
    )
    Scale RGB=&VectorFromFloat(Value=&Lerp_Float(
      A=1
      B=2
      Alpha=&FloatFromCurve(FloatCurve=DataInterface() {propertyValues {Curve {keys [
           {interpMode=RCIM_Cubic, tangentMode=RCTM_User, time=0, value=0}
           {interpMode=RCIM_Cubic, tangentMode=RCTM_User, time=0.1, value=1}
           {interpMode=RCIM_Cubic, tangentMode=RCTM_User, time=1, value=0}
          ]}}})
     ))
    ScaleA=true
    ScaleRGB=true
   )
   &SolveForcesAndVelocity(Clamp Velocity=false, Limit Acceleration=false)
   &DynamicMaterialParameters(
    Index 0 Param 1=&FloatFromCurve(FloatCurve=DataInterface() {propertyValues {Curve {keys [
         {arriveTangent=1.526, interpMode=RCIM_Cubic, leaveTangent=1.526, tangentMode=RCTM_User, time=0, value=0}
         {interpMode=RCIM_Cubic, tangentMode=RCTM_User, time=1, value=1}
        ]}}})
    Module.Index0Param0Write=true
    Module.Index0Param1Write=true
    Module.Index0Param2Write=false
    Module.Index0Param3Write=true
   )
   &ScaleColor(
    Scale Alpha=&FNTimeOfDayScalar(Night Scale=0.25, Morning Scale=0.25, Day Scale=0.7, Evening Scale=0.3)
    Scale RGB=&MakeVectorFromLinearColorRGB(Color=&FNTimeOfDayColor(
      Night Color=<rgba 0.2375 0.2917 0.3958 1>
      Morning Color=<rgba 0.8 0.52 0.52 1>
      Day Color=<rgba 1 0.9 0.8 1>
      Evening Color=<rgba 0.8 0.48 0.5067 1>
     ))
    ScaleA=true
    ScaleRGB=true
   ) {module=ScaleColor001}
   /SetVariables(Particles.Position2:NiagaraPosition=&AddVectorToPosition(Position=$Particles.Position2, Vector=&MakeVector(X=0, Y=0, Z=15))) {isSetParameters=true}
  ]
  rendererClasses [&NiagaraMeshRendererProperties]
  renderers [
   {rendererClass=&NiagaraMeshRendererProperties, rendererIndex=0}
   {rendererClass=&NiagaraMeshRendererProperties, rendererIndex=1}
  ]
 }
 decal {rendererClasses [], renderers []}
}
summary {
 systemName=NS_IceGauntlets_BigBomb_AOE
 emitters [
  {emitterName=Dome, rendererClasses [&NiagaraMeshRendererProperties]}
  {emitterName=WindGusts, rendererClasses [&NiagaraMeshRendererProperties]}
  {emitterName=Snow_Puffs, rendererClasses [&NiagaraSpriteRendererProperties]}
  {
   emitterName=Ice_Shards
   simTarget=GPUComputeSim
   rendererClasses [&NiagaraSpriteRendererProperties, &NiagaraMeshRendererProperties]
  }
  {emitterName=Spiral_Air, rendererClasses [&NiagaraMeshRendererProperties]}
  {emitterName=decal, rendererClasses []}
  {emitterName=SDF_Edge_Blizzard, simTarget=GPUComputeSim, rendererClasses [&NiagaraSpriteRendererProperties]}
 ]
 userVariables [
  User.Finish {type=NiagaraBool}
  User.FX_DomeAngle {type=Vector3f, description="", defaultValue {struct=Vector3f, value=<0 0 0>}}
  User.FX_Scale {type=NiagaraFloat, defaultValue=1, description=""}
 ]
}
