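# Niagara system for the big bomb explosion effect of the Ice Gauntlets.
# It layers emitters named for intro and wide flashes, a stylized blast, fat streaks, ground shockwave, smoke, ice chunks, snowflakes and debris, plus one named Empty.
# Dependencies include SystemState, two set-variable modules, camera fade, time-of-day color and vector-scaling dynamic inputs.
# It exposes a User.IsTrain bool, whose use is not shown here.
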NiagaraSystem /TeaCakeIceGauntlets/Assets/VFX/NS_IceGauntlets_BigBomb_Explode
dependencies {
 usedDataInterfaces [&NiagaraDataInterfaceCamera]
 usedDynamicInputs [
  /Niagara/DynamicInputs/Camera/CameraFade
  &FNTimeOfDayColor
  &Multiply_Vector2D_ByFloat
  &Multiply_VectorByFloat
 ]
 usedModules [
  &self:SystemSpawnScript.SystemScriptSource.SystemScriptGraph.NiagaraNodeAssignment_1.SetVariables_#1
  &self:SystemSpawnScript.SystemScriptSource.SystemScriptGraph.NiagaraNodeAssignment_0.SetVariables_#2
  &SystemState
 ]
}
emitters {
 Empty {rendererClasses [], renderers []}
 LW_Blast_Stylized {rendererClasses [], renderers []}
 LW_Debris {rendererClasses [], renderers []}
 LW_Flash_wide {rendererClasses [], renderers []}
 LW_Ground_Shockwave {rendererClasses [], renderers []}
 LW_Ice_Chunks {rendererClasses [], renderers []}
 LW_Intro_Flash001 {rendererClasses [], renderers []}
 LW_Intro_Flash002 {rendererClasses [], renderers []}
 LW_SnowFlakes {rendererClasses [], renderers []}
 LW_Streaks_Fat {rendererClasses [], renderers []}
 smoke {rendererClasses [], renderers []}
}
summary {
 systemName=NS_IceGauntlets_BigBomb_Explode
 emitters [
  {emitterName=LW_Intro_Flash001, rendererClasses []}
  {emitterName=LW_Flash_wide, rendererClasses []}
  {emitterName=LW_Streaks_Fat, rendererClasses []}
  {emitterName=smoke, rendererClasses []}
  {emitterName=LW_Ice_Chunks, rendererClasses []}
  {emitterName=LW_Blast_Stylized, rendererClasses []}
  {emitterName=LW_SnowFlakes, rendererClasses []}
  {emitterName=Empty, rendererClasses []}
  {emitterName=LW_Debris, rendererClasses []}
  {emitterName=LW_Ground_Shockwave, rendererClasses []}
  {emitterName=LW_Intro_Flash002, rendererClasses []}
 ]
 userVariables [User.IsTrain {type=NiagaraBool}]
}
