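# Niagara explosion effect for the Ice Gauntlets big bomb, evidently its water-impact variant.
# Outward_Splashes bursts sprites from a shape around the blast, flinging them from an origin offset below, with downward acceleration, growth and curve-driven material parameters.
# Water_Droplets is a GPU burst launched up and outward, scaled by speed under gravity and killed by volume; both emitters use distance culling.
# The remaining flash, blast, ice, snow, debris and smoke emitters show no modules or renderers.
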NiagaraSystem /TeaCakeIceGauntlets/Assets/VFX/NS_IceGauntlets_BigBomb_Explode_Water
dependencies {
 usedDataInterfaces [&NiagaraDataInterfaceCurve, &NiagaraDataInterfaceVectorCurve, &NiagaraDataInterfaceColorCurve]
 usedDynamicInputs [&FNTimeOfDayColor]
 usedModules [
  &self:SystemSpawnScript.SystemScriptSource.SystemScriptGraph.NiagaraNodeAssignment_0.SetVariables_#1
  &SystemState
 ]
}
emitters {
 Empty {rendererClasses [], renderers []}
 LW_Blast_Stylized {rendererClasses [], renderers []}
 LW_Debris {rendererClasses [], renderers []}
 LW_Flash_wide {rendererClasses [], renderers []}
 LW_Ice_Chunks {rendererClasses [], renderers []}
 LW_Intro_Flash001 {rendererClasses [], renderers []}
 LW_Intro_Flash002 {rendererClasses [], renderers []}
 LW_SnowFlakes {rendererClasses [], renderers []}
 Outward_Splashes {
  emitterUpdateScript [
   &EmitterState(
    Enable Distance Culling=true
    MaxDistance=12000
    Module.bUseMaxDistance=true
    Module.bUseMinDistance=false
    Scalability Mode=::NewEnumerator1
   )
   &SpawnBurst_Instantaneous(Spawn Count=&UniformRangedInt(Minimum=15, Maximum=20), Spawn Time=0.07, Use Spawn Probability=false)
  ]
  particleSpawnScript [
   /Niagara/Modules/Spawn/Initialization/InitializeParticle(
    Color=<rgba 1.0379 1.5 1.2246 1>
    Lifetime=&RandomRangeFloat(Minimum=0.75, Maximum=1.25)
    Module.Write Color=true
    Module.Write Lifetime=true
    Module.Write Mass=true
    Module.Write Material Random=true
    Module.Write Position=false
    Module.Write Scale=false
    Module.Write SpriteRotation=true
    Module.Write SpriteSize=true
    Position=&SimulationPosition()
    Sprite Rotation=&UniformRangedFloat(Minimum=0, Maximum=360)
    Sprite Size=&Vector2DFromFloat(Value=&UniformRangedFloat(Minimum=60, Maximum=80))
   )
   &ShapeLocation(MaskSpawnGroup=false, Override Seed=false, Ring Radius=300, Shape Primitive=::NewEnumerator4, UseOwnerScale=true)
   &AddVelocityFromPoint(
    Module.Use Falloff Distance=false
    Origin Offset=&RandomRangeVector(Minimum=<0 0 -200>, Maximum=<0 0 -400>)
    Origin Offset Coordinate Space=::World
    Velocity Strength=&UniformRangedFloat(Minimum=1500, Maximum=800)
   )
  ]
  particleUpdateScript [
   &ParticleState()
   &AccelerationForce(Acceleration=&UniformRangedVector(Minimum=<0 0 -980>, Maximum=<0 0 -1800>))
   &Drag(Use Linear Drag=true, Use Rotational Drag=false) {enabled=false}
   &SolveForcesAndVelocity(Clamp Velocity=false, Limit Acceleration=false)
   &ScaleSpriteSize(Scale Factor=&Vector2DFromFloat(Value=&FloatFromCurve(
      FloatCurve=DataInterface() {propertyValues {Curve {keys [
          {arriveTangent=12.5197, interpMode=RCIM_Cubic, leaveTangent=12.5197, tangentMode=RCTM_User, time=0, value=0.4376}
          {arriveTangent=0.8929, interpMode=RCIM_Cubic, leaveTangent=0.8929, tangentMode=RCTM_User, time=1, value=5}
         ]}}}
      Scale Curve=2
     )))
   &ScaleColor(
    Scale Alpha=&FloatFromCurve(FloatCurve=DataInterface() {propertyValues {Curve {keys [<at 0 -1>,<at 0.009 0.9316>,<at 0.7645 0.2881>,<at 1 -1>]}}})
    Scale RGB=&VectorFromCurve(VectorCurve=DataInterface() {propertyValues {
       XCurve {keys [<at 0.0047 1.2511>,<at 0.9953 0.6912>]}
       YCurve {keys [<at 0.0047 1.2511>,<at 0.9953 0.6912>]}
       ZCurve {keys [<at 0.0047 1.2511>,<at 0.9953 0.6912>]}
      }})
    ScaleA=true
    ScaleRGB=true
   )
   &DynamicMaterialParameters(
    Index 0 Param 1=&FloatFromCurve(FloatCurve=DataInterface() {propertyValues {Curve {keys [
         {interpMode=RCIM_Cubic, tangentMode=RCTM_Break, time=0, value=1}
         {arriveTangent=-3.3696, interpMode=RCIM_Cubic, leaveTangent=-3.3696, tangentMode=RCTM_User, time=1, value=0}
        ]}}})
    Index 0 Param 2=&FloatFromCurve(FloatCurve=DataInterface() {propertyValues {Curve {keys [{arriveTangent=0.4361, interpMode=RCIM_Cubic, leaveTangent=0.4361, tangentMode=RCTM_User, time=0, value=0.05}, <at 1 0.25>]}}})
    Module.Param0WriteEnabled=true
    Module.Param1WriteEnabled=true
    Module.Param2WriteEnabled=true
    Module.Param3WriteEnabled=true
   ) {enabled=false}
   &DynamicMaterialParameters(
    Index 0 Param 1=&FloatFromCurve(FloatCurve=DataInterface() {propertyValues {Curve {keys [
         {arriveTangent=-2.0663, interpMode=RCIM_Cubic, leaveTangent=-2.0663, tangentMode=RCTM_User, time=0.25, value=1}
         {arriveTangent=-1.3063, interpMode=RCIM_Cubic, leaveTangent=-1.3063, tangentMode=RCTM_User, time=1.0011, value=0}
        ]}}})
    Index 0 Param 2=&FloatFromCurve(FloatCurve=DataInterface() {propertyValues {Curve {keys [<at -0.0086 -1.3001>,<at 0.9945 1.9153>]}}})
    Index 0 Param 3=0.5
    Module.Param0WriteEnabled=true
    Module.Param1WriteEnabled=false
    Module.Param2WriteEnabled=true
    Module.Param3WriteEnabled=true
   ) {module=DynamicMaterialParameters001}
   &CameraOffset(Camera Offset Amount=&FloatFromCurve(FloatCurve=DataInterface() {propertyValues {Curve {keys [
         {interpMode=RCIM_Cubic, leaveTangent=1080.3575, tangentMode=RCTM_Break, time=0, value=15}
         {arriveTangent=135, interpMode=RCIM_Cubic, leaveTangent=135, tangentMode=RCTM_Break, time=0.12, value=94}
         <at 1 150>
        ]}}})) {enabled=false}
   &SpriteRotationRate(Rotation Rate=&Lerp_Float(A=-55, B=55, Alpha=$Particles.Initial.MaterialRandom))
  ]
  rendererClasses [&NiagaraSpriteRendererProperties]
  renderers [{rendererClass=&NiagaraSpriteRendererProperties}]
 }
 Water_Droplets {
  simTarget=GPUComputeSim
  emitterUpdateScript [
   &EmitterState(
    Enable Distance Culling=true
    Life Cycle Mode=::NewEnumerator1
    Loop Behavior=::NewEnumerator1
    Loop Duration=1
    MaxDistance=4800
    Module.bUseMaxDistance=true
    Module.bUseMinDistance=false
    Scalability Mode=::NewEnumerator1
    UseLoopDelay=false
   )
   &SpawnBurst_Instantaneous(Spawn Count=&UniformRangedInt(Minimum=45, Maximum=100), Spawn Time=0.125, Use Spawn Probability=false)
  ]
  particleSpawnScript [
   &InitializeParticle(
    Color=<rgba 4.9219 4.9654 5 10>
    Lifetime=&UniformRangedFloat(Minimum=0.5, Maximum=3)
    Mass=&UniformRangedFloat(Minimum=0.75, Maximum=1.75)
    Module.Write Color=true
    Module.Write Lifetime=true
    Module.Write Mass=true
    Module.Write Material Random=true
    Module.Write Position=false
    Module.Write Scale=false
    Module.Write SpriteRotation=false
    Module.Write SpriteSize=true
    Position=&SimulationPosition()
    Sprite Size=&RandomRangeVector2D(Minimum=<15 25>, Maximum=<15 15>)
   )
   &SphereLocation(
    MaskSpawnGroup=false
    Non Uniform Scale=<1.5 1.5 0.5>
    Offset=<0 0 10>
    Override Seed=false
    Sphere Radius=60
    Surface Only=false
    UseNonUniformScale=true
    UseOffset=true
    UseSphereOrientation=false
   )
   &AddVelocityFromPoint(Module.Use Falloff Distance=false, Velocity Strength=&UniformRangedFloat(Minimum=350, Maximum=800))
   &AddVelocity(Coordinate Space=::World, Velocity=<0 0 1500>)
  ]
  particleUpdateScript [
   &ParticleState()
   &GravityForce(Gravity=<0 0 -2250>)
   &Drag(Drag=&UniformRangedFloat(Minimum=1, Maximum=3), Use Linear Drag=true, Use Rotational Drag=false) {enabled=false}
   &SolveForcesAndVelocity(Clamp Velocity=false, Limit Acceleration=false)
   &ScaleSpriteSizeBySpeed(Max Scale Factor=<1 2>, Min Scale Factor=<0 0.05>)
   &ScaleColor(
    Scale Mode=::NewEnumerator1
    Scale RGBA=&Vector4FromCurve(Vector4Curve=DataInterface() {propertyValues {
       AlphaCurve {keys [<at 0 0>,<at 0.5 1>,<at 1 0>]}
       BlueCurve {keys [<at 0 0.96>,<at 0.5 0.578>]}
       GreenCurve {keys [<at 0 1>,<at 0.5 0.6>]}
       RedCurve {keys [<at 0 0.933>,<at 0.5 0.5625>]}
      }})
    ScaleRGBA=true
   )
   &ScaleSpriteSize(Scale Factor=&Vector2DFromFloat(Value=&FloatFromCurve(FloatCurve=DataInterface() {propertyValues {Curve {keys [<at 0 1>,<at 1 0.4>]}}})))
   &KillParticlesInVolume(Kill Shape=::NewEnumerator2)
  ]
  rendererClasses [&NiagaraSpriteRendererProperties]
  renderers [{rendererClass=&NiagaraSpriteRendererProperties}]
 }
 smoke {rendererClasses [], renderers []}
}
summary {
 systemName=NS_IceGauntlets_BigBomb_Explode_Water
 emitters [
  {emitterName=LW_Intro_Flash001, rendererClasses []}
  {emitterName=LW_Flash_wide, rendererClasses []}
  {emitterName=smoke, rendererClasses []}
  {emitterName=LW_Ice_Chunks, rendererClasses []}
  {emitterName=LW_Blast_Stylized, rendererClasses []}
  {emitterName=LW_SnowFlakes, rendererClasses []}
  {emitterName=Empty, rendererClasses []}
  {emitterName=LW_Debris, rendererClasses []}
  {emitterName=LW_Intro_Flash002, rendererClasses []}
  {emitterName=Outward_Splashes, rendererClasses [&NiagaraSpriteRendererProperties]}
  {emitterName=Water_Droplets, simTarget=GPUComputeSim, rendererClasses [&NiagaraSpriteRendererProperties]}
 ]
 userVariables [User.IsTrain {type=NiagaraBool}]
}
