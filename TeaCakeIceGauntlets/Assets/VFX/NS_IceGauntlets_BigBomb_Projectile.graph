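# Niagara projectile effect for the Ice Gauntlets' big bomb: an icy dome wrapped in snow and wind.
# The Dome's layered meshes and sprite grow while blending from User.Muzzle Pos to the simulation position.
# Movement-based spawning trails GPU ice shards and vortex-pulled snow puffs, while partitioned WindGusts meshes spin around it.
# FX_Scale drives most sizes, time-of-day inputs tint the layers, and shard meshes shrink away near the camera.
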NiagaraSystem /TeaCakeIceGauntlets/Assets/VFX/NS_IceGauntlets_BigBomb_Projectile
dependencies {
 usedDataInterfaces [
  &NiagaraDataInterfaceCurve
  &NiagaraDataInterfaceSpriteRendererInfo
  &NiagaraDataInterfaceCamera
  &NiagaraDataInterfaceMeshRendererInfo
  &NiagaraDataInterfaceArrayFloat
 ]
 usedDynamicInputs [/Game/Effects/Niagara/Modules/OrientMestToVectorFromCustomRot]
 usedModules [
  &self:SystemSpawnScript.SystemScriptSource.SystemScriptGraph.NiagaraNodeAssignment_0.SetVariables_#1
  &SystemState
  &self:SystemSpawnScript.SystemScriptSource.SystemScriptGraph.NiagaraNodeAssignment_1.SetVariables_#2
 ]
}
emitters {
 Dome {
  emitterUpdateScript [
   &EmitterState(
    Inactive Response=::NewEnumerator1
    Life Cycle Mode=::NewEnumerator1
    Loop Behavior=::NewEnumerator1
    Loop Duration Mode=::NewEnumerator1
    UseLoopDelay=false
   )
   &SpawnBurst_Instantaneous(Use Spawn Probability=false, UseLoopCountLimit=false)
  ]
  particleSpawnScript [
   &InitializeParticle(
    Color=&FNTimeOfDayColor(
     Night Color=<rgba 0.1285 1.4591 3 1>
     Morning Color=<rgba 0 1.7264 5 1>
     Day Color=<rgba 0.1302 1.7134 2.5 1>
     Evening Color=<rgba 0.1719 0.9746 2.5 1>
    )
    Lifetime=4
    Mesh Scale Mode=::NewEnumerator3
    Mesh Uniform Scale=&Multiply_Float(A=12, B=$User.FX_Scale)
    Position=$User.Muzzle Pos
    Position Mode=::NewEnumerator1
    Sprite Rotation Mode=::NewEnumerator2
    Sprite Size Mode=::NewEnumerator3
    Uniform Sprite Size=500
    UsePositionOffset=false
    Write Material Random=true
   )
   &InitialMeshOrientation(
    Mesh Orientation Mode=::NewEnumerator2
    Orientation Coordinate Space=::World
    Rotation=&RandomRangeVector(Minimum=<-0.05 -0.05 -0.05>, Maximum=<0.05 0.05 0.05>)
    Rotation Coordinate Space=::NewEnumerator1
    Use Rotation Vector=true
   ) {enabled=false}
   /SetVariables(Particles.CameraOffsetCore=-120) {isSetParameters=true}
   /SetVariables(Particles.MeshOrientation=$System.DomeOrientation, Particles.SpriteCamOfs=80) {isSetParameters=true}
  ]
  particleUpdateScript [
   &ParticleState(Kill Particles When Lifetime Has Elapsed=false, Let Infinitely Lived Particles Die When Emitter Deactivates=true)
   &ScaleSpriteSize(
    Uniform Curve Index=$Emitter.Age
    Uniform Curve Scale=&Waveform(
     "Waveform [1]"=::NewEnumerator0
     "[1] Phase Scale"=&RandomRangeFloat(Minimum=3, Maximum=8, Evaluation Type=::NewEnumerator1)
     "[1] Frequency"=1
     "[1] Amplitude Min/Max"=<0 1>
     "[1] Scale"=0.3
     "[1] Bias"=1
    )
    Uniform Curve Sprite Scale=DataInterface() {propertyValues {Curve {keys [
        <at 0 0.25 RCIM_Cubic>
        {arriveTangent=0.3939, interpMode=RCIM_Cubic, leaveTangent=0.3939, time=0.2035, value=1.0281}
        <at 0.3935 0.405>
        {tangentMode=RCTM_User, time=0.6, value=1}
       ]}}}
   )
   &ScaleMeshSize(Scale Factor=&VectorFromFloat(Value=&FloatFromCurve(
      FloatCurve=DataInterface() {propertyValues {Curve {
         defaultValue=1
         keys [
          {arriveTangent=4.1785, interpMode=RCIM_Cubic, leaveTangent=4.1785, tangentMode=RCTM_User, time=0, value=0.1}
          {arriveTangent=1.1179, interpMode=RCIM_Cubic, leaveTangent=1.1179, tangentMode=RCTM_User, time=0.2, value=0.6228}
          {arriveTangent=0.0224, interpMode=RCIM_Cubic, leaveTangent=0.0224, tangentMode=RCTM_User, time=0.7, value=1}
         ]
        }}}
      CurveIndex=$Emitter.Age
     )))
   /SetVariables(
    Particles.CameraOffsetDarkGlow=400
    Particles.ColorDarkGlow:LinearColor=&MultiplyLinearColor(
     Linear Color=&FNTimeOfDayColor(
      Night Color=<rgba 0.0762 0.0925 0.1354 1>
      Morning Color=<rgba 0.7 0.49 0.5573 0.6>
      Day Color=<rgba 0.7559 0.8381 1 1>
      Evening Color=<rgba 1 0.6049 0.4432 1>
     )
     Scale Factor=&MakeFloat4FromVectorAndFloat(
      "Vector (RGB)"=<1 1 1>
      "Float (Alpha)"=&FloatFromCurve(FloatCurve=DataInterface() {propertyValues {Curve {keys [{interpMode=RCIM_Cubic, tangentMode=RCTM_User, time=0, value=0}, {interpMode=RCIM_Cubic, tangentMode=RCTM_User, time=0.2, value=1}]}}})
     )
    )
    Particles.DynParDarkGlow:Vector4f=<600 0 0 0>
   ) {enabled=false, isSetParameters=true}
   &ScaleColor(
    Scale Alpha=&FloatFromCurve(
     FloatCurve=DataInterface() {propertyValues {Curve {keys [{interpMode=RCIM_Cubic, tangentMode=RCTM_User, time=0, value=0.1}, {interpMode=RCIM_Cubic, tangentMode=RCTM_User, time=0.06, value=1}]}}}
     Scale Curve=&Multiply_Float(A=&FNTimeOfDayScalar(Night Scale=1.9, Morning Scale=1.1, Day Scale=1.5, Evening Scale=1.3), B=2)
    )
    Scale RGB=&MakeVectorFromLinearColorRGB(Color=&MultiplyLinearColor(
      Linear Color=<rgba 0.45 0.705 0.9 1>
      Scale Factor=&MakeVector4_FromLinearColor(LinearColor=$System.ToDColor)
     ))
    ScaleA=true
    ScaleRGB=true
   )
   &DynamicMaterialParameters(
    Index 0 Param 1=0.35
    Index 0 Param 2=5
    Index 0 Param 3=&FloatFromCurve(FloatCurve=DataInterface() {propertyValues {Curve {keys [
         {arriveTangent=-0.9649, interpMode=RCIM_Cubic, leaveTangent=-0.9649, tangentMode=RCTM_User, time=0, value=1}
         {arriveTangent=-6.2582, leaveTangent=-1.25, tangentMode=RCTM_User, time=0.03, value=0}
         {arriveTangent=3.1809, interpMode=RCIM_Cubic, leaveTangent=3.1809, tangentMode=RCTM_User, time=0.8, value=0}
        ]}}})
    Index 0 Param 4=&Multiply_Float(A=$User.FX_Scale, B=2)
    Module.Index0Param0Write=true
    Module.Index0Param1Write=true
    Module.Index0Param2Write=true
    Module.Index0Param3Write=true
   )
   /SetVariables(Particles.DynMatPar_Inner:Vector4f=&MakeVector4(
     X=2
     Y=9
     Z=&FloatFromCurve(FloatCurve=DataInterface() {propertyValues {Curve {keys [
          {arriveTangent=-17.4004, interpMode=RCIM_Cubic, leaveTangent=-17.4004, tangentMode=RCTM_User, time=0.02, value=1}
          {arriveTangent=-1.25, time=0.15, value=0}
          {arriveTangent=0.6539, interpMode=RCIM_Cubic, leaveTangent=0.6539, tangentMode=RCTM_User, time=0.8, value=0}
         ]}}})
     W=&Multiply_Float(A=$User.FX_Scale, B=2)
    )) {isSetParameters=true}
   /SetVariables(Particles.DynMatPar_SnowThick:Vector4f=&MakeVector4(
     X=1
     Y=44
     Z=&FloatFromCurve(
      FloatCurve=DataInterface() {propertyValues {Curve {keys [
          {arriveTangent=-4.8072, interpMode=RCIM_Cubic, leaveTangent=-4.8072, tangentMode=RCTM_User, time=0, value=1}
          {arriveTangent=-0.2555, interpMode=RCIM_Constant, tangentMode=RCTM_User, time=0.15, value=0.73}
          {arriveTangentWeight=0.2033, interpMode=RCIM_Cubic, leaveTangentWeight=0.0467, tangentMode=RCTM_Break, time=0.8, value=0.73}
         ]}}}
      Scale Curve=1.05
     )
     W=&Multiply_Float(A=$User.FX_Scale, B=2)
    )) {isSetParameters=true}
   /SetVariables(Particles.ColorSnowThick:LinearColor=&MultiplyLinearColor(
     Linear Color=$System.ToDColor
     Scale Factor=&MakeVector4(
      X=1
      Y=1
      Z=1
      W=&Multiply_Float(
       A=&FloatFromCurve(FloatCurve=DataInterface() {propertyValues {Curve {keys [
            {interpMode=RCIM_Cubic, tangentMode=RCTM_User, time=0, value=0}
            {interpMode=RCIM_Cubic, tangentMode=RCTM_User, time=0.1, value=1}
            {arriveTangent=-0.0025, interpMode=RCIM_Cubic, leaveTangent=-0.0025, tangentMode=RCTM_User, time=0.9, value=1}
           ]}}})
       B=&FNTimeOfDayScalar(Night Scale=1.7, Morning Scale=0.7, Evening Scale=0.8)
      )
     )
    )) {isSetParameters=true}
   /SetVariables(Particles.ColorCore:LinearColor=&MultiplyLinearColor(
     Linear Color=$Particles.Color
     Scale Factor=&MakeFloat4FromVectorAndFloat("Vector (RGB)"=&MakeVectorFromLinearColorRGB(Color=<rgba 0 1.1709 3 1>), "Float (Alpha)"=1)
    )) {isSetParameters=true}
   /SetVariables(Particles.ScaleThickSnow:Vector3f=&Multiply_VectorByFloat(Vector=$Particles.Scale, Float=1)) {isSetParameters=true}
   /SetVariables(Particles.Position=&Lerp_Vector(
     A=Linked() {linkedVariable {name=User.Muzzle Pos, type=NiagaraPosition}}
     B=Linked() {linkedVariable {name=Engine.Emitter.SimulationPosition, type=NiagaraPosition}}
     Alpha=&FloatFromCurve(
      FloatCurve=DataInterface() {propertyValues {Curve {keys [
          {arriveTangent=4.807, interpMode=RCIM_Cubic, leaveTangent=4.807, tangentMode=RCTM_User, time=0.01, value=0}
          {arriveTangent=3.1897, interpMode=RCIM_Cubic, leaveTangent=3.1897, tangentMode=RCTM_User, time=0.25, value=1}
         ]}}}
      CurveIndex=$Engine.System.Age
     )
    )) {isSetParameters=true}
  ]
  rendererClasses [&NiagaraMeshRendererProperties, &NiagaraSpriteRendererProperties]
  renderers [
   {rendererClass=&NiagaraMeshRendererProperties, rendererIndex=0}
   {rendererClass=&NiagaraMeshRendererProperties, rendererIndex=1}
   {rendererClass=&NiagaraMeshRendererProperties, rendererIndex=2}
   {rendererClass=&NiagaraSpriteRendererProperties, rendererIndex=3}
  ]
 }
 Ice_Shards {
  simTarget=GPUComputeSim
  emitterUpdateScript [
   &EmitterState()
   &SpawnPerUnit(Module.Use Max Movement Threshold=true, Module.Use Movement Tolerance=false, Spawn Group=1, Use Spawn Probability=false)
   &SpawnRate(
    Spawn Probability=&FloatFromCurve(
     FloatCurve=DataInterface() {propertyValues {Curve {keys [
         {interpMode=RCIM_Cubic, tangentMode=RCTM_User, time=0.1, value=0}
         {interpMode=RCIM_Cubic, tangentMode=RCTM_User, time=1, value=1}
         {arriveTangent=0.0061, interpMode=RCIM_Cubic, leaveTangent=-0.1111, tangentMode=RCTM_Break, time=9, value=1}
         <at 10.6 0 RCIM_Cubic>
        ]}}}
     CurveIndex=$Emitter.Age
    )
    SpawnRate=90
    Use Spawn Probability=false
   ) {enabled=false}
   &SpawnRate(
    Spawn Probability=&FloatFromCurve(
     FloatCurve=DataInterface() {propertyValues {Curve {keys [
         {interpMode=RCIM_Cubic, tangentMode=RCTM_User, time=0.1, value=0}
         {interpMode=RCIM_Cubic, tangentMode=RCTM_User, time=1, value=1}
         {arriveTangent=0.0061, interpMode=RCIM_Cubic, leaveTangent=-0.1111, tangentMode=RCTM_Break, time=9, value=1}
         <at 10.6 0 RCIM_Cubic>
        ]}}}
     CurveIndex=$Emitter.Age
    )
    SpawnGroup=1
    SpawnRate=40
    Use Spawn Probability=false
   ) {enabled=false, module=SpawnRate001}
  ]
  particleSpawnScript [
   /SetVariables(
    Particles.Particle_Depth:NiagaraFloat=&FloatFromCurve(
     FloatCurve=DataInterface() {propertyValues {Curve {keys [
         {arriveTangent=1.2753, interpMode=RCIM_Cubic, leaveTangent=1.2753, tangentMode=RCTM_User, time=0, value=0}
         {arriveTangent=0.204, interpMode=RCIM_Cubic, leaveTangent=0.204, tangentMode=RCTM_User, time=0.9, value=1}
        ]}}}
     CurveIndex=&RandomRangeFloat(Minimum=0, Maximum=1)
    )
    Particles.Particle_Height:NiagaraFloat=&FloatFromCurve(
     FloatCurve=DataInterface() {propertyValues {Curve {keys [
         {arriveTangent=1.9477, interpMode=RCIM_Cubic, leaveTangent=1.9477, tangentMode=RCTM_User, time=0, value=0}
         {arriveTangent=0.1423, interpMode=RCIM_Cubic, leaveTangent=0.1423, tangentMode=RCTM_User, time=0.5, value=0.5}
         {arriveTangent=1.4427, interpMode=RCIM_Cubic, leaveTangent=1.4427, tangentMode=RCTM_User, time=1, value=1}
        ]}}}
     CurveIndex=&RandomRangeFloat(Minimum=0, Maximum=1)
    )
    Particles.Particle_Scale:NiagaraFloat=&RandomRangeFloat(Minimum=0, Maximum=1)
   ) {isSetParameters=true}
   &InitializeParticle(
    Color Maximum=<rgba 0.9 0.9833 1 1>
    Color Minimum=<rgba 0.6 0.8667 1 1>
    Color Mode=::NewEnumerator2
    Lifetime Max=1.2
    Lifetime Min=0.5
    Lifetime Mode=::NewEnumerator1
    Mass Mode=::NewEnumerator3
    Mesh Renderer Array Visibility Mode=::NewEnumerator3
    Mesh Renderer Info=DataInterface() {propertyValues {MeshRenderer=&self:Ice_Shards_1.NiagaraMeshRendererProperties_0}}
    Mesh Scale Mode=::NewEnumerator3
    Mesh Uniform Scale=&Multiply_Float(A=$User.FX_Scale, B=&Lerp_Float(A=5, B=12, Alpha=$Particles.Particle_Scale))
    Sprite Rotation Mode=::NewEnumerator2
    Sprite Size Mode=::NewEnumerator3
    Uniform Sprite Size=&Multiply_Float(A=$User.FX_Scale, B=&Lerp_Float(A=12, B=27, Alpha=$Particles.Particle_Scale))
    UsePositionOffset=false
    Write Material Random=true
   )
   &ShapeLocation(
    MaskSpawnGroup=false
    Override Seed=false
    Sphere Radius=&Multiply_Float(A=150, B=$User.FX_Scale)
    UseOwnerScale=true
   )
   &AddVelocity(
    Velocity=&AddVectorToPosition(
     Position=&Lerp_Vector(
      A=&Multiply_VectorByFloat(Vector=$Particles.ShapeLocation.ShapeTangent, Float=1)
      B=&Multiply_VectorByFloat(Vector=$Particles.ShapeLocation.ShapeBitangent, Float=-4)
      Alpha=0.5
     )
     Vector=&RandomRangeVector(Minimum=<0 0 0.5>, Maximum=<0 0 2>)
    )
    Velocity Speed Scale=&RandomRangeFloat(Minimum=200, Maximum=300)
   ) {enabled=false}
   &CameraOffset(Camera Offset Amount=6)
   &InitialMeshOrientation(Use Rotation Vector=false)
   /SetVariables(Particles.RotationDirection:Vector3f=&Lerp_Vector(
     A=&RandomRangeVector(Minimum=<-0.6 -0.6 1>, Maximum=<0.8 0.8 1>)
     B=&RandomRangeVector(Minimum=<-0.2 -0.2 1>, Maximum=<0.3 0.3 1>)
     Alpha=&FloatFromCurve(
      FloatCurve=DataInterface() {propertyValues {Curve {keys [
          {interpMode=RCIM_Cubic, tangentMode=RCTM_User, time=0, value=0}
          {arriveTangent=0.5342, interpMode=RCIM_Cubic, leaveTangent=0.5342, tangentMode=RCTM_User, time=1, value=1}
         ]}}}
      CurveIndex=$Particles.Particle_Depth
     )
    )) {isSetParameters=true}
   /SetVariables(Particles.isSnowFlake:NiagaraBool=&RandomBool()) {isSetParameters=true}
   /SetVariables(Particles.VisibilityTag=&MakeCustomIntFromBool(Bool=$Particles.isSnowFlake, True INT=0, False INT=1)) {enabled=false, isSetParameters=true}
   &DynamicMaterialParameters(
    Index 0 Param 1=&RandomRangeFloat(Minimum=1, Maximum=2)
    Module.Index0Param0Write=true
    Module.Index0Param1Write=false
    Module.Index0Param2Write=false
    Module.Index0Param3Write=false
   )
   &AddVelocity(Module.Use Falloff Distance=false, Override Seed=false, Velocity Mode=::NewEnumerator1, Velocity Speed=500) {module=AddVelocity001}
  ]
  particleUpdateScript [
   /SetVariables(Particles.ColorSnowFlake:LinearColor=&MakeLinearColorFromVectorAndFloat(
     "Vector (RGB)"=&MakeVectorFromLinearColorRGB(Color=&MultiplyLinearColor(
       Linear Color=&RandomRangeLinearColor(Minimum=<rgba 0.8 0.9333 1 1>, Maximum=<rgba 0.6 0.7333 1 1>)
       Scale Factor=&MakeVector4_FromLinearColor(LinearColor=$System.ToDColor)
      ))
     "Float (Alpha)"=&MakeCustomFloatFromBool(
      Bool=$Particles.isSnowFlake
      True Float=&FloatFromCurve(
       FloatCurve=DataInterface() {propertyValues {Curve {keys [
           {interpMode=RCIM_Cubic, tangentMode=RCTM_User, time=0, value=0}
           {interpMode=RCIM_Cubic, tangentMode=RCTM_User, time=0.15, value=1}
           {arriveTangent=-0.0592, interpMode=RCIM_Cubic, leaveTangent=-0.0592, tangentMode=RCTM_User, time=0.7, value=1}
           {interpMode=RCIM_Cubic, tangentMode=RCTM_User, time=1, value=0}
          ]}}}
       Scale Curve=&RandomRangeFloat(Minimum=0.5, Maximum=1)
      )
      False Float=0
     )
    )) {isSetParameters=true}
   &ParticleState()
   &SpriteRotationRate(Rotation Rate=&Lerp_Float(A=100, B=300, Alpha=$Particles.Particle_Scale))
   &UpdateMeshOrientation(
    Rotation Rate=&Multiply_Float(
     A=&Lerp_Float(A=2, B=1, Alpha=$Particles.Particle_Depth)
     B=&Lerp_Float(A=3, B=1, Alpha=$Particles.Particle_Scale)
    )
    Rotation Vector=&RandomRangeVector(Minimum=<-1 -1 -1>, Maximum=<1 1 1>)
   )
   &ScaleMeshSize(Scale Factor=&VectorFromFloat(Value=&FloatFromCurve(FloatCurve=DataInterface() {propertyValues {Curve {keys [
          {interpMode=RCIM_Cubic, tangentMode=RCTM_User, time=0, value=0}
          {interpMode=RCIM_Cubic, tangentMode=RCTM_User, time=0.1, value=1}
          {arriveTangent=-3.406, interpMode=RCIM_Cubic, leaveTangent=-3.406, tangentMode=RCTM_User, time=1, value=0}
         ]}}})))
   &ScaleSpriteSize(Uniform Curve Sprite Scale=DataInterface() {propertyValues {Curve {keys [{interpMode=RCIM_Cubic, tangentMode=RCTM_User, time=0, value=0.6}, {interpMode=RCIM_Cubic, tangentMode=RCTM_User, time=0.2, value=1}]}}})
   &VortexForce(
    Influence Falloff Exponent=2
    Influence Falloff Radius=1000
    Invert Influence Falloff=false
    Origin Pull Amount=1000
    Use Influence Falloff=true
    UseOriginOffset=false
    Vortex Force Amount=10000
   ) {enabled=false}
   &GravityForce(Gravity=&RandomRangeVector(Minimum=<0 0 -100>, Maximum=<0 0 -250>))
   &CurlNoiseForce(
    Bias Noise=false
    Noise Strength=&MaskFloatBySpawnGroup(Masked Float=350, Passthrough Float=100)
    Pan Noise=false
   )
   &Drag(Drag=4, Use Linear Drag=true, Use Rotational Drag=true)
   &SolveForcesAndVelocity(Clamp Velocity=false, Limit Acceleration=false)
   &SpriteFacingAndAlignment(
    Sprite Facing=&Lerp_Vector(
     A=&RandomRangeVector(Minimum=<-1 -1 -1>, Maximum=<1 1 1>)
     B=&ReturnCameraVector(Camera Vectors=::NewEnumerator1, Coordinate Space=::Simulation)
     Alpha=&FloatFromCurve(
      FloatCurve=DataInterface() {propertyValues {Curve {keys [
          {arriveTangent=2.0555, interpMode=RCIM_Cubic, leaveTangent=2.0555, tangentMode=RCTM_User, time=0, value=0}
          {interpMode=RCIM_Cubic, tangentMode=RCTM_User, time=1, value=1}
         ]}}}
      CurveIndex=&RandomRangeFloat(Minimum=0, Maximum=1)
     )
    )
    WriteAlignment=false
    WriteFacing=true
   )
   &ScaleMeshSize(Scale Factor=&VectorFromFloat(Value=&FloatFromCurve(
      FloatCurve=DataInterface() {propertyValues {Curve {keys [{interpMode=RCIM_Cubic, tangentMode=RCTM_User, time=280, value=0}, {interpMode=RCIM_Cubic, tangentMode=RCTM_User, time=300, value=1}]}}}
      CurveIndex=&Abs_Float(Float:NiagaraFloat=&DistanceBetweenPositions(
        A=$Particles.Position
        B=&ReturnCameraPosition(Coordinate Space:ENiagaraCoordinateSpace=::Simulation, Camera Query:NiagaraDataInterfaceCamera=DataInterface())
       ))
     ))) {module=ScaleMeshSize001}
   &ScaleColor(
    Scale Alpha=2
    Scale RGB=&MakeVectorFromLinearColorRGB(Color=$System.ToDColor)
    ScaleA=true
    ScaleRGB=true
   )
  ]
  rendererClasses [&NiagaraSpriteRendererProperties, &NiagaraMeshRendererProperties]
  renderers [
   {rendererClass=&NiagaraSpriteRendererProperties}
   {rendererClass=&NiagaraMeshRendererProperties, rendererIndex=1}
  ]
 }
 Snow_Puffs {
  emitterUpdateScript [
   &EmitterState()
   &SpawnPerUnit(Module.Use Max Movement Threshold=true, Module.Use Movement Tolerance=false, Spawn Spacing=55, Use Spawn Probability=false)
   &SpawnRate(
    Spawn Probability=&FloatFromCurve(
     FloatCurve=DataInterface() {propertyValues {Curve {keys [
         {interpMode=RCIM_Cubic, tangentMode=RCTM_User, time=0, value=0}
         {interpMode=RCIM_Cubic, tangentMode=RCTM_User, time=1, value=1}
         {arriveTangent=-0.2753, interpMode=RCIM_Cubic, leaveTangent=-0.2753, tangentMode=RCTM_User, time=8.5, value=0}
        ]}}}
     CurveIndex=$Engine.System.Age
    )
    SpawnRate=17
    Use Spawn Probability=false
   ) {enabled=false}
  ]
  particleSpawnScript [
   &InitializeParticle(
    Color Maximum=<rgba 0.3594 0.5709 1 0.6>
    Color Minimum=<rgba 0.3863 0.5936 0.7344 0.2>
    Color Mode=::NewEnumerator2
    Lifetime Max=1.5
    Lifetime Min=1.5
    Lifetime Mode=::NewEnumerator1
    Mass Mode=::NewEnumerator3
    Position Offset=&Multiply_VectorByFloat(Vector=<2000 0 0>, Float=$User.FX_Scale)
    Sprite Rotation Mode=::NewEnumerator2
    Sprite Size Mode=::NewEnumerator3
    Uniform Sprite Size=&Multiply_Float(A=$User.FX_Scale, B=&RandomRangeFloat(Minimum=450, Maximum=560))
    UsePositionOffset=false
    Write Material Random=true
   )
   &ShapeLocation(MaskSpawnGroup=false, Override Seed=false, Shape Primitive=::NewEnumerator4, UseOwnerScale=true) {enabled=false}
   &AddVelocity(
    Velocity=&AddVectorToPosition(
     Position=&Lerp_Vector(
      A=&Multiply_VectorByFloat(Vector=$Particles.ShapeLocation.ShapeTangent, Float=2)
      B=&Multiply_VectorByFloat(Vector=$Particles.ShapeLocation.ShapeBitangent, Float=-4)
      Alpha=0.5
     )
     Vector=&RandomRangeVector(Minimum=<0 0 0.5>, Maximum=<0 0 3>)
    )
    Velocity Speed Scale=&RandomRangeFloat(Minimum=250, Maximum=300)
   ) {enabled=false}
   &CameraOffset(Camera Offset Amount=&Multiply_Float(A=160, B=$User.FX_Scale))
   &SubUVAnimation(
    "Sprite Renderer "=DataInterface() {propertyValues {SpriteRenderer=&self:Snow_Puffs_0.NiagaraSpriteRendererProperties_0}}
    SubUV Animation Mode:ENiagara_SubUVLookupModeV2=::NewEnumerator2
    UseEndFrame=false
    UseStartFrame=false
   ) {enabled=false, module=SubUVAnimation001}
   &ShapeLocation(
    MaskSpawnGroup=false
    Non Uniform Scale=<1 0.8 0.8>
    Override Seed=false
    Sphere Radius=&Multiply_Float(A=200, B=$User.FX_Scale)
    UseOwnerScale=true
   ) {module=ShapeLocation001}
  ]
  particleUpdateScript [
   &ParticleState()
   &SpriteRotationRate(Rotation Rate=&RandomRangeFloat(Minimum=-80, Maximum=-160))
   &ScaleSpriteSize(Uniform Curve Sprite Scale=DataInterface() {propertyValues {Curve {keys [{interpMode=RCIM_Cubic, tangentMode=RCTM_User, time=0, value=0.6}, {interpMode=RCIM_Cubic, tangentMode=RCTM_User, time=0.7, value=1}]}}})
   &ScaleColor(
    Scale Alpha=&FloatFromCurve(
     FloatCurve=DataInterface() {propertyValues {Curve {keys [
         {interpMode=RCIM_Cubic, tangentMode=RCTM_User, time=0.0056, value=0.3647}
         {interpMode=RCIM_Cubic, tangentMode=RCTM_User, time=0.4, value=1}
         {interpMode=RCIM_Cubic, tangentMode=RCTM_User, time=1, value=0}
        ]}}}
     Scale Curve=&Multiply_Float(
      A=&FNTimeOfDayScalar(Night Scale=0.5, Morning Scale=0.2, Day Scale=0.3, Evening Scale=0.3)
      B=&ClampFloat(
       Float=&Abs_Float(Float:NiagaraFloat=&DotProduct(
         Vector A=&ReturnCameraVector(Camera Vectors=::NewEnumerator3, Coordinate Space=::Simulation)
         Vector B=$Particles.Velocity
        ))
       Max=0.5
      )
     )
    )
    Scale RGB=&MakeVectorFromLinearColorRGB(Color=&FNTimeOfDayColor(
      Night Color=<rgba 0.48 0.5895 0.8 1>
      Morning Color=<rgba 0.9347 0.9078 1 1>
      Day Color=<rgba 1 0.925 0.85 1>
      Evening Color=<rgba 1 0.9427 0.9857 1>
     ))
    ScaleA=true
    ScaleRGB=true
   )
   &SubUVAnimation(
    "Sprite Renderer "=DataInterface() {propertyValues {SpriteRenderer=&self:Snow_Puffs_0.NiagaraSpriteRendererProperties_0}}
    UseEndFrame=false
    UseStartFrame=false
   )
   &VortexForce(
    Origin Pull Amount:NiagaraFloat=&Multiply_Float(A=$User.FX_Scale, B=4000)
    UseOriginOffset=false
    Vortex Axis=&RandomRangeVector(Minimum=<1 -0.2 0>, Maximum=<1 0.2 0>)
    Vortex Axis Coordinate Space:ENiagaraCoordinateSpace=::Local
    Vortex Force Amount=&Multiply_Float(A=$User.FX_Scale, B=4000)
   )
   &Drag(Drag=4, Rotational Drag=2, Use Linear Drag=true, Use Rotational Drag=true)
   &SolveForcesAndVelocity(Clamp Velocity=false, Limit Acceleration=false)
   &RotateAroundPoint(
    Derive the Rotation Axis=false
    Radius:NiagaraFloat=&RandomRangeFloat(Minimum=200, Maximum=500)
    Rotation Axis:Vector3f=&RandomRangeVector(Minimum=<-1 -1 -1>, Maximum=<1 1 1>)
    Rotation Center:NiagaraPosition=$Engine.Owner.Position
    Rotation Rate=1
   ) {enabled=false}
   &DynamicMaterialParameters(
    Index 0 Param 1=2
    Index 0 Param 3=1
    Module.Index0Param0Write=true
    Module.Index0Param1Write=true
    Module.Index0Param2Write=true
    Module.Index0Param3Write=true
   )
  ]
  rendererClasses [&NiagaraSpriteRendererProperties]
  renderers [{rendererClass=&NiagaraSpriteRendererProperties}]
 }
 WindGusts {
  emitterUpdateScript [
   &EmitterState()
   &SpawnRate(
    Spawn Probability=&FloatFromCurve(
     FloatCurve=DataInterface() {propertyValues {Curve {keys [
         {interpMode=RCIM_Cubic, tangentMode=RCTM_User, time=0, value=0}
         {interpMode=RCIM_Cubic, tangentMode=RCTM_User, time=1, value=1}
         {arriveTangent=-0.3971, interpMode=RCIM_Cubic, leaveTangent=-0.3971, tangentMode=RCTM_User, time=10, value=0}
        ]}}}
     CurveIndex=$Engine.System.Age
    )
    SpawnRate=20
    Use Spawn Probability=false
   )
  ]
  particleSpawnScript [
   &PartitionParticles()
   /SetVariables(
    Particles.Particle_Height:NiagaraFloat=&RandomRangeFloat(Minimum=0, Maximum=1)
    Particles.VisibilityTag=$Particles.PartitionParticles.Partition
   ) {isSetParameters=true}
   /SetVariables(Particles.Position2:NiagaraPosition=&AddVectorToPosition(
     Position=<0 0 0>
     Vector=&MakeVector(
      X=0
      Y=0
      Z=&MergeHelper_FloatSwitch(
       0=&Multiply_Float(A=$User.FX_Scale, B=&Lerp_Float(A=-300, B=300, Alpha=$Particles.Particle_Height))
       1=0
       2=?
       3=?
       Index=$Particles.PartitionParticles.Partition
       Num=::NewEnumerator0
      )
     )
    )) {isSetParameters=true}
   &InitializeParticle(
    Color=&RandomRangeLinearColor(Minimum=<rgba 0.536 0.713 1 0.5>, Maximum=<rgba 0.6094 1.0957 1.5 0.5>)
    Lifetime Max=1.3
    Lifetime Min=0.7
    Lifetime Mode=::NewEnumerator1
    Mass Mode=::NewEnumerator3
    Mesh Scale=&Multiply_VectorByFloat(
     Vector=&MergeHelper_VectorSwitch(
      0=&Multiply_VectorByFloat(
       Vector=&RandomRangeVector(Minimum=<1 1 0.5>, Maximum=<1 1 1.5>)
       Float=&Multiply_Float(
        A=$User.FX_Scale
        B=&Lerp_Float(
         A=7.3
         B=8.7
         Alpha=&FloatFromCurve(
          FloatCurve=DataInterface() {propertyValues {Curve {keys [
              {arriveTangent=5.7133, interpMode=RCIM_Cubic, leaveTangent=5.7133, tangentMode=RCTM_User, time=0, value=0}
              {interpMode=RCIM_Cubic, tangentMode=RCTM_User, time=0.5, value=1}
              {arriveTangent=-6.0152, interpMode=RCIM_Cubic, leaveTangent=-6.0152, tangentMode=RCTM_User, time=1, value=0}
             ]}}}
          CurveIndex=$Particles.Particle_Height
          Scale Curve=&RandomRangeFloat(Minimum=0.9, Maximum=1)
         )
        )
       )
      )
      1=&RandomRangeVector(Minimum=<5 5 5>, Maximum=<15 15 7>)
      2=?
      3=?
      Index=$Particles.PartitionParticles.Partition
      Num=::NewEnumerator0
     )
     Float=&FloatFromCurve(
      FloatCurve=DataInterface() {propertyValues {Curve {defaultValue=1, keys [{tangentMode=RCTM_User, time=0, value=0.3}, {tangentMode=RCTM_User, time=0.7, value=1}]}}}
      CurveIndex=$Emitter.Age
     )
    )
    Mesh Scale Mode=::NewEnumerator4
    Sprite Rotation Mode=::NewEnumerator2
    UsePositionOffset=false
    Write Material Random=true
   )
   &SystemLocation(
    MaskSpawnGroup=false
    Offset=&MakeVector(
     X=0
     Y=0
     Z=&MergeHelper_FloatSwitch(
      0=&Multiply_Float(A=$User.FX_Scale, B=&Lerp_Float(A=-300, B=300, Alpha=$Particles.Particle_Height))
      1=0
      2=?
      3=?
      Index=$Particles.PartitionParticles.Partition
      Num=::NewEnumerator0
     )
    )
   )
   &InitialMeshOrientation(
    Mesh Orientation Mode=::NewEnumerator2
    Orientation Coordinate Space=::World
    Rotation=&MergeHelper_VectorSwitch(
     0=&RandomRangeVector(Minimum=<-0.04 -0.04 -1>, Maximum=<0 0 1>)
     1=&RandomRangeVector(Minimum=<0 0 0>, Maximum=<1 1 1>)
     2=?
     3=?
     Index=$Particles.PartitionParticles.Partition
     Num=::NewEnumerator0
    )
    Use Rotation Vector=true
   )
   &CameraOffset(Camera Offset Amount=15)
   &ShapeLocation(
    MaskSpawnGroup=false
    Override Seed=false
    Sphere Radius=&MergeHelper_FloatSwitch(0=0, 1=180, 2=?, 3=?, Index=$Particles.PartitionParticles.Partition, Num=::NewEnumerator0)
    UseOwnerScale=true
   )
  ]
  particleUpdateScript [
   &ParticleState()
   &ScaleMeshSize(Scale Factor=&MergeHelper_VectorSwitch(
     0=&Multiply_VectorByFloat(
      Vector=&MakeVector(
       X=1.5
       Y=1.3
       Z=&SelectFloatFromArray(
        Array Sampling Mode=::NewEnumerator0
        Direct Array Index=?
        Float Selection Array=DataInterface() {propertyValues {FloatData [1,-1]}}
        Random Seed=0
       )
      )
      Float=&Lerp_Float(
       A=0.8
       B=1.2
       Alpha=&FloatFromCurve(FloatCurve=DataInterface() {propertyValues {Curve {keys [
            {arriveTangent=2.1221, interpMode=RCIM_Cubic, leaveTangent=2.1221, tangentMode=RCTM_User, time=0, value=0}
            {interpMode=RCIM_Cubic, tangentMode=RCTM_User, time=0.86, value=1}
           ]}}})
      )
     )
     1=&Multiply_VectorByFloat(
      Vector=<1 1 1>
      Float=&FloatFromCurve(FloatCurve=DataInterface() {propertyValues {Curve {
          defaultValue=1
          keys [
           {arriveTangent=6.6205, interpMode=RCIM_Cubic, leaveTangent=6.6205, tangentMode=RCTM_User, time=0, value=0}
           {interpMode=RCIM_Cubic, tangentMode=RCTM_User, time=0.5, value=1}
           {arriveTangent=-6.754, tangentMode=RCTM_User, time=1, value=0}
          ]
         }}})
     )
     2=?
     3=?
     Index=$Particles.PartitionParticles.Partition
     Num=::NewEnumerator0
    ))
   &UpdateMeshOrientation(
    Rotation Rate=&RandomRangeFloat(Minimum=0.5, Maximum=1.1)
    Rotation Vector=&MergeHelper_VectorSwitch(0=<0 0 1>, 1=<0.5 0.5 1>, 2=?, 3=?, Index=$Particles.PartitionParticles.Partition, Num=::NewEnumerator0)
   )
   &ScaleColor(
    Scale Alpha=&FloatFromCurve(
     FloatCurve=DataInterface() {propertyValues {Curve {keys [
         {arriveTangent=6.2301, interpMode=RCIM_Cubic, leaveTangent=6.2301, tangentMode=RCTM_User, time=0, value=0}
         {arriveTangent=-6.912, interpMode=RCIM_Cubic, leaveTangent=-6.912, tangentMode=RCTM_User, time=0.1, value=3}
         {arriveTangent=0.4048, interpMode=RCIM_Cubic, leaveTangent=0.4048, tangentMode=RCTM_User, time=0.7, value=0}
        ]}}}
     Scale Curve=&RandomRangeFloat(Minimum=0.05, Maximum=0.1)
    )
    Scale RGB=&VectorFromFloat(Value=&Lerp_Float(
      A=1.2
      B=1.5
      Alpha=&FloatFromCurve(FloatCurve=DataInterface() {propertyValues {Curve {keys [
           {interpMode=RCIM_Cubic, tangentMode=RCTM_User, time=0, value=0}
           {interpMode=RCIM_Cubic, tangentMode=RCTM_User, time=0.1, value=1}
           {interpMode=RCIM_Cubic, tangentMode=RCTM_User, time=1, value=0}
          ]}}})
     ))
    ScaleA=true
    ScaleRGB=true
   )
   &SolveForcesAndVelocity(Clamp Velocity=false, Limit Acceleration=false)
   &DynamicMaterialParameters(
    Index 0 Param 1=&FloatFromCurve(FloatCurve=DataInterface() {propertyValues {Curve {keys [
         {arriveTangent=1.526, interpMode=RCIM_Cubic, leaveTangent=1.526, tangentMode=RCTM_User, time=0, value=0}
         {interpMode=RCIM_Cubic, tangentMode=RCTM_User, time=1, value=1}
        ]}}})
    Index 0 Param 3=1
    Module.Index0Param0Write=true
    Module.Index0Param1Write=true
    Module.Index0Param2Write=true
    Module.Index0Param3Write=true
   )
   &ScaleColor(
    Scale Alpha=&FNTimeOfDayScalar(Night Scale=0.25, Morning Scale=0.25, Day Scale=0.7, Evening Scale=5)
    Scale RGB=&MakeVectorFromLinearColorRGB(Color=&FNTimeOfDayColor(
      Night Color=<rgba 0.48 0.5895 0.8 1>
      Morning Color=<rgba 0.8 0.52 0.52 1>
      Day Color=<rgba 1 0.9 0.8 1>
      Evening Color=<rgba 0.8 0.48 0.5067 1>
     ))
    ScaleA=true
    ScaleRGB=true
   ) {module=ScaleColor001}
   &RotateAroundPoint(
    Radius:NiagaraFloat=&RandomRangeFloat(Minimum=145, Maximum=160)
    Rotation Center:NiagaraPosition=&RandomRangeVector(Minimum=<0 0 -50>, Maximum=<0 0 50>)
    Rotation Phase:NiagaraFloat=&RandomRangeFloat(Minimum=0, Maximum=1)
    Rotation Rate:NiagaraFloat=&RandomRangeFloat(Minimum=0.1, Maximum=0.9)
    X Axis:Vector3f=&RandomRangeVector(Minimum=<1 0 -0.2>, Maximum=<1 0 0.2>)
   )
  ]
  rendererClasses [&NiagaraMeshRendererProperties]
  renderers [
   {rendererClass=&NiagaraMeshRendererProperties, rendererIndex=0}
   {rendererClass=&NiagaraMeshRendererProperties, rendererIndex=1}
  ]
 }
}
summary {
 systemName=NS_IceGauntlets_BigBomb_Projectile
 emitters [
  {emitterName=Dome, rendererClasses [&NiagaraMeshRendererProperties, &NiagaraSpriteRendererProperties]}
  {emitterName=Snow_Puffs, rendererClasses [&NiagaraSpriteRendererProperties]}
  {
   emitterName=Ice_Shards
   simTarget=GPUComputeSim
   rendererClasses [&NiagaraSpriteRendererProperties, &NiagaraMeshRendererProperties]
  }
  {emitterName=WindGusts, rendererClasses [&NiagaraMeshRendererProperties]}
 ]
 userVariables [
  User.DEBUG {type=NiagaraBool}
  User.FX_DomeAngle {type=Vector3f, description="", defaultValue {struct=Vector3f, value=<0 0 0>}}
  User.FX_Scale {type=NiagaraFloat, defaultValue=0.5, description=""}
  User.Muzzle Pos {type=NiagaraPosition, description="", defaultValue {}}
  User.POS {type=Vector3f, description="", defaultValue {struct=Vector3f, value=<0 0 0>}}
 ]
}
