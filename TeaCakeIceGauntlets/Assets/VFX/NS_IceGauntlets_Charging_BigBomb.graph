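# Niagara effect for the Ice Gauntlets' big bomb charge, combining flash, snowflake, streak and snow puff emitters.
# System modules include skeletal mesh sampling; Snow_Puffs randomly spawns sprites at System.PL or System.PR, likely the two hand points.
# Puffs grow, fade, drag and animate through SubUV, with color and opacity adjusted by time of day and size scaled by User.FX_Scale.
# Only Snow_Puffs shows modules and a renderer; the other emitters list none.
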NiagaraSystem /TeaCakeIceGauntlets/Assets/VFX/NS_IceGauntlets_Charging_BigBomb
dependencies {
 usedDataInterfaces [
  &NiagaraDataInterfaceCurve
  &NiagaraDataInterfaceSpriteRendererInfo
  &NiagaraDataInterfaceSkeletalMesh
  &NiagaraDataInterfaceColorCurve
 ]
 usedDynamicInputs [
  &TransformPosition
  &MakeVector4
  &Multiply_Vector2D_ByFloat
  &Sine
  &ColorFromCurve
 ]
 usedModules [
  &self:SystemSpawnScript.SystemScriptSource.SystemScriptGraph.NiagaraNodeAssignment_0.SetVariables_#1
  &SystemState
  &SampleSkeletalMesh
  &self:SystemSpawnScript.SystemScriptSource.SystemScriptGraph.NiagaraNodeAssignment_3.SetVariables_#2
  &self:SystemSpawnScript.SystemScriptSource.SystemScriptGraph.NiagaraNodeAssignment_1.SetVariables_#3
  &self:SystemSpawnScript.SystemScriptSource.SystemScriptGraph.NiagaraNodeAssignment_4.SetVariables_#4
 ]
}
emitters {
 Empty001 {rendererClasses [], renderers []}
 Empty002 {rendererClasses [], renderers []}
 LW_Intro_Flash {rendererClasses [], renderers []}
 LW_Intro_Flash001 {rendererClasses [], renderers []}
 LW_SnowFlakes {rendererClasses [], renderers []}
 LW_SnowFlakes001 {rendererClasses [], renderers []}
 LW_Streaks_Fat {rendererClasses [], renderers []}
 LW_Streaks_Fat002 {rendererClasses [], renderers []}
 Snow_Puffs {
  emitterUpdateScript [
   &EmitterState(Life Cycle Mode=::NewEnumerator1, Loop Behavior=::NewEnumerator1, Loop Duration=1, UseLoopDelay=false)
   &SpawnRate(
    Spawn Probability=&FloatFromCurve(
     FloatCurve=DataInterface() {propertyValues {Curve {keys [
         {interpMode=RCIM_Cubic, tangentMode=RCTM_User, time=0, value=0}
         {interpMode=RCIM_Cubic, tangentMode=RCTM_User, time=1, value=1}
         {arriveTangent=-0.2753, interpMode=RCIM_Cubic, leaveTangent=-0.2753, tangentMode=RCTM_User, time=8.5, value=0}
        ]}}}
     CurveIndex=$Engine.System.Age
    )
    SpawnRate=25
    Use Spawn Probability=false
   )
  ]
  particleSpawnScript [
   &InitializeParticle(
    Color Maximum=<rgba 0.2221 0.43 1 0.6>
    Color Minimum=<rgba 0.2434 0.4509 0.7344 0.2>
    Color Mode=::NewEnumerator2
    Lifetime Max=0.6
    Lifetime Min=0.5
    Lifetime Mode=::NewEnumerator1
    Mass Mode=::NewEnumerator3
    Position Offset=&Multiply_VectorByFloat(
     Vector=&MergeHelper_VectorSwitch(
      0=Linked() {linkedVariable {name=System.PL, type=NiagaraPosition}}
      1=Linked() {linkedVariable {name=System.PR, type=NiagaraPosition}}
      2=?
      3=?
      Index=&RandomRangeInt(Minimum=0, Maximum=1)
      Num=::NewEnumerator0
     )
     Float=1
    )
    Position Offset Coordinate Space=::Local
    Sprite Rotation Mode=::NewEnumerator2
    Sprite Size Mode=::NewEnumerator3
    Uniform Sprite Size=&Multiply_Float(A=$User.FX_Scale, B=&RandomRangeFloat(Minimum=50, Maximum=60))
    UsePositionOffset=true
    Write Material Random=true
   )
   &CameraOffset(Camera Offset Amount=&Multiply_Float(A=10, B=$User.FX_Scale))
  ]
  particleUpdateScript [
   &ParticleState()
   &ScaleSpriteSize(Uniform Curve Sprite Scale=DataInterface() {propertyValues {Curve {keys [{interpMode=RCIM_Cubic, tangentMode=RCTM_User, time=0, value=0.6}, {interpMode=RCIM_Cubic, tangentMode=RCTM_User, time=0.7, value=1}]}}})
   &ScaleColor(
    Scale Alpha=&FloatFromCurve(
     FloatCurve=DataInterface() {propertyValues {Curve {keys [
         {arriveTangent=4.8598, interpMode=RCIM_Cubic, leaveTangent=4.8598, tangentMode=RCTM_User, time=0.0056, value=0}
         {interpMode=RCIM_Cubic, tangentMode=RCTM_User, time=0.4, value=1}
         {interpMode=RCIM_Cubic, tangentMode=RCTM_User, time=1, value=0}
        ]}}}
     Scale Curve=&FNTimeOfDayScalar(Night Scale=0.5, Morning Scale=0.2, Day Scale=0.3, Evening Scale=0.3)
    )
    Scale RGB=&MakeVectorFromLinearColorRGB(Color=&FNTimeOfDayColor(
      Night Color=<rgba 0.48 0.5895 0.8 1>
      Morning Color=<rgba 0.9347 0.9078 1 1>
      Day Color=<rgba 1 0.925 0.85 1>
      Evening Color=<rgba 1 0.9427 0.9857 1>
     ))
    ScaleA=true
    ScaleRGB=true
   )
   &SubUVAnimation(
    "Sprite Renderer "=DataInterface() {propertyValues {SpriteRenderer=&self:Snow_Puffs_0.NiagaraSpriteRendererProperties_0}}
    SubUV Lookup Index Scale=0.2
    UseEndFrame=false
    UseStartFrame=false
   )
   &Drag(Drag=4, Rotational Drag=2, Use Linear Drag=true, Use Rotational Drag=true)
   &SolveForcesAndVelocity(Clamp Velocity=false, Limit Acceleration=false)
   &DynamicMaterialParameters(
    Index 0 Param 1=2
    Index 0 Param 3=1
    Module.Index0Param0Write=true
    Module.Index0Param1Write=true
    Module.Index0Param2Write=true
    Module.Index0Param3Write=true
   )
  ]
  rendererClasses [&NiagaraSpriteRendererProperties]
  renderers [{rendererClass=&NiagaraSpriteRendererProperties}]
 }
}
summary {
 systemName=NS_IceGauntlets_Charging_BigBomb
 emitters [
  {emitterName=LW_Intro_Flash001, rendererClasses []}
  {emitterName=LW_SnowFlakes, rendererClasses []}
  {emitterName=Empty001, rendererClasses []}
  {emitterName=Empty002, rendererClasses []}
  {emitterName=LW_Streaks_Fat002, rendererClasses []}
  {emitterName=LW_Streaks_Fat, rendererClasses []}
  {emitterName=LW_Intro_Flash, rendererClasses []}
  {emitterName=LW_SnowFlakes001, rendererClasses []}
  {emitterName=Snow_Puffs, rendererClasses [&NiagaraSpriteRendererProperties]}
 ]
 userVariables [User.FX_Scale {type=NiagaraFloat, defaultValue=1, description=""}]
}
