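# Niagara system for the Ice Gauntlets equip effect, with two LW_Intro_Flash emitters and two generically named Empty emitters.
# Its modules include SampleSkeletalMesh and several Set Variables steps, suggesting placement on a character or gauntlet mesh.
# Dynamic inputs cover position transforms, float and color curves, sine and random ranges, and a User.FX_Scale parameter is exposed.
# No renderers are listed for any emitter.
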NiagaraSystem /TeaCakeIceGauntlets/Assets/VFX/NS_IceGauntlets_Equip
dependencies {
 usedDataInterfaces [&NiagaraDataInterfaceSkeletalMesh, &NiagaraDataInterfaceCurve, &NiagaraDataInterfaceColorCurve]
 usedDynamicInputs [
  &TransformPosition
  &MakeVector4
  &FloatFromCurve
  &Multiply_Vector2D_ByFloat
  &Sine
  &RandomRangeFloat
  &ColorFromCurve
 ]
 usedModules [
  &self:SystemSpawnScript.SystemScriptSource.SystemScriptGraph.NiagaraNodeAssignment_0.SetVariables_#1
  &SystemState
  &SampleSkeletalMesh
  &self:SystemSpawnScript.SystemScriptSource.SystemScriptGraph.NiagaraNodeAssignment_3.SetVariables_#2
  &self:SystemSpawnScript.SystemScriptSource.SystemScriptGraph.NiagaraNodeAssignment_1.SetVariables_#3
  &self:SystemSpawnScript.SystemScriptSource.SystemScriptGraph.NiagaraNodeAssignment_4.SetVariables_#4
 ]
}
emitters {
 Empty001 {rendererClasses [], renderers []}
 Empty002 {rendererClasses [], renderers []}
 LW_Intro_Flash {rendererClasses [], renderers []}
 LW_Intro_Flash001 {rendererClasses [], renderers []}
}
summary {
 systemName=NS_IceGauntlets_Equip
 emitters [
  {emitterName=LW_Intro_Flash001, rendererClasses []}
  {emitterName=Empty001, rendererClasses []}
  {emitterName=Empty002, rendererClasses []}
  {emitterName=LW_Intro_Flash, rendererClasses []}
 ]
 userVariables [User.FX_Scale {type=NiagaraFloat, defaultValue=1, description=""}]
}
