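# Niagara system for the Ice Gauntlets BigBomb muzzle effect.
# Ten emitters cover an intro flash, smoke, snowflakes, three fat streak layers, a Lit emitter and three generically named Empty emitters.
# System-level modules sample a skeletal mesh and set variables, with dynamic inputs for position transforms, camera fade, sine and random ranges.
# It exposes no user parameters, and no renderers are listed for any emitter.
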NiagaraSystem /TeaCakeIceGauntlets/Assets/VFX/NS_IceGauntlets_Muzzle_BigBomb
dependencies {
 usedDataInterfaces [&NiagaraDataInterfaceSkeletalMesh, &NiagaraDataInterfaceCamera]
 usedDynamicInputs [
  &MakeVector4
  &Multiply_Float
  &Add_Float
  &OneMinusFloat
  /Niagara/DynamicInputs/Camera/CameraFade
  &Multiply_LinearColorByFloat
  &Multiply_VectorByFloat
  &Lerp_Float
  &TransformPosition
  &Multiply_Vector2D_ByFloat
  &Sine
  &RandomRangeFloat
 ]
 usedModules [
  &self:SystemSpawnScript.SystemScriptSource.SystemScriptGraph.NiagaraNodeAssignment_0.SetVariables_#1
  &self:SystemSpawnScript.SystemScriptSource.SystemScriptGraph.NiagaraNodeAssignment_5.SetVariables_#2
  &SystemState
  &SampleSkeletalMesh
  &self:SystemSpawnScript.SystemScriptSource.SystemScriptGraph.NiagaraNodeAssignment_3.SetVariables_#3
  &self:SystemSpawnScript.SystemScriptSource.SystemScriptGraph.NiagaraNodeAssignment_4.SetVariables_#4
 ]
}
emitters {
 Empty {rendererClasses [], renderers []}
 Empty001 {rendererClasses [], renderers []}
 Empty002 {rendererClasses [], renderers []}
 LW_Intro_Flash001 {rendererClasses [], renderers []}
 LW_Smoke {rendererClasses [], renderers []}
 LW_SnowFlakes {rendererClasses [], renderers []}
 LW_Streaks_Fat {rendererClasses [], renderers []}
 LW_Streaks_Fat001 {rendererClasses [], renderers []}
 LW_Streaks_Fat002 {rendererClasses [], renderers []}
 Lit {rendererClasses [], renderers []}
}
summary {
 systemName=NS_IceGauntlets_Muzzle_BigBomb
 emitters [
  {emitterName=Empty, rendererClasses []}
  {emitterName=Lit, rendererClasses []}
  {emitterName=LW_SnowFlakes, rendererClasses []}
  {emitterName=Empty001, rendererClasses []}
  {emitterName=Empty002, rendererClasses []}
  {emitterName=LW_Smoke, rendererClasses []}
  {emitterName=LW_Streaks_Fat001, rendererClasses []}
  {emitterName=LW_Streaks_Fat002, rendererClasses []}
  {emitterName=LW_Streaks_Fat, rendererClasses []}
  {emitterName=LW_Intro_Flash001, rendererClasses []}
 ]
 userVariables []
}
