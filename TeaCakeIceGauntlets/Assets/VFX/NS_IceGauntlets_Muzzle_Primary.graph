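# Niagara muzzle effect system for the primary fire of the Ice Gauntlets.
# Emitter names indicate two intro flashes, snowflakes and fat streaks, plus one emitter named Empty; the LW prefix suggests lightweight emitters.
# User parameters expose an aim vector, defaulting to +Z, and a muzzle FX position.
# Only the SystemState module is listed, with no renderers or emitter internals shown.
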NiagaraSystem /TeaCakeIceGauntlets/Assets/VFX/NS_IceGauntlets_Muzzle_Primary
dependencies {usedDataInterfaces [], usedDynamicInputs [], usedModules [&SystemState]}
emitters {
 Empty {rendererClasses [], renderers []}
 LW_Intro_Flash {rendererClasses [], renderers []}
 LW_Intro_Flash001 {rendererClasses [], renderers []}
 LW_SnowFlakes {rendererClasses [], renderers []}
 LW_Streaks_Fat {rendererClasses [], renderers []}
}
summary {
 systemName=NS_IceGauntlets_Muzzle_Primary
 emitters [
  {emitterName=Empty, rendererClasses []}
  {emitterName=LW_Streaks_Fat, rendererClasses []}
  {emitterName=LW_Intro_Flash001, rendererClasses []}
  {emitterName=LW_SnowFlakes, rendererClasses []}
  {emitterName=LW_Intro_Flash, rendererClasses []}
 ]
 userVariables [
  User.AimVec {type=Vector3f, description="", defaultValue {struct=Vector3f, value=<0 0 1>}}
  User.Muzzle FX Pos {type=NiagaraPosition, description="", defaultValue {}}
 ]
}
