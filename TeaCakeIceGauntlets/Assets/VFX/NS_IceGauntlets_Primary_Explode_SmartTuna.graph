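# Niagara system for the Ice Gauntlets primary explosion effect, named as the SmartTuna variant.
# Ten emitters cover an intro flash, sphere glow, light, spikes, fat streaks, chunks, debris, snowflakes, ground smoke and a decal.
# System scripts use SystemState and a spawn-time variable assignment, with camera-fade and Vector2D-scaling dynamic inputs among the dependencies.
# User parameters expose a global scale plus water state, depth, plane and wave values.
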NiagaraSystem /TeaCakeIceGauntlets/Assets/VFX/NS_IceGauntlets_Primary_Explode_SmartTuna
dependencies {
 usedDataInterfaces [&NiagaraDataInterfaceCamera]
 usedDynamicInputs [&Multiply_Vector2D_ByFloat, /Niagara/DynamicInputs/Camera/CameraFade]
 usedModules [
  &self:SystemSpawnScript.SystemScriptSource.SystemScriptGraph.NiagaraNodeAssignment_1.SetVariables_#1
  &SystemState
 ]
}
emitters {
 Decal {rendererClasses [], renderers []}
 LW_Debris {rendererClasses [], renderers []}
 LW_Intro_Flash {rendererClasses [], renderers []}
 LW_Smoke_Ground {rendererClasses [], renderers []}
 LW_SnowFlakes {rendererClasses [], renderers []}
 LW_Spike {rendererClasses [], renderers []}
 LW_Streaks_Fat {rendererClasses [], renderers []}
 Light001 {rendererClasses [], renderers []}
 SphereGlow {rendererClasses [], renderers []}
 chunks {rendererClasses [], renderers []}
}
summary {
 systemName=NS_IceGauntlets_Primary_Explode_SmartTuna
 emitters [
  {emitterName=LW_Streaks_Fat, rendererClasses []}
  {emitterName=LW_Spike, rendererClasses []}
  {emitterName=LW_Smoke_Ground, rendererClasses []}
  {emitterName=Light001, rendererClasses []}
  {emitterName=chunks, rendererClasses []}
  {emitterName=LW_Debris, rendererClasses []}
  {emitterName=LW_SnowFlakes, rendererClasses []}
  {emitterName=LW_Intro_Flash, rendererClasses []}
  {emitterName=SphereGlow, rendererClasses []}
  {emitterName=Decal, rendererClasses []}
 ]
 userVariables [
  User.bIsInWater {type=NiagaraBool}
  User.Global Scale {type=NiagaraFloat, defaultValue=1, description=""}
  User.IsInWater {type=NiagaraFloat, defaultValue=0, description=""}
  User.WaterBodyIndex {type=NiagaraFloat, defaultValue=0, description=""}
  User.WaterDepth {type=NiagaraFloat, defaultValue=0, description=""}
  User.WaterPlaneLocation {type=Vector3f, description="", defaultValue {struct=Vector3f, value=<0 0 0>}}
  User.WaterPlaneNormal {type=Vector3f, description="", defaultValue {struct=Vector3f, value=<0 0 0>}}
  User.WaveAttenuationFactor {type=NiagaraFloat, defaultValue=0, description=""}
  User.WaveTime {type=NiagaraFloat, defaultValue=0, description=""}
 ]
}
