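# Niagara effect for the Ice Gauntlets primary projectile: GPU ice shards and snowflakes, snow puffs, spinning wind-gust meshes and a ribbon trail.
# Shard and trail spawn positions blend from User.Muzzle Pos to the simulation position over the first 0.2 seconds.
# A random flag splits shard particles between snowflake and mesh looks, and meshes shrink away near the camera.
# Time of day tints shards, puffs and gusts, while FX_Scale sizes them.
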NiagaraSystem /TeaCakeIceGauntlets/Assets/VFX/NS_IceGauntlets_Primary_Projectile
dependencies {
 usedDataInterfaces [
  &NiagaraDataInterfacePlatformSet
  &NiagaraDataInterfaceCurve
  &NiagaraDataInterfaceCamera
  &NiagaraDataInterfaceSpriteRendererInfo
  &NiagaraDataInterfaceMeshRendererInfo
  &NiagaraDataInterfaceArrayFloat
 ]
 usedDynamicInputs [/Game/Effects/Niagara/Modules/OrientMestToVectorFromCustomRot]
 usedModules [
  &self:SystemSpawnScript.SystemScriptSource.SystemScriptGraph.NiagaraNodeAssignment_0.SetVariables_#1
  &self:SystemSpawnScript.SystemScriptSource.SystemScriptGraph.NiagaraNodeAssignment_1.SetVariables_#2
  &SystemState
 ]
}
emitters {
 Empty {rendererClasses [], renderers []}
 Ice_Shards {
  simTarget=GPUComputeSim
  emitterUpdateScript [
   &EmitterState()
   &SpawnRate(
    Spawn Probability=&FloatFromCurve(
     FloatCurve=DataInterface() {propertyValues {Curve {keys [
         {interpMode=RCIM_Cubic, tangentMode=RCTM_User, time=0.1, value=0}
         {interpMode=RCIM_Cubic, tangentMode=RCTM_User, time=1, value=1}
         {arriveTangent=0.0061, interpMode=RCIM_Cubic, leaveTangent=-0.1111, tangentMode=RCTM_Break, time=9, value=1}
         <at 10.6 0 RCIM_Cubic>
        ]}}}
     CurveIndex=$Emitter.Age
    )
    SpawnRate=120
    Use Spawn Probability=false
   )
   &SpawnBurst_Instantaneous(Use Spawn Probability=false, UseLoopCountLimit=false)
  ]
  particleSpawnScript [
   /SetVariables(
    Particles.Particle_Depth:NiagaraFloat=&FloatFromCurve(
     FloatCurve=DataInterface() {propertyValues {Curve {keys [
         {arriveTangent=1.2753, interpMode=RCIM_Cubic, leaveTangent=1.2753, tangentMode=RCTM_User, time=0, value=0}
         {arriveTangent=0.204, interpMode=RCIM_Cubic, leaveTangent=0.204, tangentMode=RCTM_User, time=0.9, value=1}
        ]}}}
     CurveIndex=&RandomRangeFloat(Minimum=0, Maximum=1)
    )
    Particles.Particle_Height:NiagaraFloat=&FloatFromCurve(
     FloatCurve=DataInterface() {propertyValues {Curve {keys [
         {arriveTangent=1.9477, interpMode=RCIM_Cubic, leaveTangent=1.9477, tangentMode=RCTM_User, time=0, value=0}
         {arriveTangent=0.1423, interpMode=RCIM_Cubic, leaveTangent=0.1423, tangentMode=RCTM_User, time=0.5, value=0.5}
         {arriveTangent=1.4427, interpMode=RCIM_Cubic, leaveTangent=1.4427, tangentMode=RCTM_User, time=1, value=1}
        ]}}}
     CurveIndex=&RandomRangeFloat(Minimum=0, Maximum=1)
    )
    Particles.Particle_Scale:NiagaraFloat=&RandomRangeFloat(Minimum=0, Maximum=1)
   ) {isSetParameters=true}
   &InitializeParticle(
    Color Maximum=<rgba 0.9 0.9833 1 1>
    Color Minimum=<rgba 1.2 1.7333 2 1>
    Color Mode=::NewEnumerator2
    Lifetime Max=1
    Lifetime Min=0.7
    Lifetime Mode=::NewEnumerator1
    Mass Mode=::NewEnumerator3
    Mesh Renderer Array Visibility Mode=::NewEnumerator3
    Mesh Renderer Info=DataInterface() {propertyValues {MeshRenderer=&self:Ice_Shards_1.NiagaraMeshRendererProperties_0}}
    Mesh Scale Mode=::NewEnumerator3
    Mesh Uniform Scale=&Multiply_Float(
     A=&Multiply_Float(A=$User.FX_Scale, B=&Lerp_Float(A=3, B=7, Alpha=$Particles.Particle_Scale))
     B=&FloatFromCurve(
      FloatCurve=DataInterface() {propertyValues {Curve {
         defaultValue=1
         keys [
          {arriveTangent=4.1785, interpMode=RCIM_Cubic, leaveTangent=4.1785, tangentMode=RCTM_User, time=0, value=0}
          {arriveTangent=5, interpMode=RCIM_Cubic, leaveTangent=5, time=0.0846, value=0.65}
          {arriveTangent=0.0224, interpMode=RCIM_Cubic, leaveTangent=0.0224, tangentMode=RCTM_User, time=0.2, value=1}
         ]
        }}}
      CurveIndex=$Engine.System.Age
     )
    )
    Position=&Lerp_Vector(
     A=$User.Muzzle Pos
     B=/Niagara/DynamicInputs/Helpers/SimulationPosition()
     Alpha=&FloatFromCurve(
      FloatCurve=DataInterface() {propertyValues {Curve {keys [
          {arriveTangent=9.109, interpMode=RCIM_Cubic, leaveTangent=9.109, tangentMode=RCTM_User, time=0.01, value=0}
          {interpMode=RCIM_Cubic, leaveTangent=2.5007, time=0.2, value=1}
         ]}}}
      CurveIndex=$Engine.System.Age
     )
    )
    Position Mode=::NewEnumerator1
    Sprite Rotation Mode=::NewEnumerator2
    Sprite Size Mode=::NewEnumerator3
    Uniform Sprite Size=&Multiply_Float(
     A=&Multiply_Float(A=$User.FX_Scale, B=&Lerp_Float(A=9, B=16, Alpha=$Particles.Particle_Scale))
     B=&FloatFromCurve(
      FloatCurve=DataInterface() {propertyValues {Curve {
         defaultValue=1
         keys [
          {arriveTangent=4.1785, interpMode=RCIM_Cubic, leaveTangent=4.1785, tangentMode=RCTM_User, time=0, value=0}
          {arriveTangent=5, interpMode=RCIM_Cubic, leaveTangent=5, time=0.0789, value=0.7}
          {arriveTangent=0.0224, interpMode=RCIM_Cubic, leaveTangent=0.0224, tangentMode=RCTM_User, time=0.2, value=1}
         ]
        }}}
      CurveIndex=$Engine.System.Age
     )
    )
    UsePositionOffset=false
    Write Material Random=true
   )
   &ShapeLocation(
    MaskSpawnGroup=false
    Override Seed=false
    Sphere Radius=&Multiply_Float(A=20, B=$User.FX_Scale)
    UseOwnerScale=true
   )
   &AddVelocity(
    Module.Use Falloff Distance=false
    Override Seed=false
    Velocity Mode=::NewEnumerator1
    Velocity Origin=$StackContext.ShapeLocation.ShapePosition
    Velocity Speed=1111
   )
   &CameraOffset(Camera Offset Amount=6)
   &InitialMeshOrientation(Use Rotation Vector=false)
   /SetVariables(Particles.RotationDirection:Vector3f=&Lerp_Vector(
     A=&RandomRangeVector(Minimum=<-0.6 -0.6 1>, Maximum=<0.8 0.8 1>)
     B=&RandomRangeVector(Minimum=<-0.2 -0.2 1>, Maximum=<0.3 0.3 1>)
     Alpha=&FloatFromCurve(
      FloatCurve=DataInterface() {propertyValues {Curve {keys [
          {interpMode=RCIM_Cubic, tangentMode=RCTM_User, time=0, value=0}
          {arriveTangent=0.5342, interpMode=RCIM_Cubic, leaveTangent=0.5342, tangentMode=RCTM_User, time=1, value=1}
         ]}}}
      CurveIndex=$Particles.Particle_Depth
     )
    )) {isSetParameters=true}
   /SetVariables(Particles.isSnowFlake:NiagaraBool=&RandomBool()) {isSetParameters=true}
   /SetVariables(Particles.VisibilityTag=&MakeCustomIntFromBool(Bool=$Particles.isSnowFlake, True INT=0, False INT=1)) {isSetParameters=true}
   &DynamicMaterialParameters(
    Index 0 Param 1=&RandomRangeFloat(Minimum=1, Maximum=2)
    Module.Index0Param0Write=true
    Module.Index0Param1Write=false
    Module.Index0Param2Write=false
    Module.Index0Param3Write=false
   )
  ]
  particleUpdateScript [
   /SetVariables(Particles.ColorSnowFlake:LinearColor=&MakeLinearColorFromVectorAndFloat(
     "Vector (RGB)"=&MakeVectorFromLinearColorRGB(Color=&MultiplyLinearColor(
       Linear Color=&RandomRangeLinearColor(Minimum=<rgba 0.8 0.9333 1 1>, Maximum=<rgba 0.6 0.7333 1 1>)
       Scale Factor=&MakeVector4_FromLinearColor(LinearColor=$System.ToDColor)
      ))
     "Float (Alpha)"=&MakeCustomFloatFromBool(
      Bool=$Particles.isSnowFlake
      True Float=&FloatFromCurve(
       FloatCurve=DataInterface() {propertyValues {Curve {keys [
           {interpMode=RCIM_Cubic, tangentMode=RCTM_User, time=0, value=0}
           {interpMode=RCIM_Cubic, tangentMode=RCTM_User, time=0.15, value=1}
           {arriveTangent=-0.0592, interpMode=RCIM_Cubic, leaveTangent=-0.0592, tangentMode=RCTM_User, time=0.7, value=1}
           {interpMode=RCIM_Cubic, tangentMode=RCTM_User, time=1, value=0}
          ]}}}
       Scale Curve=&RandomRangeFloat(Minimum=0.2, Maximum=0.7)
      )
      False Float=0
     )
    )) {isSetParameters=true}
   &ParticleState()
   &SpriteRotationRate(Rotation Rate=&Lerp_Float(A=100, B=300, Alpha=$Particles.Particle_Scale))
   &UpdateMeshOrientation(Rotation Rate=2, Rotation Vector=&RandomRangeVector(Minimum=<-1 -1 -1>, Maximum=<1 1 1>))
   &ScaleMeshSize(Scale Factor=&VectorFromFloat(Value=&FloatFromCurve(FloatCurve=DataInterface() {propertyValues {Curve {keys [
          {interpMode=RCIM_Cubic, tangentMode=RCTM_User, time=0, value=0}
          {interpMode=RCIM_Cubic, tangentMode=RCTM_User, time=0.1, value=1}
          {arriveTangent=-3.406, interpMode=RCIM_Cubic, leaveTangent=-3.406, tangentMode=RCTM_User, time=1, value=0}
         ]}}})))
   &ScaleSpriteSize(Uniform Curve Sprite Scale=DataInterface() {propertyValues {Curve {keys [{interpMode=RCIM_Cubic, tangentMode=RCTM_User, time=0, value=0.6}, {interpMode=RCIM_Cubic, tangentMode=RCTM_User, time=0.2, value=1}]}}})
   &GravityForce(Gravity=&RandomRangeVector(Minimum=<0 0 -250>, Maximum=<0 0 -800>))
   &CurlNoiseForce(Bias Noise=false, Noise Strength=800, Pan Noise=false)
   &Drag(Drag=11, Use Linear Drag=true, Use Rotational Drag=true)
   &SolveForcesAndVelocity(Clamp Velocity=false, Limit Acceleration=false)
   &SpriteFacingAndAlignment(
    Sprite Facing=&Lerp_Vector(
     A=&RandomRangeVector(Minimum=<-1 -1 -1>, Maximum=<1 1 1>)
     B=&ReturnCameraVector(Camera Vectors=::NewEnumerator1, Coordinate Space=::Simulation)
     Alpha=&FloatFromCurve(
      FloatCurve=DataInterface() {propertyValues {Curve {keys [
          {arriveTangent=2.0555, interpMode=RCIM_Cubic, leaveTangent=2.0555, tangentMode=RCTM_User, time=0, value=0}
          {interpMode=RCIM_Cubic, tangentMode=RCTM_User, time=1, value=1}
         ]}}}
      CurveIndex=&RandomRangeFloat(Minimum=0, Maximum=1)
     )
    )
    WriteAlignment=false
    WriteFacing=true
   )
   &ScaleMeshSize(Scale Factor=&VectorFromFloat(Value=&FloatFromCurve(
      FloatCurve=DataInterface() {propertyValues {Curve {keys [{interpMode=RCIM_Cubic, tangentMode=RCTM_User, time=280, value=0}, {interpMode=RCIM_Cubic, tangentMode=RCTM_User, time=300, value=1}]}}}
      CurveIndex=&Abs_Float(Float:NiagaraFloat=&DistanceBetweenPositions(
        A=$Particles.Position
        B=&ReturnCameraPosition(Coordinate Space:ENiagaraCoordinateSpace=::Simulation, Camera Query:NiagaraDataInterfaceCamera=DataInterface())
       ))
     ))) {module=ScaleMeshSize001}
   &ScaleColor(
    Scale Alpha=&FloatFromCurve(
     FloatCurve=DataInterface() {propertyValues {Curve {
        defaultValue=1
        keys [
         {arriveTangent=4.1785, interpMode=RCIM_Cubic, leaveTangent=4.1785, tangentMode=RCTM_User, time=0, value=0}
         {arriveTangent=5, interpMode=RCIM_Cubic, leaveTangent=5, time=0.1548, value=0.525}
         {arriveTangent=0.0224, interpMode=RCIM_Cubic, leaveTangent=0.0224, tangentMode=RCTM_User, time=0.2, value=1}
        ]
       }}}
     CurveIndex=$Engine.System.Age
    )
    Scale RGB=&MakeVectorFromLinearColorRGB(Color=$System.ToDColor)
    ScaleA=true
    ScaleRGB=true
   )
  ]
  rendererClasses [&NiagaraSpriteRendererProperties, &NiagaraMeshRendererProperties]
  renderers [
   {rendererClass=&NiagaraSpriteRendererProperties}
   {rendererClass=&NiagaraMeshRendererProperties, rendererIndex=1}
  ]
 }
 LW_Head {rendererClasses [], renderers []}
 Snow_Puffs {
  emitterUpdateScript [
   &EmitterState()
   &SpawnRate(SpawnRate=60, Use Spawn Probability=false)
   &SpawnBurst_Instantaneous(Use Spawn Probability=false, UseLoopCountLimit=false)
   &SpawnPerUnit(Module.Use Max Movement Threshold=true, Module.Use Movement Tolerance=false, Spawn Spacing=50, Use Spawn Probability=false) {enabled=false}
  ]
  particleSpawnScript [
   &InitializeParticle(
    Color Maximum=<rgba 0.3958 0.5972 1 0.6>
    Color Minimum=<rgba 0.3865 0.4693 0.5521 0.2>
    Color Mode=::NewEnumerator2
    Lifetime Max=&FN_PickFloatDependOnPlatform(
     TargetPlatform=DataInterface() {propertyValues {Platforms {qualityLevelMask=-4, deviceProfileStates [{profileName=Switch2, qualityLevelMask=4294967295, setQualityLevelMask=4294967295}]}}}
     TargetPlatformActive=0.57
     TargetPlatformNoActive=0.35
    )
    Lifetime Min=0.5
    Lifetime Mode=::NewEnumerator1
    Mass Max=1
    Mass Mode=::NewEnumerator3
    Position Offset=&Multiply_VectorByFloat(Vector=<0 0 0>, Float=$User.FX_Scale)
    Sprite Rotation Mode=::NewEnumerator2
    Sprite Size Mode=::NewEnumerator3
    Sprite UV Mode=::NewEnumerator3
    Uniform Sprite Size=&Multiply_Float(A=$User.FX_Scale, B=&RandomRangeFloat(Minimum=400, Maximum=480))
    UsePositionOffset=true
    Write Material Random=true
   )
   &CameraOffset(Camera Offset Amount=&Multiply_Float(A=160, B=$User.FX_Scale))
   &ShapeLocation(
    MaskSpawnGroup=false
    Override Seed=false
    Sphere Radius=&Multiply_Float(A=20, B=$User.FX_Scale)
    UseOwnerScale=true
   ) {module=ShapeLocation001}
   &InheritVelocity(Inherited Velocity Speed Limit=0.1, Speed Limit=true, UseSpeedThreshold=false)
   &AddVelocity(Velocity=<1 0 0>) {enabled=false}
  ]
  particleUpdateScript [
   &ParticleState()
   &ScaleSpriteSize(Uniform Curve Sprite Scale=DataInterface() {propertyValues {Curve {keys [{interpMode=RCIM_Cubic, tangentMode=RCTM_User, time=0, value=0.6}, {interpMode=RCIM_Cubic, tangentMode=RCTM_User, time=0.7, value=1}]}}})
   &ScaleColor(
    Scale Alpha=&Multiply_Float(
     A=&FloatFromCurve(
      FloatCurve=DataInterface() {propertyValues {Curve {keys [
          {interpMode=RCIM_Cubic, tangentMode=RCTM_User, time=0.0107, value=1}
          {interpMode=RCIM_Cubic, tangentMode=RCTM_User, time=0.629, value=0.8824}
          {interpMode=RCIM_Cubic, tangentMode=RCTM_User, time=1, value=0}
         ]}}}
      Scale Curve=&FNTimeOfDayScalar(Night Scale=0.5, Morning Scale=0.2, Day Scale=0.3, Evening Scale=0.3)
     )
     B=&ClampFloat(
      Float=&Abs_Float(Float:NiagaraFloat=&DotProduct(
        Vector A=&ReturnCameraVector(Camera Vectors=::NewEnumerator3, Coordinate Space=::Simulation)
        Vector B=$Particles.Velocity
       ))
      Min=0.2
      Max=1.1
     )
    )
    Scale RGB=&MakeVectorFromLinearColorRGB(Color=&FNTimeOfDayColor(
      Night Color=<rgba 0.3292 0.4903 0.8 1>
      Morning Color=<rgba 0.9347 0.9078 1 1>
      Day Color=<rgba 0.6889 0.7667 1 1>
      Evening Color=<rgba 1 0.9106 0.9502 1>
     ))
    ScaleA=true
    ScaleRGB=true
   )
   &SubUVAnimation(
    "Sprite Renderer "=DataInterface() {propertyValues {SpriteRenderer=&self:Snow_Puffs_0.NiagaraSpriteRendererProperties_0}}
    UseEndFrame=false
    UseStartFrame=false
   )
   &Drag(Drag=4, Rotational Drag=2, Use Linear Drag=true, Use Rotational Drag=true)
   &SolveForcesAndVelocity(Clamp Velocity=false, Limit Acceleration=false)
   &DynamicMaterialParameters(
    Index 0 Param 1=2
    Index 0 Param 3=1
    Module.Index0Param0Write=true
    Module.Index0Param1Write=true
    Module.Index0Param2Write=true
    Module.Index0Param3Write=true
   )
  ]
  rendererClasses [&NiagaraSpriteRendererProperties]
  renderers [{rendererClass=&NiagaraSpriteRendererProperties}]
 }
 Trail {
  emitterUpdateScript [
   &EmitterState()
   &SpawnPerUnit(
    Module.Use Max Movement Threshold=false
    Module.Use Movement Tolerance=false
    Spawn Spacing:NiagaraFloat=&FN_PickFloatDependOnPlatform(TargetPlatformActive=100, TargetPlatformNoActive=200)
    Use Spawn Probability=false
   )
   &SpawnBurst_Instantaneous(Use Spawn Probability=false, UseLoopCountLimit=true) {enabled=false}
  ]
  particleSpawnScript [&InitializeRibbon(
    Color=<rgba 0.4143 0.9318 4 0.2>
    Lifetime=0.4
    Position:NiagaraPosition=&Lerp_Vector(
     A=$User.Muzzle Pos
     B=&SimulationPosition()
     Alpha=&FloatFromCurve(
      FloatCurve=DataInterface() {propertyValues {Curve {keys [
          {arriveTangent=9.109, interpMode=RCIM_Cubic, leaveTangent=9.109, tangentMode=RCTM_User, time=0.01, value=0}
          {interpMode=RCIM_Cubic, leaveTangent=2.5007, time=0.2, value=1}
         ]}}}
      CurveIndex=$Engine.System.Age
     )
    )
    Position Mode:ENiagara_PositionInitializationMode=::NewEnumerator1
    Ribbon Width=75
    UsePositionOffset=false
    Write Material Random=true
   )]
  particleUpdateScript [
   &ParticleState()
   &ScaleRibbonWidth(Ribbon Width Scale=&FloatFromCurve(FloatCurve=DataInterface() {propertyValues {Curve {keys [<at -0.0059 0.8693>,<at 0.1 0.9>,<at 1 0>]}}}))
   &ScaleColor(
    Scale RGB=&VectorFromFloat(Value=&FloatFromCurve(FloatCurve=DataInterface() {propertyValues {Curve {keys [
          {interpMode=RCIM_Cubic, tangentMode=RCTM_User, time=0.0028, value=0.9904}
          {arriveTangent=-0.5791, interpMode=RCIM_Cubic, leaveTangent=-0.5791, time=0.0144, value=1.0026}
          {interpMode=RCIM_Cubic, tangentMode=RCTM_User, time=0.9853, value=0.4214}
         ]}}}))
    ScaleA=true
    ScaleRGB=true
   )
   &SolveForcesAndVelocity(Clamp Velocity=false, Limit Acceleration=false)
  ]
  rendererClasses [&NiagaraRibbonRendererProperties]
  renderers [{rendererClass=&NiagaraRibbonRendererProperties}]
 }
 WindGusts {
  emitterUpdateScript [
   &EmitterState(Inactive Response=::NewEnumerator1, Life Cycle Mode=::NewEnumerator1, Loop Duration=2, UseLoopDelay=false)
   &SpawnBurst_Instantaneous(Use Spawn Probability=false, UseLoopCountLimit=false) {enabled=false}
   &SpawnRate(SpawnRate=14, Use Spawn Probability=false)
  ]
  particleSpawnScript [
   /SetVariables(Particles.Particle_Height:NiagaraFloat=&RandomRangeFloat(Minimum=0, Maximum=1)) {isSetParameters=true}
   &InitializeParticle(
    Color=&MultiplyLinearColor(
     Linear Color=&RandomRangeLinearColor(Minimum=<rgba 0.7 0.875 1 1>, Maximum=<rgba 0.5856 0.9425 1 0.5>)
     Scale Factor=&MakeVector4(
      X=1
      Y=1
      Z=1
      W=&FloatFromCurve(
       FloatCurve=DataInterface() {propertyValues {Curve {
          defaultValue=1
          keys [
           {arriveTangent=0.2038, interpMode=RCIM_Cubic, leaveTangent=0.2038, tangentMode=RCTM_User, time=0, value=0}
           {arriveTangent=1.1179, interpMode=RCIM_Cubic, leaveTangent=1.1179, tangentMode=RCTM_User, time=0.3871, value=0.675}
           {arriveTangent=0.0224, interpMode=RCIM_Cubic, leaveTangent=0.0224, tangentMode=RCTM_User, time=1, value=1}
          ]
         }}}
       CurveIndex=$Engine.System.Age
      )
     )
    )
    Lifetime Max=0.8
    Lifetime Min=0.4
    Lifetime Mode=::NewEnumerator1
    Mass Mode=::NewEnumerator3
    Mesh Scale=&Multiply_VectorByFloat(
     Vector=&RandomRangeVector(Minimum=<0.35 0.35 0.35>, Maximum=<0.5 0.5 0.5>)
     Float=&Multiply_Float(A=$User.FX_Scale, B=&RandomRangeFloat(Minimum=4, Maximum=4))
    )
    Mesh Scale Mode=::NewEnumerator4
    Sprite Rotation Mode=::NewEnumerator2
    UsePositionOffset=false
    Write Material Random=true
   )
   &SystemLocation(
    MaskSpawnGroup=false
    Offset=&MakeVector(
     X=0
     Y=&Multiply_Float(A=$User.FX_Scale, B=&Lerp_Float(A=-50, B=50, Alpha=$Particles.Particle_Height))
     Z=&Multiply_Float(A=$User.FX_Scale, B=&Lerp_Float(A=-10, B=10, Alpha=$Particles.Particle_Height))
    )
   )
   &InitialMeshOrientation(Rotation=&RandomRangeVector(Minimum=<-0.04 -0.04 -1>, Maximum=<0 0 1>), Use Rotation Vector=true)
   &CameraOffset(Camera Offset Amount=15)
  ]
  particleUpdateScript [
   &ParticleState()
   &ScaleMeshSize(Scale Factor=&Multiply_VectorByFloat(
     Vector=&MakeVector(
      X=1
      Y=1
      Z=&SelectFloatFromArray(
       Array Sampling Mode=::NewEnumerator0
       Direct Array Index=?
       Float Selection Array=DataInterface() {propertyValues {FloatData [1,-1]}}
       Random Seed=0
      )
     )
     Float=&Lerp_Float(
      A=0.8
      B=1.2
      Alpha=&FloatFromCurve(
       FloatCurve=DataInterface() {propertyValues {Curve {keys [
           {arriveTangent=0.0148, interpMode=RCIM_Cubic, leaveTangent=0.0148, tangentMode=RCTM_User, time=0, value=0.1858}
           {arriveTangent=2.8247, interpMode=RCIM_Cubic, leaveTangent=2.8247, tangentMode=RCTM_User, time=0.5048, value=0.8428}
          ]}}}
       Scale Curve=5
      )
     )
    ))
   &UpdateMeshOrientation(Rotation Rate=&RandomRangeFloat(Minimum=1, Maximum=3), Rotation Vector=<0 0 1>)
   &ScaleColor(
    Scale Alpha=&FloatFromCurve(
     FloatCurve=DataInterface() {propertyValues {Curve {keys [
         {arriveTangent=6.2301, interpMode=RCIM_Cubic, leaveTangent=6.2301, tangentMode=RCTM_User, time=0.0047, value=2.0548}
         {arriveTangent=-6.912, interpMode=RCIM_Cubic, leaveTangent=-6.912, tangentMode=RCTM_User, time=0.1, value=3}
         {arriveTangent=0.4048, interpMode=RCIM_Cubic, leaveTangent=0.4048, tangentMode=RCTM_User, time=0.7, value=0}
        ]}}}
     Scale Curve=&RandomRangeFloat(Minimum=0.8, Maximum=1)
    )
    Scale RGB=&VectorFromFloat(Value=&Lerp_Float(
      A=1
      B=3
      Alpha=&FloatFromCurve(FloatCurve=DataInterface() {propertyValues {Curve {keys [
           {interpMode=RCIM_Cubic, tangentMode=RCTM_User, time=0.0095, value=0.6235}
           {interpMode=RCIM_Cubic, tangentMode=RCTM_User, time=0.1, value=1}
           {interpMode=RCIM_Cubic, tangentMode=RCTM_User, time=1, value=0}
          ]}}})
     ))
    ScaleA=true
    ScaleRGB=true
   )
   &InheritVelocity(Inherited Velocity Amount Scale=<0.01 0.01 0.01>, Inherited Velocity Speed Limit=1, Speed Limit=true, UseSpeedThreshold=false)
   &SolveForcesAndVelocity(Clamp Velocity=false, Limit Acceleration=false)
   &DynamicMaterialParameters(
    Index 0 Param 1=&FloatFromCurve(FloatCurve=DataInterface() {propertyValues {Curve {keys [
         {arriveTangent=1.3639, interpMode=RCIM_Cubic, leaveTangent=1.3639, tangentMode=RCTM_User, time=0.0023, value=0.2098}
         {arriveTangent=1.6379, interpMode=RCIM_Cubic, leaveTangent=1.6379, tangentMode=RCTM_User, time=1, value=1}
        ]}}})
    Module.Index0Param0Write=true
    Module.Index0Param1Write=true
    Module.Index0Param2Write=false
    Module.Index0Param3Write=true
   )
   &ScaleColor(
    Scale Alpha=&FNTimeOfDayScalar(Night Scale=0.25, Morning Scale=0.25, Day Scale=0.7, Evening Scale=0.3)
    Scale RGB=&MakeVectorFromLinearColorRGB(Color=&FNTimeOfDayColor(
      Night Color=<rgba 0.48 0.5895 0.8 1>
      Morning Color=<rgba 0.8 0.7106 0.7259 1>
      Day Color=<rgba 1 0.9 0.8 1>
      Evening Color=<rgba 0.8 0.6855 0.7085 1>
     ))
    ScaleA=true
    ScaleRGB=true
   ) {module=ScaleColor001}
  ]
  rendererClasses [&NiagaraMeshRendererProperties]
  renderers [{rendererClass=&NiagaraMeshRendererProperties, rendererIndex=0}]
 }
}
summary {
 systemName=NS_IceGauntlets_Primary_Projectile
 emitters [
  {emitterName=Snow_Puffs, rendererClasses [&NiagaraSpriteRendererProperties]}
  {
   emitterName=Ice_Shards
   simTarget=GPUComputeSim
   rendererClasses [&NiagaraSpriteRendererProperties, &NiagaraMeshRendererProperties]
  }
  {emitterName=WindGusts, rendererClasses [&NiagaraMeshRendererProperties]}
  {emitterName=Trail, rendererClasses [&NiagaraRibbonRendererProperties]}
  {emitterName=Empty, rendererClasses []}
  {emitterName=LW_Head, rendererClasses []}
 ]
 userVariables [
  User.FX_DomeAngle {type=Vector3f, description="", defaultValue {struct=Vector3f, value=<0 0 0>}}
  User.FX_Scale {type=NiagaraFloat, defaultValue=0.5, description=""}
  User.Muzzle Pos {type=Vector3f, description="", defaultValue {struct=Vector3f, value=<0 0 0>}}
  User.PjrLife {type=NiagaraFloat, defaultValue=2, description=""}
 ]
}
