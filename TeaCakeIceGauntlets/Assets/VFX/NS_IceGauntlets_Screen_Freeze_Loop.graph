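# Looping Niagara screen-freeze overlay for the Ice Gauntlets, layering a frost sprite with small ice specks.
# System Timeline's normalized time drives frost material and opacity and scales speck alpha, so the effect ramps with it.
# FrostScreen holds a looping burst sprite, tinted by time of day and killed through an inactive-state plus timeline-zero check.
# ScreenSpecs continuously spawns short-lived sprites on camera-oriented cylinder surfaces with curve-biased random radii, forming a ring.
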NiagaraSystem /TeaCakeIceGauntlets/Assets/VFX/NS_IceGauntlets_Screen_Freeze_Loop
dependencies {
 usedDataInterfaces [&NiagaraDataInterfaceCurve, &NiagaraDataInterfaceCamera, &NiagaraDataInterfaceArrayFloat]
 usedDynamicInputs []
 usedModules [&SystemState, &Timeline]
}
emitters {
 FrostScreen {
  emitterUpdateScript [
   &EmitterState()
   &SpawnBurst_Instantaneous(Use Spawn Probability=false, UseLoopCountLimit=true)
  ]
  particleSpawnScript [&InitializeParticle(
    Color=<rgba 1.118 1.5153 2 2>
    Sprite Size Mode=::NewEnumerator3
    Uniform Sprite Size=300
    UsePositionOffset=false
    Write Material Random=true
   )]
  particleUpdateScript [
   &ParticleState(Kill Particles When Lifetime Has Elapsed=false, Loop Particles Lifetime=true)
   &DynamicMaterialParameters(
    Index 0 Param 1=&Lerp_Float(
     A=10
     B=1.5
     Alpha=&FloatFromCurve(
      FloatCurve=DataInterface() {propertyValues {Curve {keys [<at 0 0 RCIM_Cubic>,<at 0.2 1 RCIM_Cubic>]}}}
      CurveIndex=$System.Timeline.NormalizedTime
     )
    )
    Module.Index0Param0Write=true
    Module.Index0Param1Write=true
    Module.Index0Param2Write=true
    Module.Index0Param3Write=true
   )
   &DynamicMaterialParameters(
    Index 0 Param 1=&FloatFromCurve(FloatCurve=DataInterface() {propertyValues {Curve {keys [
         <at 0 5 RCIM_Cubic>
         {arriveTangent=-3.9157, interpMode=RCIM_Cubic, leaveTangent=-3.9157, tangentMode=RCTM_User, time=0.1, value=2}
         {arriveTangent=8.8889, interpMode=RCIM_Cubic, leaveTangent=8.8889, time=0.7, value=4}
         <at 1 10 RCIM_Cubic>
        ]}}})
    Module.Index0Param0Write=true
    Module.Index0Param1Write=true
    Module.Index0Param2Write=true
    Module.Index0Param3Write=true
   ) {enabled=false, module=DynamicMaterialParameters001}
   &ScaleColor(
    Scale Alpha=&Lerp_Float(A=0, B=0.4, Alpha=$System.Timeline.NormalizedTime)
    Scale RGB=&MakeVectorFromLinearColorRGB(Color=&FNTimeOfDayColor(
      Night Color=<rgba 0.0779 0.1995 0.25 1>
      Morning Color=<rgba 0.4661 0.6014 0.6302 1>
      Day Color=<rgba 0.4277 0.531 0.5625 1>
      Evening Color=<rgba 0.2344 0.2037 0.2015 1>
     ))
    ScaleA=true
    ScaleRGB=true
   )
   &KillParticles(Kill Particles=&Boolean_AndOperation(
     Bool1=&Boolean_NotOperation(Bool=/Game/Effects/Niagara/DynamicInputs/State/FN_IsExecutionStateActive())
     Bool2=&SetBoolByFloatComparison(A=$System.Timeline.NormalizedTime, B=0, Comparison Type=::NewEnumerator7)
    ))
  ]
  rendererClasses [&NiagaraSpriteRendererProperties]
  renderers [{rendererClass=&NiagaraSpriteRendererProperties}]
 }
 ScreenSpecs {
  emitterUpdateScript [
   &EmitterState()
   &SpawnRate(SpawnRate=100, Use Spawn Probability=false)
  ]
  particleSpawnScript [
   &InitializeParticle(
    Color Mode=::NewEnumerator0
    Lifetime Max=3
    Lifetime Min=2
    Lifetime Mode=::NewEnumerator1
    Sprite Rotation Mode=::NewEnumerator2
    Sprite Size Mode=::NewEnumerator1
    UsePositionOffset=false
    Write Material Random=true
   )
   &ShapeLocation(
    Cylinder Height=10
    Cylinder Radius=&Lerp_Float(
     A=50
     B=140
     Alpha=&FloatFromCurve(
      FloatCurve=DataInterface() {propertyValues {Curve {keys [{arriveTangent=2.6301, time=0, value=0}, {arriveTangent=1, leaveTangent=1, time=0.15, value=0.5}, <at 1 1>]}}}
      CurveIndex=&RandomRangeFloat(Minimum=0, Maximum=1)
     )
    )
    MaskSpawnGroup=false
    Override Seed=false
    Rotation Axis=&ReturnCameraVector(Camera Vectors=::NewEnumerator1, Coordinate Space=::Simulation)
    Rotation Coordinate Space=::Simulation
    Rotation Mode=::NewEnumerator1
    Shape Primitive=::NewEnumerator1
    Surface Only=true
    Use Endcaps In Surface Only Mode=false
    UseOwnerScale=true
   )
   &Color(Color=<rgba 0.5388 0.8583 1 0.25>)
  ]
  particleUpdateScript [
   &ParticleState()
   &ScaleSpriteSize(Uniform Curve Sprite Scale=DataInterface() {propertyValues {Curve {keys [
        {arriveTangent=4.1785, interpMode=RCIM_Cubic, leaveTangent=4.1785, tangentMode=RCTM_User, time=0, value=0}
        {arriveTangent=-0.1448, interpMode=RCIM_Cubic, leaveTangent=-0.1448, tangentMode=RCTM_User, time=0.025, value=1}
        {arriveTangent=0.0224, interpMode=RCIM_Cubic, leaveTangent=0.0224, tangentMode=RCTM_User, time=1, value=1}
       ]}}})
   &ScaleColor(
    Scale Alpha=&FloatFromCurve(FloatCurve=DataInterface() {propertyValues {Curve {keys [<at 0.5 1>,<at 1 0>]}}}, Scale Curve=$System.Timeline.NormalizedTime)
    ScaleA=true
    ScaleRGB=true
   )
   /SetVariables(Particles.float:NiagaraFloat=&SelectFloatFromArray(Direct Array Index=0, Direct Array Mode=::NewEnumerator1, Float Selection Array=DataInterface() {propertyValues {FloatData [0,1,2]}})) {isSetParameters=true}
   &SolveForcesAndVelocity(Clamp Velocity=false, Limit Acceleration=false)
  ]
  rendererClasses [&NiagaraSpriteRendererProperties]
  renderers [{rendererClass=&NiagaraSpriteRendererProperties}]
 }
}
summary {
 systemName=NS_IceGauntlets_Screen_Freeze_Loop
 emitters [
  {emitterName=ScreenSpecs, rendererClasses [&NiagaraSpriteRendererProperties]}
  {emitterName=FrostScreen, rendererClasses [&NiagaraSpriteRendererProperties]}
 ]
 userVariables []
}
