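# Secondary Niagara VFX for the Ice Gauntlets: a short GPU burst scattering icy particles across the ground around the owner.
# Particles launch outward, fall under heavy gravity and stick on first collision, driving splash sprites and mesh growth.
# A scratch input uses player position to accumulate a reaction that shrinks meshes and feeds material parameters.
# Time-of-day functions tint the effect; SDF and radius checks cull strays.
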NiagaraSystem /TeaCakeIceGauntlets/Assets/VFX/NS_IceGauntlets_Secondary_EPIC
dependencies {
 usedDataInterfaces [
  &NiagaraDataInterfaceCurve
  &NiagaraDataInterfaceColorCurve
  &NiagaraDataInterfaceSpriteRendererInfo
  &NiagaraDataInterfaceCollisionQuery
  &NiagaraDataInterfaceCamera
 ]
 usedDynamicInputs [/Niagara/DynamicInputs/Camera/CameraFade]
 usedModules [
  &self:SystemSpawnScript.SystemScriptSource.SystemScriptGraph.NiagaraNodeAssignment_1.SetVariables_#1
  &self:SystemSpawnScript.SystemScriptSource.SystemScriptGraph.NiagaraNodeAssignment_0.SetVariables_#2
  &SystemState
 ]
}
emitters {GroundInteraction {
  simTarget=GPUComputeSim
  emitterUpdateScript [
   &EmitterState(Life Cycle Mode=::NewEnumerator1, Loop Behavior=::NewEnumerator1, Loop Delay=0.14, Loop Duration=0.2, UseLoopDelay=false)
   &SpawnRate(
    Spawn Probability=&MakeCustomFloatFromBool(Bool=true, True Float=0, False Float=1)
    SpawnRate=100
    Use Spawn Probability=false
   )
  ]
  particleSpawnScript [
   &InitializeParticle(
    Color=<rgba 0.3021 0.4995 1 0.1>
    Lifetime=&RandomRangeFloat(Minimum=5, Maximum=5.8)
    Mesh Scale Mode=::NewEnumerator3
    Mesh Uniform Scale=&Multiply_Float(A=&RandomRangeFloat(Minimum=0.7, Maximum=1.4), B=1)
    Position=$Engine.Owner.Position
    Position Mode=::NewEnumerator1
    Sprite Rotation Mode=::NewEnumerator2
    Sprite Size Max=<70 50>
    Sprite Size Min=<65 45>
    Sprite Size Mode=::NewEnumerator2
    Sprite UV Mode=::NewEnumerator3
    UsePositionOffset=false
    Write Material Random=true
   ) {module=InitializeParticle001}
   &ShapeLocation(
    Handle Radius=35
    Large Radius=20
    MaskSpawnGroup=false
    Non Uniform Scale=<1 1 0>
    Offset=<0 0 30>
    Offset Coordinate Space=::Local
    Offset Mode=::NewEnumerator0
    Override Seed=false
    Rotation Axis=<0 0 1>
    Rotation Mode=::NewEnumerator1
    Shape Primitive=::NewEnumerator3
    UseOwnerScale=true
   ) {enabled=false}
   &ShapeLocation(
    Hemisphere Distribution=<1 0.5>
    MaskSpawnGroup=false
    Non Uniform Scale=<1 1 0.1>
    Offset=<0 0 5>
    Offset Coordinate Space=::Local
    Offset Mode=::NewEnumerator0
    Override Seed=false
    UseOwnerScale=true
   ) {module=ShapeLocation001}
   /SetVariables(Particles.SpriteMesh_Quat:Quat4f=/Niagara/DynamicInputs/Rotation/RandomQuaternion(), Particles.Sprite_CamOfs=25) {isSetParameters=true}
   &InitialMeshOrientation(Rotation=&RandomRangeVector(Minimum=<-0.04 -0.04 -1>, Maximum=<0.04 0.04 1>), Use Rotation Vector=true)
   /SetVariables(
    Particles.InitPos:NiagaraPosition=$Particles.Initial.Position
    Particles.StoredVector:Vector3f=&TransformVector(
     Vector=Linked() {linkedVariable {name=Particles.Initial.Position, type=NiagaraPosition}}
     Source Space=::Simulation
     Destination Space=::World
     Apply Scale=true
     Transform Vector as Position=true
    )
   ) {isSetParameters=true}
   &AddVelocity(
    Module.Use Falloff Distance=false
    Origin Offset=&MakeVector(X=0, Y=0, Z=-10)
    Override Seed=false
    Velocity Mode=::NewEnumerator1
    Velocity Origin=$Engine.Owner.Position
    Velocity Speed=&RandomRangeFloat(Minimum=2200, Maximum=2900)
   )
   &CameraOffset(Camera Offset Amount=-5)
  ]
  particleUpdateScript [
   &ParticleState()
   &GravityForce(
    Coordinate Space=::Local
    Gravity=&Multiply_VectorByFloat(
     Vector=<0 0 -12000>
     Float=&MakeCustomFloatFromBool(
      Bool=&SetBoolByFloatComparison(A=15, B=HlslExpression() {hlslExpression="length(Particles.Velocity)"})
      True Float=0
      False Float=1
     )
    )
   )
   &Collision(
    Age Colliding Particles=false
    Collision Enabled=&InvertBool(Boolean:NiagaraBool=$Particles.bool)
    Enable Rest State=false
    Friction=10000
    Friction During a Bounce=10000
    GPU Collision Type=::NewEnumerator2
    Method for Calculating Particle Radius:ENiagaraCalculateRadiusOptions=::NewEnumerator1
    Module.EnableMaxCollisionCount=true
    Module.Randomize Collision Normal Vector=false
    Particle Radius Scale=0.04
    Randomize Collision Normal=0
    Restitution=0
   )
   &self:Trigger_Do_Once() {module=Trigger_Do_Once}
   &InterpolateOverTime(
    Rate Of Change:NiagaraFloat=&MakeCustomFloatFromBool(Bool=$Particles.State Changed, True Float=1, False Float=0)
    Target Float Value=1
   )
   &SolveForcesAndVelocity(Clamp Velocity=false, Limit Acceleration=false)
   &ScaleSpriteSize(
    Uniform Curve Scale=&MakeCustomFloatFromBool(Bool=$Particles.HasCollided, True Float=3, False Float=0)
    Uniform Curve Sprite Scale=DataInterface() {propertyValues {Curve {keys [{tangentMode=RCTM_User, time=0, value=0.2}, <at 0.0723 1>, {tangentMode=RCTM_User, time=1, value=1}]}}}
   )
   &DynamicMaterialParameters(
    Index 0 Param 1=&MakeFloatFromVector(Vector=$Particles.StoredVector, Channel=::NewEnumerator0)
    Index 0 Param 2=&MakeFloatFromVector(Vector=$Particles.StoredVector, Channel=::NewEnumerator1)
    Index 0 Param 3=&MakeFloatFromVector(Vector=$Particles.StoredVector, Channel=::NewEnumerator2)
    Module.Param0WriteEnabled=true
    Module.Param1WriteEnabled=true
    Module.Param2WriteEnabled=true
    Module.Param3WriteEnabled=true
   ) {module=DynamicMaterialParameters001}
   &ScaleColor(
    Scale Alpha=&MakeCustomFloatFromBool(
     Bool=&SetBoolByFloatComparison(A=10, B=HlslExpression() {hlslExpression="length(Particles.Velocity)"})
     True Float=&FloatFromCurve(
      FloatCurve=DataInterface() {propertyValues {Curve {keys [
          <at -0.0034 0.625 RCIM_Cubic>
          {arriveTangent=-0.0405, interpMode=RCIM_Cubic, leaveTangent=-0.0405, tangentMode=RCTM_User, time=0.0645, value=1.225}
          {interpMode=RCIM_Cubic, leaveTangent=1.75, time=0.15, value=0}
         ]}}}
      Scale Curve=&FNTimeOfDayScalar(Night Scale=0.6, Morning Scale=0.7, Evening Scale=0.7)
     )
     False Float=0
    )
    ScaleA=true
    ScaleRGB=false
   )
   /SetVariables(Particles.Reaction:NiagaraFloat=&self:ScratchDynamicInput(PlayerLocation:LinearColor=$NPC.FortniteNPC.PlayerPosition, Rection Distance:NiagaraFloat=100, Collided:NiagaraBool=$Particles.HasCollided)) {isSetParameters=true}
   &InterpolateOverTime(Rate Of Change:NiagaraFloat=$Particles.Reaction, Target Float Value=1) {module=InterpolateOverTime001}
   /SetVariables(
    Particles.Color2=&MultiplyLinearColor(
     Linear Color=&Multiply_LinearColorByFloat(
      Linear Color=&ColorFromCurve(ColorCurve=DataInterface() {propertyValues {
         AlphaCurve {keys [{arriveTangent=1.1089, interpMode=RCIM_Cubic, time=0.107, value=1}, <at 0.9307 1>, <at 1 0 RCIM_Cubic>]}
         BlueCurve {keys [<at 0 3 RCIM_Cubic>]}
         GreenCurve {keys [<at 0 1.1336 RCIM_Cubic>]}
         RedCurve {keys [<at 0 0.5312 RCIM_Cubic>]}
        }})
      Float=1
     )
     Scale Factor=&MakeVector4_FromLinearColor(LinearColor=&FNTimeOfDayColor(Night Color=<rgba 2 2 2 1>, Morning Color=<rgba 1.7 1.2722 1.2397 1>, Day Color=<rgba 1 1 1 1>, Evening Color=<rgba 1.5 0.9322 1.0476 1>))
    )
    Particles.Color3:LinearColor=&MakeLinearColorFromVectorAndFloat(
     "Vector (RGB)"=&MakeVectorFromLinearColorRGB(Color=<rgba 0.1516 0.3481 0.6771 1>)
     "Float (Alpha)"=&MakeCustomFloatFromBool(
      Bool=$Particles.HasCollided
      True Float=&FloatFromCurve(FloatCurve=DataInterface() {propertyValues {Curve {keys [<at 0.55 1>,<at 1 0>]}}}, Scale Curve=0.7)
      False Float=0
     )
    )
    Particles.DM2001:Vector4f=&MakeFloat4FromVectorAndFloat("Vector (RGB)"=&TransformVector(Vector=<0 0 0>, Destination Space=::World), "Float (Alpha)"=1)
    Particles.DM2_0:Vector4f=&MakeVector4(
     X=$Particles.InterpolateOverTime001.InterpolateOverTime_Float.Moving Average
     Y=&FloatFromCurve(
      FloatCurve=DataInterface() {propertyValues {Curve {
         defaultValue=1
         keys [
          {arriveTangent=2.3691, interpMode=RCIM_Cubic, leaveTangent=2.3691, tangentMode=RCTM_User, time=0, value=0}
          {tangentMode=RCTM_User, time=1, value=1}
         ]
        }}}
      CurveIndex=$Particles.InterpolateOverTime001.InterpolateOverTime_Float.Moving Average
      Scale Curve=200
     )
     Z=&MakeCustomFloatFromBool(Bool=$Particles.HasCollided, True Float=1, False Float=0)
     W=&FloatFromCurve(
      FloatCurve=DataInterface() {propertyValues {Curve {
         defaultValue=1
         keys [
          {arriveTangent=4.0103, interpMode=RCIM_Cubic, leaveTangent=4.0103, tangentMode=RCTM_User, time=0, value=0}
          {arriveTangent=-0.3536, interpMode=RCIM_Cubic, leaveTangent=-0.3536, tangentMode=RCTM_User, time=0.2, value=0.4376}
          {arriveTangent=-6.5847, interpMode=RCIM_Cubic, leaveTangent=-6.5847, tangentMode=RCTM_User, time=0.9963, value=-2.7753}
         ]
        }}}
      CurveIndex=$Particles.InterpolateOverTime001.InterpolateOverTime_Float.Moving Average
      Scale Curve=500
     )
    )
    Particles.Mesh Scale2:Vector3f=<10 10 10>
    Particles.SplashSize:Vector2f=&Multiply_Vector2D_ByFloat(
     Float=&FloatFromCurve(FloatCurve=DataInterface() {propertyValues {Curve {keys [
          {arriveTangent=7.2322, interpMode=RCIM_Cubic, leaveTangent=7.2322, tangentMode=RCTM_User, time=0, value=0.02}
          {tangentMode=RCTM_User, time=0.15, value=2}
          <at 0.1501 0>
         ]}}})
     Vector2D=<115 115>
    )
    Particles.Vector 2D:Vector2f=&Multiply_Vector2D_ByFloat(
     Float=&MakeCustomFloatFromBool(
      Bool=&SetBoolByFloatComparison(A=15, B=HlslExpression() {hlslExpression="length(Particles.Velocity)"})
      True Float=0
      False Float=1
     )
     Vector2D=&RandomRangeVector2D(Minimum=<9 12>, Maximum=<14 35>)
    )
    Particles.Vector 4:Vector4f=&MakeVector4(
     X=&FloatFromCurve(FloatCurve=DataInterface() {propertyValues {Curve {keys [{arriveTangent=-1.7191, interpMode=RCIM_Cubic, leaveTangent=-1.7191, tangentMode=RCTM_User, time=0, value=1}, <at 0.7 0>]}}})
     Y=0
     Z=0
     W=0
    )
    Particles.bool:NiagaraBool=$Particles.HasCollided
   ) {isSetParameters=true}
   &ScaleMeshSize(Scale Factor=&Multiply_VectorByFloat(
     Vector=&VectorFromFloat(Value=&FloatFromCurve(FloatCurve=DataInterface() {propertyValues {Curve {
          defaultValue=1
          keys [
           {arriveTangent=4.1785, interpMode=RCIM_Cubic, leaveTangent=4.1785, tangentMode=RCTM_User, time=0, value=0}
           {arriveTangent=1.1179, interpMode=RCIM_Cubic, leaveTangent=1.1179, tangentMode=RCTM_User, time=0.2042, value=0.6228}
           {arriveTangent=0.0224, interpMode=RCIM_Cubic, leaveTangent=0.0224, tangentMode=RCTM_User, time=0.8712, value=1}
           {arriveTangent=-7.9129, interpMode=RCIM_Cubic, leaveTangent=-7.9129, tangentMode=RCTM_User, time=1.0174, value=0.2}
          ]
         }}}))
     Float=&MakeCustomFloatFromBool(
      Bool=$Particles.State Changed
      True Float=&FloatFromCurve(
       FloatCurve=DataInterface() {propertyValues {Curve {keys [{arriveTangent=-1.621, interpMode=RCIM_Cubic, leaveTangent=-1.621, tangentMode=RCTM_User, time=0, value=1}, <at 1 0>]}}}
       CurveIndex=$Particles.InterpolateOverTime001.InterpolateOverTime_Float.Moving Average
      )
      False Float=0
     )
    ))
   &SubUVAnimation(
    Evaluation Type=::NewEnumerator0
    "Sprite Renderer "=DataInterface() {propertyValues {SpriteRenderer=&self:GroundInteraction_0.NiagaraSpriteRendererProperties_4}}
    SubUV Animation Mode:ENiagara_SubUVLookupModeV2=::NewEnumerator2
    UseEndFrame=false
    UseStartFrame=false
   )
   &self:Kill_Particles_by_SDF() {module=Kill_Particles_by_SDF}
   &KillParticlesInVolume(Invert Volume=true, Sphere Radius=650)
  ]
  rendererClasses [&NiagaraSpriteRendererProperties, &NiagaraMeshRendererProperties]
  renderers [
   {rendererClass=&NiagaraSpriteRendererProperties}
   {rendererClass=&NiagaraMeshRendererProperties, rendererIndex=1}
   {rendererClass=&NiagaraSpriteRendererProperties, rendererIndex=2}
  ]
 }}
summary {
 systemName=NS_IceGauntlets_Secondary_EPIC
 emitters [{
   emitterName=GroundInteraction
   simTarget=GPUComputeSim
   rendererClasses [&NiagaraSpriteRendererProperties, &NiagaraMeshRendererProperties]
  }]
 userVariables [User.IsTrain {type=NiagaraBool}]
}
